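# Control Rig for the large male Cattus player body that drives digitigrade dog_ leg bones from the standard human leg bones.
# Each side places dog_ball at its initial offset relative to the animated foot.
# SpringIK then solves dog_thigh to dog_ball using a thigh-relative pole vector, and FABRIK re-solves the chain onto the foot-derived target.
# Leg IK and knee pole Control nodes receive transforms but feed nothing.
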ControlRigBlueprint /Game/Characters/Player/Male/Large/Bodies/M_LRG_Cattus/Meshes/M_LRG_Cattus_CtrlRig
graphs [
 RigVMFunctionLibrary {links [], nodes []}
 RigVMModel {nodes [
   n1 = &RigUnit_BeginExecution()
   n2 = &RigUnit_GetBoneTransform(Bone=foot_r)
   n3 = &RigVMFunction_MathTransformMul(
    B=@n2.Transform
    A=&RigVMFunction_MathTransformMakeRelative(
     Parent=/Script/ControlRig.RigUnit_GetInitialBoneTransform(Bone=foot_r, Space=GlobalSpace) {is=struct RigVMUnitNode}.Transform
     Global=&RigUnit_GetInitialBoneTransform(Bone=dog_ball_r, Space=GlobalSpace).Transform
    ).Local
   ) {template="Multiply::Execute(in A,in B,out Result)"}
   /Script/ControlRig.RigUnit_SetBoneTransform(Bone=dog_ball_r, Space=GlobalSpace, Weight=1, bPropagateToChildren=True, ExecutePin=@n1.ExecutePin, Transform=@n3.Result) {is=struct RigVMUnitNode}
   n4 = &RigVMFunction_MathTransformMul(
    A=(Rotation=(X=0,Y=0,Z=0,W=1),Translation=(X=50,Y=-100,Z=0),Scale3D=(X=1,Y=1,Z=1))
    B=&RigUnit_GetBoneTransform(Bone=thigh_r).Transform
   ) {template="Multiply::Execute(in A,in B,out Result)"}
   > /Script/ControlRig.RigUnit_SpringIK(
    Damping=0.75
    DebugSettings=(bEnabled=true,Scale=1,Color=(R=0,G=0,B=1,A=1),WorldOffset=(Rotation=(X=0,Y=0,Z=0,W=1),Translation=(X=0,Y=0,Z=0),Scale3D=(X=1,Y=1,Z=1)))
    EffectorRatio=0.75
    EffectorStrength=200
    EndBone=dog_ball_r
    HierarchyStrength=1024
    Iterations=8
    PoleVectorKind=Location
    PoleVectorSpace=None
    PrimaryAxis=(X=1,Y=0,Z=0)
    RootRatio=0.1
    RootStrength=200
    SecondaryAxis=(X=0,Y=-1,Z=0)
    StartBone=dog_thigh_r
    bFlipPolePlane=False
    bLimitLocalPosition=True
    bLiveSimulation=True
    bPropagateToChildren=True
    PoleVector=@n4.Result.Translation
   ) {is=struct RigVMUnitNode}
   > /Script/ControlRig.RigUnit_FABRIK(
    EffectorBone=dog_ball_r
    MaxIterations=10
    Precision=1
    StartBone=dog_thigh_r
    Weight=1
    bPropagateToChildren=True
    bSetEffectorTransform=True
    EffectorTransform=@n3.Result
   ) {is=struct RigVMUnitNode}
   n5 = &RigUnit_BeginExecution()
   n6 = &RigUnit_GetBoneTransform(Bone=foot_l)
   n7 = &RigVMFunction_MathTransformMul(
    B=@n6.Transform
    A=&RigVMFunction_MathTransformMakeRelative(
     Parent=&RigUnit_GetInitialBoneTransform(Bone=foot_l, Space=GlobalSpace).Transform
     Global=&RigUnit_GetInitialBoneTransform(Bone=dog_ball_l, Space=GlobalSpace).Transform
    ).Local
   ) {template="Multiply::Execute(in A,in B,out Result)"}
   &RigUnit_SetBoneTransform(Bone=dog_ball_l, Space=GlobalSpace, Weight=1, bPropagateToChildren=True, ExecutePin=@n5.ExecutePin, Transform=@n7.Result)
   n8 = &RigVMFunction_MathTransformMul(
    A=(Rotation=(X=0,Y=0,Z=0,W=1),Translation=(X=-50,Y=100,Z=0),Scale3D=(X=1,Y=1,Z=1))
    B=&RigUnit_GetBoneTransform(Bone=thigh_l).Transform
   ) {template="Multiply::Execute(in A,in B,out Result)"}
   > &RigUnit_SpringIK(
    Damping=0.75
    DebugSettings=(bEnabled=true,Scale=1,Color=(R=0,G=0,B=1,A=1),WorldOffset=(Rotation=(X=0,Y=0,Z=0,W=1),Translation=(X=0,Y=0,Z=0),Scale3D=(X=1,Y=1,Z=1)))
    EffectorRatio=0.5
    EffectorStrength=200
    EndBone=dog_ball_l
    HierarchyStrength=1024
    Iterations=8
    PoleVectorKind=Location
    PoleVectorSpace=None
    PrimaryAxis=(X=-1,Y=0,Z=0)
    RootRatio=0.1
    RootStrength=200
    SecondaryAxis=(X=0,Y=1,Z=0)
    StartBone=dog_thigh_l
    bFlipPolePlane=False
    bLimitLocalPosition=True
    bLiveSimulation=True
    bPropagateToChildren=True
    PoleVector=@n8.Result.Translation
   )
   > &RigUnit_FABRIK(
    EffectorBone=dog_ball_l
    MaxIterations=10
    Precision=1
    StartBone=dog_thigh_l
    Weight=1
    bPropagateToChildren=True
    bSetEffectorTransform=True
    EffectorTransform=@n7.Result
   )
   Knee Pole Vector R = /Script/ControlRig.RigUnit_Control(
    Base=(Rotation=(X=0,Y=0,Z=0,W=1),Translation=(X=0,Y=0,Z=0),Scale3D=(X=1,Y=1,Z=1))
    Filter=(TranslationFilter=(bX=True,bY=True,bZ=True),RotationFilter=(bX=True,bY=True,bZ=True),ScaleFilter=(bX=True,bY=True,bZ=True))
    InitTransform=@n4.Result
   ) {is=struct RigVMUnitNode}
   Knee Pole Vector R_0 = &RigUnit_Control(
    Base=(Rotation=(X=0,Y=0,Z=0,W=1),Translation=(X=0,Y=0,Z=0),Scale3D=(X=1,Y=1,Z=1))
    Filter=(TranslationFilter=(bX=True,bY=True,bZ=True),RotationFilter=(bX=True,bY=True,bZ=True),ScaleFilter=(bX=True,bY=True,bZ=True))
    InitTransform=@n8.Result
   )
   Leg IK R = &RigUnit_Control(
    Base=(Rotation=(X=0,Y=0,Z=0,W=1),Translation=(X=0,Y=0,Z=0),Scale3D=(X=1,Y=1,Z=1))
    Filter=(TranslationFilter=(bX=True,bY=True,bZ=True),RotationFilter=(bX=True,bY=True,bZ=True),ScaleFilter=(bX=True,bY=True,bZ=True))
    InitTransform=(Rotation=(X=0,Y=0,Z=0,W=1),Translation=(X=0,Y=0,Z=0),Scale3D=(X=1,Y=1,Z=1))
    InitTransform.Translation=@n2.Transform.Translation
   )
   Leg IK R_0 = &RigUnit_Control(
    Base=(Rotation=(X=0,Y=0,Z=0,W=1),Translation=(X=0,Y=0,Z=0),Scale3D=(X=1,Y=1,Z=1))
    Filter=(TranslationFilter=(bX=True,bY=True,bZ=True),RotationFilter=(bX=True,bY=True,bZ=True),ScaleFilter=(bX=True,bY=True,bZ=True))
    InitTransform=(Rotation=(X=0,Y=0,Z=0,W=1),Translation=(X=0,Y=0,Z=0),Scale3D=(X=1,Y=1,Z=1))
    InitTransform.Translation=@n6.Transform.Translation
   )
  ]}
]
hierarchy [
 root {loc=<0 0 0>, rot=<0 0 0 1>}
 pelvis {parent=root, loc=<0 -0.7321 92.4218>, rot=<0.0127 0.707 -0.0127 -0.707>}
 spine_01 {parent=pelvis, loc=<5.6002 -0.5128 0>, rot=<0 0 0.034 -0.9994>}
 spine_02 {parent=spine_01, loc=<8.0753 0.0139 0>, rot=<0 0 0.0238 -0.9997>}
 spine_03 {parent=spine_02, loc=<7.0393 0 0>, rot=<0 0 0.0149 -0.9999>}
 spine_04 {parent=spine_03, loc=<8.1024 0 0>, rot=<0 0 -0.0132 -0.9999>}
 spine_05 {parent=spine_04, loc=<10.0146 0 0>, rot=<0 0 -0.0015 -1>}
 clavicle_l {parent=spine_05, loc=<2.5856 -0.2866 -4.3777>, rot=<0.0465 -0.7289 -0.009 -0.6829>}
 upperarm_l {parent=clavicle_l, loc=<18.2264 0.1342 -0.001>, rot=<0.0253 -0.2693 0.0013 -0.9627>}
 lowerarm_l {parent=upperarm_l, loc=<29.0071 0.1686 -1.239>, rot=<0.0391 -0.0099 -0.2453 -0.9686>}
 hand_l {parent=lowerarm_l, loc=<25.8607 0.222 1.0274>, rot=<-0.7193 -0.0988 0.1193 -0.6773>}
 index_metacarpal_l {parent=hand_l, loc=<1.6535 0.0183 -2.96>, rot=<0.0158 -0.0124 -0.0173 -0.9996>}
 index_01_l {parent=index_metacarpal_l, loc=<8.3008 0.0357 -0.0524>, rot=<-0.0044 0.0485 0.0924 -0.9945>}
 index_02_l {parent=index_01_l, loc=<3.3587 -0.3003 0.0452>, rot=<-0.002 0.0027 0.3225 -0.9466>}
 index_03_l {parent=index_02_l, loc=<2.5578 -0.1231 -0.0252>, rot=<-0.002 0.0067 0.2558 -0.9667>}
 middle_metacarpal_l {parent=hand_l, loc=<2.1856 -0.1725 -0.5984>, rot=<0.006 0.0168 -0.0329 -0.9993>}
 middle_01_l {parent=middle_metacarpal_l, loc=<7.3275 -0.2447 0.0129>, rot=<-0.0208 0.0042 0.0858 -0.9961>}
 middle_02_l {parent=middle_01_l, loc=<4.0123 0.1401 0.1175>, rot=<-0.0218 0.0258 0.3512 -0.9357>}
 middle_03_l {parent=middle_02_l, loc=<2.6077 0.0534 0.1373>, rot=<0.0082 0.0188 0.1708 -0.9851>}
 pinky_metacarpal_l {parent=hand_l, loc=<1.7616 -0.9753 3.3814>, rot=<-0.0397 0.0814 -0.069 -0.9935>}
 pinky_01_l {parent=pinky_metacarpal_l, loc=<7.1709 -0.4831 -0.1094>, rot=<-0.0654 -0.0035 0.2242 -0.9723>}
 pinky_02_l {parent=pinky_01_l, loc=<2.8274 0.0893 -0.1043>, rot=<0.0037 0.0011 0.2854 -0.9584>}
 pinky_03_l {parent=pinky_02_l, loc=<2.2046 0.0307 -0.0078>, rot=<-0.0002 0.0098 0.1445 -0.9895>}
 ring_metacarpal_l {parent=hand_l, loc=<2.2074 -0.3625 1.3638>, rot=<-0.0284 0.0722 -0.0458 -0.9959>}
 ring_01_l {parent=ring_metacarpal_l, loc=<7.0295 -0.6197 -0.1446>, rot=<-0.0489 -0.0252 0.1892 -0.9804>}
 ring_02_l {parent=ring_01_l, loc=<3.6174 0.4647 -0.1027>, rot=<0 0 0.2806 -0.9598>}
 ring_03_l {parent=ring_02_l, loc=<2.4152 0.114 0.0199>, rot=<0 0 0.1189 -0.9929>}
 thumb_01_l {parent=hand_l, loc=<1.3118 -0.3191 -2.459>, rot=<0.6546 -0.4042 0.0664 -0.6354>}
 thumb_02_l {parent=thumb_01_l, loc=<5.9336 -0.0448 0.0401>, rot=<-0.0037 0.0213 0.298 -0.9543>}
 thumb_03_l {parent=thumb_02_l, loc=<2.2223 0.3682 -0.0907>, rot=<0.0118 -0.013 0.0066 -0.9998>}
 lowerarm_twist_01_l {parent=lowerarm_l, loc=<17.8549 0 0>, rot=<0 0 0 -1>}
 lowerarm_twist_02_l {parent=lowerarm_l, loc=<9.6199 0 0>, rot=<0 0 0 -1>}
 upperarm_twist_01_l {parent=upperarm_l, loc=<7.9796 0.0843 -0.6195>, rot=<0 0 0 1>}
 upperarm_twist_02_l {parent=upperarm_l, loc=<19.5059 0.0843 -0.6195>, rot=<0 0 0 1>}
 clavicle_r {parent=spine_05, loc=<2.5856 -0.2865 4.3777>, rot=<0.7289 0.0465 -0.6829 0.009>}
 upperarm_r {parent=clavicle_r, loc=<-18.2264 -0.1342 0.001>, rot=<0.0253 -0.2693 0.0013 -0.9627>}
 lowerarm_r {parent=upperarm_r, loc=<-29.0072 -0.1686 1.239>, rot=<0.0391 -0.0099 -0.2453 -0.9686>}
 hand_r {parent=lowerarm_r, loc=<-25.8608 -0.222 -1.0274>, rot=<-0.7193 -0.0988 0.1193 -0.6773>}
 index_metacarpal_r {parent=hand_r, loc=<-1.6534 -0.0181 2.9601>, rot=<0.0158 -0.0124 -0.0173 -0.9996>}
 index_01_r {parent=index_metacarpal_r, loc=<-8.3008 -0.0357 0.0524>, rot=<-0.0044 0.0485 0.0924 -0.9945>}
 index_02_r {parent=index_01_r, loc=<-3.3588 0.3003 -0.0453>, rot=<-0.002 0.0027 0.3225 -0.9466>}
 index_03_r {parent=index_02_r, loc=<-2.5578 0.123 0.0252>, rot=<-0.002 0.0067 0.2558 -0.9667>}
 middle_metacarpal_r {parent=hand_r, loc=<-2.1856 0.1726 0.5984>, rot=<0.006 0.0168 -0.0329 -0.9993>}
 middle_01_r {parent=middle_metacarpal_r, loc=<-7.3273 0.2447 -0.0129>, rot=<-0.0208 0.0042 0.0858 -0.9961>}
 middle_02_r {parent=middle_01_r, loc=<-4.0124 -0.14 -0.1175>, rot=<-0.0218 0.0258 0.3512 -0.9357>}
 middle_03_r {parent=middle_02_r, loc=<-2.6077 -0.0533 -0.1373>, rot=<0.0082 0.0188 0.1708 -0.9851>}
 pinky_metacarpal_r {parent=hand_r, loc=<-1.7617 0.9753 -3.3814>, rot=<-0.0397 0.0814 -0.069 -0.9935>}
 pinky_01_r {parent=pinky_metacarpal_r, loc=<-7.1707 0.4832 0.1094>, rot=<-0.0654 -0.0035 0.2242 -0.9723>}
 pinky_02_r {parent=pinky_01_r, loc=<-2.8275 -0.0892 0.1043>, rot=<0.0037 0.0011 0.2854 -0.9584>}
 pinky_03_r {parent=pinky_02_r, loc=<-2.2045 -0.031 0.0078>, rot=<-0.0002 0.0098 0.1445 -0.9895>}
 ring_metacarpal_r {parent=hand_r, loc=<-2.2073 0.3627 -1.3638>, rot=<-0.0284 0.0722 -0.0458 -0.9959>}
 ring_01_r {parent=ring_metacarpal_r, loc=<-7.0297 0.6197 0.1446>, rot=<-0.0489 -0.0252 0.1892 -0.9804>}
 ring_02_r {parent=ring_01_r, loc=<-3.6172 -0.4647 0.1027>, rot=<0 0 0.2806 -0.9598>}
 ring_03_r {parent=ring_02_r, loc=<-2.4153 -0.1141 -0.0199>, rot=<0 0 0.1189 -0.9929>}
 thumb_01_r {parent=hand_r, loc=<-1.3119 0.3193 2.459>, rot=<0.6546 -0.4042 0.0664 -0.6354>}
 thumb_02_r {parent=thumb_01_r, loc=<-5.9336 0.0447 -0.0401>, rot=<-0.0037 0.0213 0.298 -0.9543>}
 thumb_03_r {parent=thumb_02_r, loc=<-2.2223 -0.3682 0.0907>, rot=<0.0118 -0.013 0.0066 -0.9998>}
 lowerarm_twist_01_r {parent=lowerarm_r, loc=<-17.855 0 0>, rot=<0 0 0 1>}
 lowerarm_twist_02_r {parent=lowerarm_r, loc=<-9.6199 0 0>, rot=<0 0 0 1>}
 upperarm_twist_01_r {parent=upperarm_r, loc=<-7.9796 -0.0843 0.6195>, rot=<0 0 0 1>}
 upperarm_twist_02_r {parent=upperarm_r, loc=<-19.506 -0.0843 0.6195>, rot=<0 0 0 1>}
 neck_01 {parent=spine_05, loc=<10.815 1.9196 0>, rot=<0 0 -0.1864 -0.9825>}
 neck_02 {parent=neck_01, loc=<4.7347 0.3045 0>, rot=<0 0 0.0363 -0.9993>}
 head {parent=neck_02, loc=<5.7632 0.5552 0>, rot=<0 0 0.115 -0.9934>}
 thigh_l {parent=pelvis, loc=<-1.8716 -0.0109 -11.9936>, rot=<0.0012 0.0572 0.0155 -0.9982>}
 calf_l {parent=thigh_l, loc=<-41.9178 -0.2246 0.0926>, rot=<0.0006 -0.0161 -0.0391 -0.9991>}
 calf_twist_02_l {parent=calf_l, loc=<-15.9268 -0.0967 0.0397>, rot=<0.0237 0.0095 -0.0035 -0.9997>}
 foot_l {parent=calf_l, loc=<-38.7057 1.2438 0.2452>, rot=<0.0975 -0.0474 0.0317 -0.9936>}
 ball_l {parent=foot_l, loc=<-6.1614 13.1469 0.6807>, rot=<0.0001 -0.0001 -0.7063 -0.7079>}
 thigh_twist_01_l {parent=thigh_l, loc=<-21.6303 0 0>, rot=<0.0013 0 0 -1>}
 thigh_r {parent=pelvis, loc=<-1.8714 -0.0109 11.9936>, rot=<0.0572 -0.0012 0.9982 0.0155>}
 calf_r {parent=thigh_r, loc=<41.918 0.2246 -0.0926>, rot=<0.0006 -0.0161 -0.0391 -0.9991>}
 calf_twist_02_r {parent=calf_r, loc=<15.9269 0.0967 -0.0397>, rot=<0.0237 0.0095 -0.0035 -0.9997>}
 foot_r {parent=calf_r, loc=<38.7058 -1.2438 -0.2453>, rot=<0.0975 -0.0474 0.0317 -0.9936>}
 ball_r {parent=foot_r, loc=<6.1614 -13.1469 -0.6807>, rot=<0.0001 -0.0001 -0.7063 -0.7079>}
 thigh_twist_01_r {parent=thigh_r, loc=<21.6303 0 0>, rot=<0.0013 0 0 -1>}
 dog_thigh_r {parent=pelvis, loc=<-7.42 0.1622 10.6325>, rot=<0.0728 -0.0015 0.9972 0.0154>}
 dog_calf_r {parent=dog_thigh_r, loc=<36.017 0.2246 -0.0926>, rot=<-0.0035 -0.0157 -0.2926 -0.9561>}
 dog_foot_r {parent=dog_calf_r, loc=<28.3047 -1.2438 -0.2453>, rot=<0.0054 0.0081 0.5533 -0.8329>}
 dog_ball_r {parent=dog_foot_r, loc=<23.8964 0 0>, rot=<0.0679 0.0199 0.4687 -0.8805>}
 dog_toe_mid_01_r {parent=dog_ball_r, loc=<4.5918 1.1763 -1.3432>, rot=<0.0146 -0.0368 0.3012 -0.9528>}
 dog_toe_mid_02_r {parent=dog_toe_mid_01_r, loc=<6.013 0 0>, rot=<0 0 -0.5138 -0.8579>}
 dog_toe_mid_03_r {parent=dog_toe_mid_02_r, loc=<10.7523 0 0>, rot=<0 0 0 -1>}
 dog_toe_in_01_r {parent=dog_ball_r, loc=<4.3597 1.1516 -5.4353>, rot=<0.0411 -0.1235 0.3012 -0.9447>}
 dog_toe_in_02_r {parent=dog_toe_in_01_r, loc=<6.013 0 0>, rot=<0 0 -0.6012 -0.7991>}
 dog_toe_in_03_r {parent=dog_toe_in_02_r, loc=<8.8538 0 0>, rot=<0 0 0 1>}
 dog_toe_out_01_r {parent=dog_ball_r, loc=<3.5636 1.6337 2.3686>, rot=<-0.0339 0.1221 0.2947 -0.9471>}
 dog_toe_out_02_r {parent=dog_toe_out_01_r, loc=<6.7064 0 0>, rot=<0 0 -0.5138 -0.8579>}
 dog_toe_out_03_r {parent=dog_toe_out_02_r, loc=<10.7523 0 0>, rot=<0 0 0 1>}
 dog_thigh_twist_01_r {parent=dog_thigh_r, loc=<21.6303 0 0>, rot=<0.0013 0 0 -1>}
 dog_thigh_l {parent=pelvis, loc=<-7.4201 0.1622 -10.6325>, rot=<0.0015 0.0728 0.0154 -0.9972>}
 dog_calf_l {parent=dog_thigh_l, loc=<-36.017 -0.2246 0.0925>, rot=<-0.0035 -0.0157 -0.2926 -0.9561>}
 dog_foot_l {parent=dog_calf_l, loc=<-28.3047 1.2438 0.2453>, rot=<0.0054 0.0081 0.5533 -0.8329>}
 dog_ball_l {parent=dog_foot_l, loc=<-23.8964 0 -0.0001>, rot=<0.0679 0.0199 0.4687 -0.8805>}
 dog_toe_mid_01_l {parent=dog_ball_l, loc=<-4.5918 -1.1763 1.3432>, rot=<0.0146 -0.0368 0.3012 -0.9528>}
 dog_toe_mid_02_l {parent=dog_toe_mid_01_l, loc=<-6.013 0 0>, rot=<0 0 -0.5138 -0.8579>}
 dog_toe_mid_03_l {parent=dog_toe_mid_02_l, loc=<-10.7523 0 0>, rot=<0 0 0 -1>}
 dog_toe_in_01_l {parent=dog_ball_l, loc=<-4.3597 -1.1516 5.4353>, rot=<0.0411 -0.1235 0.3012 -0.9447>}
 dog_toe_in_02_l {parent=dog_toe_in_01_l, loc=<-6.013 0 0>, rot=<0 0 -0.6012 -0.7991>}
 dog_toe_in_03_l {parent=dog_toe_in_02_l, loc=<-8.8538 0 0>, rot=<0 0 0 1>}
 dog_toe_out_01_l {parent=dog_ball_l, loc=<-3.5636 -1.6337 -2.3686>, rot=<-0.0339 0.1221 0.2947 -0.9471>}
 dog_toe_out_02_l {parent=dog_toe_out_01_l, loc=<-6.7064 0 0>, rot=<0 0 -0.5138 -0.8579>}
 dog_toe_out_03_l {parent=dog_toe_out_02_l, loc=<-10.7523 0 0>, rot=<0 0 0 -1>}
 dog_thigh_twist_01_l {parent=dog_thigh_l, loc=<-21.6302 0 0>, rot=<0.0013 0 0 -1>}
]
