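# Animation Blueprint for the large male Cattus player body, copying the attached parent's pose and adapting it to this skeleton.
# Targeting blends in head and hair offsets before a Control Rig, then constraints drive phantom fingers and look-ats plus twist constraints align the dog-style legs.
# The event graph raises a timed, interpolated hair offset during the roar emote and tracks jump boots.
# The shoe-mesh LegAdjust morph logic has no incoming execution link.
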AnimBlueprint /Game/Characters/Player/Male/Large/Bodies/M_LRG_Cattus/Meshes/M_LRG_Cattus_AnimBP
target_skeleton=&/M_LRG_Cattus_Skeleton
graphs [
 AnimGraph {nodes [
   n1 = "Transform (Modify) Bone - Bone: head"!(ComponentPose=Local To Component(LocalPose=Use cached pose 'SGI Cache'()), Alpha=1, Rotation="0,-30,0") -> (Pose) {rotationMode=BMM_Additive, scaleMode=BMM_Additive, alphaBoolBlend {blendOption=Linear}, boneToModify {boneName=head}, rotation {yaw=-30}}
   n2 = "Transform (Modify) Bone - Bone: dyn_hairBase"!(ComponentPose=@n1, Alpha=1, Rotation="0,18,0") -> (Pose) {
    rotationMode=BMM_Additive
    scaleMode=BMM_Additive
    alphaBoolBlend {blendOption=Linear}
    boneToModify {boneName=dyn_hairBase}
    rotation {yaw=18}
   }
   n3 = Control Rig(Source=Blend Poses by bool(
     BlendPose_0=Component To Local(ComponentPose=@n2)
     BlendPose_1=Use cached pose 'SGI Cache'()
     BlendTime_0=0.15
     bActiveValue=Get bIsTargeting()
    ) {blendType=Linear}) {
    alphaBoolBlend {blendOption=Linear}
    controlRigAssetReference {blueprintRigClass=&/M_LRG_Cattus_CtrlRig.M_LRG_Cattus_CtrlRig_C}
   }
   n4 = "Constraint - phantom_thumb_02_r (R)"(ComponentPose=Local To Component(LocalPose=@n3), Alpha=1, ConstraintWeights_0=1) {
    alphaBoolBlend {blendOption=Linear}
    boneToModify {boneName=phantom_thumb_02_r}
    constraintSetup [{transformType=Rotation, targetBone {boneName=thumb_02_r}}]
    constraintWeights [1]
   }
   n5 = "Constraint - phantom_thumb_03_r (R)"(ComponentPose=@n4, Alpha=1, ConstraintWeights_0=1) {
    alphaBoolBlend {blendOption=Linear}
    boneToModify {boneName=phantom_thumb_03_r}
    constraintSetup [{transformType=Rotation, targetBone {boneName=thumb_03_r}}]
    constraintWeights [1]
   }
   n6 = "Constraint - phantom_index_01_l (R)"(ComponentPose=@n5, Alpha=1, ConstraintWeights_0=1) {
    alphaBoolBlend {blendOption=Linear}
    boneToModify {boneName=phantom_index_01_l}
    constraintSetup [{transformType=Rotation, targetBone {boneName=index_01_l}}]
    constraintWeights [1]
   }
   n7 = "Constraint - phantom_index_02_l (R)"(ComponentPose=@n6, Alpha=1, ConstraintWeights_0=1) {
    alphaBoolBlend {blendOption=Linear}
    boneToModify {boneName=phantom_index_02_l}
    constraintSetup [{transformType=Rotation, targetBone {boneName=index_02_l}}]
    constraintWeights [1]
   }
   n8 = "Constraint - phantom_index_03_l (R)"(ComponentPose=@n7, Alpha=1, ConstraintWeights_0=1) {
    alphaBoolBlend {blendOption=Linear}
    boneToModify {boneName=phantom_index_03_l}
    constraintSetup [{transformType=Rotation, targetBone {boneName=index_03_l}}]
    constraintWeights [1]
   }
   n9 = "Constraint - phantom_middle_01_r (R)"(ComponentPose=@n8, Alpha=1, ConstraintWeights_0=1) {
    alphaBoolBlend {blendOption=Linear}
    boneToModify {boneName=phantom_middle_01_r}
    constraintSetup [{transformType=Rotation, targetBone {boneName=middle_01_r}}]
    constraintWeights [1]
   }
   n10 = "Constraint - phantom_middle_02_r (R)"(ComponentPose=@n9, Alpha=1, ConstraintWeights_0=1) {
    alphaBoolBlend {blendOption=Linear}
    boneToModify {boneName=phantom_middle_02_r}
    constraintSetup [{transformType=Rotation, targetBone {boneName=middle_02_r}}]
    constraintWeights [1]
   }
   n11 = "Constraint - phantom_middle_03_r (R)"(ComponentPose=@n10, Alpha=1, ConstraintWeights_0=1) {
    alphaBoolBlend {blendOption=Linear}
    boneToModify {boneName=phantom_middle_03_r}
    constraintSetup [{transformType=Rotation, targetBone {boneName=middle_03_r}}]
    constraintWeights [1]
   }
   n12 = "Constraint - phantom_ring_01_r (R)"(ComponentPose=@n11, Alpha=1, ConstraintWeights_0=1) {
    alphaBoolBlend {blendOption=Linear}
    boneToModify {boneName=phantom_ring_01_r}
    constraintSetup [{transformType=Rotation, targetBone {boneName=ring_01_r}}]
    constraintWeights [1]
   }
   n13 = "Constraint - phantom_ring_02_r (R)"(ComponentPose=@n12, Alpha=1, ConstraintWeights_0=1) {
    alphaBoolBlend {blendOption=Linear}
    boneToModify {boneName=phantom_ring_02_r}
    constraintSetup [{transformType=Rotation, targetBone {boneName=ring_02_r}}]
    constraintWeights [1]
   }
   n14 = "Constraint - phantom_ring_03_r (R)"(ComponentPose=@n13, Alpha=1, ConstraintWeights_0=1) {
    alphaBoolBlend {blendOption=Linear}
    boneToModify {boneName=phantom_ring_03_r}
    constraintSetup [{transformType=Rotation, targetBone {boneName=ring_03_r}}]
    constraintWeights [1]
   }
   n15 = "Constraint - phantom_thumb_02_l (R)"(ComponentPose=@n14, Alpha=1, ConstraintWeights_0=1) {
    alphaBoolBlend {blendOption=Linear}
    boneToModify {boneName=phantom_thumb_02_l}
    constraintSetup [{transformType=Rotation, targetBone {boneName=thumb_02_l}}]
    constraintWeights [1]
   }
   n16 = "Constraint - phantom_thumb_03_l (R)"(ComponentPose=@n15, Alpha=1, ConstraintWeights_0=1) {
    alphaBoolBlend {blendOption=Linear}
    boneToModify {boneName=phantom_thumb_03_l}
    constraintSetup [{transformType=Rotation, targetBone {boneName=thumb_03_l}}]
    constraintWeights [1]
   }
   n17 = "Constraint - phantom_index_01_r (R)"(ComponentPose=@n16, Alpha=1, ConstraintWeights_0=1) {
    alphaBoolBlend {blendOption=Linear}
    boneToModify {boneName=phantom_index_01_r}
    constraintSetup [{transformType=Rotation, targetBone {boneName=index_01_r}}]
    constraintWeights [1]
   }
   n18 = "Constraint - phantom_index_02_r (R)"(ComponentPose=@n17, Alpha=1, ConstraintWeights_0=1) {
    alphaBoolBlend {blendOption=Linear}
    boneToModify {boneName=phantom_index_02_r}
    constraintSetup [{transformType=Rotation, targetBone {boneName=index_02_r}}]
    constraintWeights [1]
   }
   n19 = "Constraint - phantom_index_03_r (R)"(ComponentPose=@n18, Alpha=1, ConstraintWeights_0=1) {
    alphaBoolBlend {blendOption=Linear}
    boneToModify {boneName=phantom_index_03_r}
    constraintSetup [{transformType=Rotation, targetBone {boneName=index_03_r}}]
    constraintWeights [1]
   }
   n20 = "Constraint - phantom_middle_01_l (R)"(ComponentPose=@n19, Alpha=1, ConstraintWeights_0=1) {
    alphaBoolBlend {blendOption=Linear}
    boneToModify {boneName=phantom_middle_01_l}
    constraintSetup [{transformType=Rotation, targetBone {boneName=middle_01_l}}]
    constraintWeights [1]
   }
   n21 = "Constraint - phantom_middle_02_l (R)"(ComponentPose=@n20, Alpha=1, ConstraintWeights_0=1) {
    alphaBoolBlend {blendOption=Linear}
    boneToModify {boneName=phantom_middle_02_l}
    constraintSetup [{transformType=Rotation, targetBone {boneName=middle_02_l}}]
    constraintWeights [1]
   }
   n22 = "Constraint - phantom_middle_03_l (R)"(ComponentPose=@n21, Alpha=1, ConstraintWeights_0=1) {
    alphaBoolBlend {blendOption=Linear}
    boneToModify {boneName=phantom_middle_03_l}
    constraintSetup [{transformType=Rotation, targetBone {boneName=middle_03_l}}]
    constraintWeights [1]
   }
   n23 = "Constraint - phantom_ring_01_l (R)"(ComponentPose=@n22, Alpha=1, ConstraintWeights_0=1) {
    alphaBoolBlend {blendOption=Linear}
    boneToModify {boneName=phantom_ring_01_l}
    constraintSetup [{transformType=Rotation, targetBone {boneName=ring_01_l}}]
    constraintWeights [1]
   }
   n24 = "Constraint - phantom_ring_02_l (R)"(ComponentPose=@n23, Alpha=1, ConstraintWeights_0=1) {
    alphaBoolBlend {blendOption=Linear}
    boneToModify {boneName=phantom_ring_02_l}
    constraintSetup [{transformType=Rotation, targetBone {boneName=ring_02_l}}]
    constraintWeights [1]
   }
   n25 = "Constraint - phantom_ring_03_l (R)"(ComponentPose=@n24, Alpha=1, ConstraintWeights_0=1) {
    alphaBoolBlend {blendOption=Linear}
    boneToModify {boneName=phantom_ring_03_l}
    constraintSetup [{transformType=Rotation, targetBone {boneName=ring_03_l}}]
    constraintWeights [1]
   }
   n26 = "Transform (Modify) Bone - Bone: dyn_hairBase"(ComponentPose=@n25, Alpha=Get isRoaringAlpha() -> (isRoaringAlpha:double), Rotation="0,15,0") {
    rotationMode=BMM_Additive
    rotationSpace=BCS_BoneSpace
    scaleMode=BMM_Additive
    alphaBoolBlend {blendOption=Linear}
    alphaScaleBiasClamp {bInterpResult=true, interpSpeedDecreasing=3, interpSpeedIncreasing=3}
    boneToModify {boneName=dyn_hairBase}
    rotation {yaw=15}
   }
   n27 = Look At - Bone: dog_thigh_l(ComponentPose=@n26, Alpha=1) {
    alphaBoolBlend {blendOption=Linear}
    boneToModify {boneName=dog_thigh_l}
    lookAtLocation {x=0}
    lookAtTarget {boneReference {boneName=calf_l}}
    lookAt_Axis {axis=<-1 0>}
   }
   n28 = Look At - Bone: dog_thigh_r(ComponentPose=@n27, Alpha=1) {
    alphaBoolBlend {blendOption=Linear}
    boneToModify {boneName=dog_thigh_r}
    lookAtLocation {x=0}
    lookAtTarget {boneReference {boneName=calf_r}}
    lookAt_Axis {axis=<1 0>}
   }
   n29 = "Constraint - dog_thigh_twist_01_r (R)"(ComponentPose=@n28, Alpha=1, ConstraintWeights_0=1) {
    alphaBoolBlend {blendOption=Linear}
    boneToModify {boneName=dog_thigh_twist_01_r}
    constraintSetup [{transformType=Rotation, targetBone {boneName=thigh_twist_01_r}}]
    constraintWeights [1]
   }
   n30 = "Constraint - dog_thigh_twist_01_l (R)"(ComponentPose=@n29, Alpha=1, ConstraintWeights_0=1) {
    alphaBoolBlend {blendOption=Linear}
    boneToModify {boneName=dog_thigh_twist_01_l}
    constraintSetup [{transformType=Rotation, targetBone {boneName=thigh_twist_01_l}}]
    constraintWeights [1]
   }
   Output Pose(Result=Component To Local(ComponentPose=@n30)) {layerGroup=None}
   Save cached pose 'SGI Cache'(Pose=Copy Pose From Mesh() {bUseAttachedParent=true}) {cacheName=SGI Cache}
  ]}
 EventGraph {nodes [
   Event BlueprintUpdateAnimation()
   > n1 = Sequence!(execute) -> (then_0, then_1, then_2:Exec)
   n2 = GetAnimInstance(self=GetOwnerSkeletalMeshComponent())
   > Is Valid(InputObject=@n2)
   > n3 = Branch(Condition=Montage_IsPlaying(self=@n2, Montage=/Game/Animation/Game/MainPlayer/Emotes/CattusRoar/Emote_CattusRoar_CMM_M))
   > Do Once(Reset=@n4.then)
   > Set isRoaringAlpha(isRoaringAlpha:double=1) -> (Output_Get:double)
   > Delay(Duration=4)
   > Set isRoaringAlpha(isRoaringAlpha=0)
   n4 = Set isRoaringAlpha(execute=@n3.else, isRoaringAlpha=0)
   n5 = Cast To FortPlayerPawn(execute=@n1.then_1, Object=GetOwningActor())
   > Set As Fort Player Pawn(As Fort Player Pawn:Fort Player Pawn Object Reference=@n5) -> (Output_Get:Fort Player Pawn Object Reference)
   > n6 = Get!(execute:Exec) -> (then:Exec, else:Exec, As Fort Player Pawn:Fort Player Pawn Object Reference)
   > Set is_wearing_jumpboots(is_wearing_jumpboots:bool=HasMatchingGameplayTag!(self:Gameplay Tag Asset Interface Interface=@n6, TagToCheck:Gameplay Tag Structure=(TagName="Item.Gadget.JumpBoots.AnimTracker")) -> (ReturnValue:bool)) -> (Output_Get:bool)
   n7 = Get!(execute) -> (then, else, ShoeMesh:Skeletal Mesh Component Budgeted Object Reference)
   > n8 = Is Valid(InputObject=Get SkeletalMesh!(self:Skinned Mesh Component Object Reference=@n7) -> (SkeletalMesh:Skeletal Mesh Object Reference))
   SetMorphTarget!(execute=[@n8.Is Not Valid, @n7.else], self:Anim Instance Object Reference, MorphTargetName:Name=LegAdjust, Value:float=0) -> (then)
   SetMorphTarget(execute=@n8.Is Valid, MorphTargetName=LegAdjust, Value=1)
   Event OnCharacterCustomizationLoaded()
   Set ShoeMesh(ShoeMesh:Skeletal Mesh Component Budgeted Object Reference=Get Array Item with EFort Custom Part Type(PartType=ExtraPart)) -> (Output_Get:Skeletal Mesh Component Budgeted Object Reference)
  ]}
]
variables [
 As Fort Player Pawn {type=FortPlayerPawn}
 is_wearing_jumpboots {type=boolean}
 isRoaringAlpha {type=number}
 ShoeMesh {type=SkeletalMeshComponentBudgeted}
]
