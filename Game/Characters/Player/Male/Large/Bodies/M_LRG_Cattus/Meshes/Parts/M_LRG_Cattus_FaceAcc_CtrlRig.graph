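# Control Rig for the Cattus large male face accessory that procedurally sways five head tentacle bone chains (center, plus two per side).
# A forward-solve sequence runs one ChainHarmonics node per chain from its second bone, using only Wave motion, with staggered offsets and stronger amplitude on the center.
# The isRoaringAlpha variable scales the shared wave speed up to five times, and useHarmonics toggles the wave on every chain.
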ControlRigBlueprint /Game/Characters/Player/Male/Large/Bodies/M_LRG_Cattus/Meshes/Parts/M_LRG_Cattus_FaceAcc_CtrlRig
graphs [
 RigVMFunctionLibrary {links [], nodes []}
 RigVMModel {nodes [
   n1 = &RigUnit_BeginExecution()
   n2 = /Script/RigVM.RigVMFunction_MathFloatLerp(A=1, B=5, T=RigVMVariableNode() {variable=isRoaringAlpha}.Value) {is=struct RigVMUnitNode, template="Interpolate::Execute(in A,in B,in T,out Result)"}
   n3 = /Script/RigVM.RigVMFunction_MathVectorMul(
    A=(X=1,Y=1,Z=1)
    A.X=@n2.Result
    A.Y=@n2.Result
    A.Z=@n2.Result
    B=/Script/RigVM.RigVMFunction_MathVectorAbs(Value=(X=2,Y=2,Z=2)) {is=struct RigVMUnitNode, template="Absolute::Execute(in Value,out Result)"}.Result
   ) {is=struct RigVMUnitNode, template="Multiply::Execute(in A,in B,out Result)"}
   n4 = RigVMVariableNode() {variable=useHarmonics}
   /Script/ControlRig.RigUnit_ChainHarmonics(
    ChainRoot=tent_c_2
    DrawWorldOffset=(Rotation=(X=0,Y=0,Z=0,W=1),Translation=(X=0,Y=0,Z=0),Scale3D=(X=1,Y=1,Z=1))
    Pendulum=(bEnabled=False,PendulumStiffness=2,PendulumGravity=(X=0,Y=0,Z=-0.98),PendulumBlend=0.75,PendulumDrag=0.98,PendulumMinimum=0,PendulumMaximum=0.1,PendulumEase=Linear,UnwindAxis=(X=0,Y=1,Z=0),UnwindMinimum=0.2,UnwindMaximum=0.05)
    Reach=(bEnabled=False,ReachTarget=(X=0,Y=0,Z=0),ReachAxis=(X=1,Y=0,Z=0),ReachMinimum=0,ReachMaximum=0,ReachEase=Linear)
    Wave=(bEnabled=True,WaveFrequency=(X=3,Y=4,Z=3.5),WaveAmplitude=(X=0.02,Y=0.2,Z=0.4),WaveOffset=(X=0,Y=0.1,Z=0.2),WaveNoise=(X=0,Y=0.2,Z=0.12),WaveMinimum=0.5,WaveMaximum=1,WaveEase=ExponentialEaseInOut)
    WaveCurve=(EditorCurveData=(Keys=((Value=0.5),(Time=1,Value=1)),PreInfinityExtrap=RCCE_Constant,PostInfinityExtrap=RCCE_Constant,DefaultValue=340282346638528859811704183484516925440))
    bDrawDebug=false
    ExecutePin=@n1.ExecutePin
    Speed=@n3.Result
    Wave.bEnabled=@n4.Value
   ) {is=struct RigVMUnitNode}
   > &RigUnit_ChainHarmonics(
    ChainRoot=tent_1_l_2
    DrawWorldOffset=(Rotation=(X=0,Y=0,Z=0,W=1),Translation=(X=0,Y=0,Z=0),Scale3D=(X=1,Y=1,Z=1))
    Pendulum=(bEnabled=False,PendulumStiffness=2,PendulumGravity=(X=0,Y=0,Z=-0.98),PendulumBlend=0.75,PendulumDrag=0.98,PendulumMinimum=0,PendulumMaximum=0.1,PendulumEase=Linear,UnwindAxis=(X=0,Y=1,Z=0),UnwindMinimum=0.2,UnwindMaximum=0.05)
    Reach=(bEnabled=False,ReachTarget=(X=0,Y=0,Z=0),ReachAxis=(X=1,Y=0,Z=0),ReachMinimum=0,ReachMaximum=0,ReachEase=Linear)
    Wave=(bEnabled=True,WaveFrequency=(X=3,Y=4,Z=3.5),WaveAmplitude=(X=0.02,Y=0.1,Z=0.2),WaveOffset=(X=0,Y=3,Z=4),WaveNoise=(X=0,Y=0.2,Z=0.12),WaveMinimum=0.5,WaveMaximum=1,WaveEase=ExponentialEaseInOut)
    WaveCurve=(EditorCurveData=(Keys=((Value=0.5),(Time=1,Value=1)),PreInfinityExtrap=RCCE_Constant,PostInfinityExtrap=RCCE_Constant,DefaultValue=340282346638528859811704183484516925440))
    bDrawDebug=false
    Speed=@n3.Result
    Wave.bEnabled=@n4.Value
   )
   > &RigUnit_ChainHarmonics(
    ChainRoot=tent_2_l_2
    DrawWorldOffset=(Rotation=(X=0,Y=0,Z=0,W=1),Translation=(X=0,Y=0,Z=0),Scale3D=(X=1,Y=1,Z=1))
    Pendulum=(bEnabled=False,PendulumStiffness=2,PendulumGravity=(X=0,Y=0,Z=-0.98),PendulumBlend=0.75,PendulumDrag=0.98,PendulumMinimum=0,PendulumMaximum=0.1,PendulumEase=Linear,UnwindAxis=(X=0,Y=1,Z=0),UnwindMinimum=0.2,UnwindMaximum=0.05)
    Reach=(bEnabled=False,ReachTarget=(X=0,Y=0,Z=0),ReachAxis=(X=1,Y=0,Z=0),ReachMinimum=0,ReachMaximum=0,ReachEase=Linear)
    Wave=(bEnabled=True,WaveFrequency=(X=3,Y=4,Z=3.5),WaveAmplitude=(X=0.02,Y=0.1,Z=0.2),WaveOffset=(X=0,Y=5,Z=6),WaveNoise=(X=0,Y=0.2,Z=0.12),WaveMinimum=0.5,WaveMaximum=1,WaveEase=ExponentialEaseInOut)
    WaveCurve=(EditorCurveData=(Keys=((Value=0.5),(Time=1,Value=1)),PreInfinityExtrap=RCCE_Constant,PostInfinityExtrap=RCCE_Constant,DefaultValue=340282346638528859811704183484516925440))
    bDrawDebug=false
    Speed=@n3.Result
    Wave.bEnabled=@n4.Value
   )
   > &RigUnit_ChainHarmonics(
    ChainRoot=tent_2_r_2
    DrawWorldOffset=(Rotation=(X=0,Y=0,Z=0,W=1),Translation=(X=0,Y=0,Z=0),Scale3D=(X=1,Y=1,Z=1))
    Pendulum=(bEnabled=False,PendulumStiffness=2,PendulumGravity=(X=0,Y=0,Z=-0.98),PendulumBlend=0.75,PendulumDrag=0.98,PendulumMinimum=0,PendulumMaximum=0.1,PendulumEase=Linear,UnwindAxis=(X=0,Y=1,Z=0),UnwindMinimum=0.2,UnwindMaximum=0.05)
    Reach=(bEnabled=False,ReachTarget=(X=0,Y=0,Z=0),ReachAxis=(X=1,Y=0,Z=0),ReachMinimum=0,ReachMaximum=0,ReachEase=Linear)
    Wave=(bEnabled=True,WaveFrequency=(X=3,Y=4,Z=3.5),WaveAmplitude=(X=0.02,Y=0.1,Z=0.2),WaveOffset=(X=0,Y=7,Z=8),WaveNoise=(X=0,Y=0.2,Z=0.12),WaveMinimum=0.5,WaveMaximum=1,WaveEase=Linear)
    WaveCurve=(EditorCurveData=(Keys=((Value=0.5),(Time=1,Value=1)),PreInfinityExtrap=RCCE_Constant,PostInfinityExtrap=RCCE_Constant,DefaultValue=340282346638528859811704183484516925440))
    bDrawDebug=false
    Speed=@n3.Result
    Wave.bEnabled=@n4.Value
   )
   > &RigUnit_ChainHarmonics(
    ChainRoot=tent_1_r_2
    DrawWorldOffset=(Rotation=(X=0,Y=0,Z=0,W=1),Translation=(X=0,Y=0,Z=0),Scale3D=(X=1,Y=1,Z=1))
    Pendulum=(bEnabled=False,PendulumStiffness=2,PendulumGravity=(X=0,Y=0,Z=-0.98),PendulumBlend=0.75,PendulumDrag=0.98,PendulumMinimum=0,PendulumMaximum=0.1,PendulumEase=Linear,UnwindAxis=(X=0,Y=1,Z=0),UnwindMinimum=0.2,UnwindMaximum=0.05)
    Reach=(bEnabled=False,ReachTarget=(X=0,Y=0,Z=0),ReachAxis=(X=1,Y=0,Z=0),ReachMinimum=0,ReachMaximum=0,ReachEase=Linear)
    Wave=(bEnabled=True,WaveFrequency=(X=3,Y=4,Z=3.5),WaveAmplitude=(X=0.02,Y=0.1,Z=0.2),WaveOffset=(X=0,Y=9,Z=10),WaveNoise=(X=0,Y=0.2,Z=0.12),WaveMinimum=0.5,WaveMaximum=1,WaveEase=ExponentialEaseInOut)
    WaveCurve=(EditorCurveData=(Keys=((Value=0.5),(Time=1,Value=1)),PreInfinityExtrap=RCCE_Constant,PostInfinityExtrap=RCCE_Constant,DefaultValue=340282346638528859811704183484516925440))
    bDrawDebug=false
    Speed=@n3.Result
    Wave.bEnabled=@n4.Value
   )
  ]}
]
hierarchy [
 root {loc=<0 0 0>, rot=<0 0 0 1>}
 pelvis {parent=root, loc=<0 -0.7321 92.4218>, rot=<0.0127 0.707 -0.0127 -0.707>}
 spine_01 {parent=pelvis, loc=<5.6002 -0.5128 0>, rot=<0 0 0.034 -0.9994>}
 spine_02 {parent=spine_01, loc=<8.0753 0.0139 0>, rot=<0 0 0.0238 -0.9997>}
 spine_03 {parent=spine_02, loc=<7.0393 0 0>, rot=<0 0 0.0149 -0.9999>}
 spine_04 {parent=spine_03, loc=<8.1024 0 0>, rot=<0 0 -0.0132 -0.9999>}
 spine_05 {parent=spine_04, loc=<10.0146 0 0>, rot=<0 0 -0.0015 -1>}
 clavicle_l {parent=spine_05, loc=<2.5856 -0.2866 -4.3777>, rot=<0.0465 -0.7289 -0.009 -0.6829>}
 upperarm_l {parent=clavicle_l, loc=<18.2264 0.1342 -0.001>, rot=<0.0253 -0.2693 0.0013 -0.9627>}
 clavicle_r {parent=spine_05, loc=<2.5856 -0.2865 4.3777>, rot=<0.7289 0.0465 -0.6829 0.009>}
 upperarm_r {parent=clavicle_r, loc=<-18.2264 -0.1342 0.001>, rot=<0.0253 -0.2693 0.0013 -0.9627>}
 neck_01 {parent=spine_05, loc=<10.815 1.9196 0>, rot=<0 0 -0.1864 -0.9825>}
 neck_02 {parent=neck_01, loc=<4.7347 0.3045 0>, rot=<0 0 0.0363 -0.9993>}
 head {parent=neck_02, loc=<5.7632 0.5552 0>, rot=<0 0 0.115 -0.9934>}
 dyn_hairBase {parent=head, loc=<10.4139 12.3465 0>, rot=<0 0 0 -1>}
 tent_1_r_1 {parent=dyn_hairBase, loc=<-1.3295 -11.2731 6.2932>, rot=<0.5883 -0.7664 0.1816 0.1833>}
 tent_1_r_2 {parent=tent_1_r_1, loc=<9.5159 0 0>, rot=<0 0.0838 -0.2101 -0.9741>}
 tent_1_r_3 {parent=tent_1_r_2, loc=<8.5045 0 0>, rot=<0 0.0709 -0.1417 -0.9874>}
 tent_1_r_4 {parent=tent_1_r_3, loc=<7.7419 0 0>, rot=<0 0.0648 -0.0741 -0.9951>}
 tent_1_r_5 {parent=tent_1_r_4, loc=<6.8391 0 0>, rot=<0 0.0334 -0.0774 -0.9964>}
 tent_1_r_6 {parent=tent_1_r_5, loc=<5.9605 0 0>, rot=<0 0.079 -0.0447 -0.9959>}
 tent_1_r_7 {parent=tent_1_r_6, loc=<5.0542 0 0>, rot=<0 0.0611 -0.078 -0.9951>}
 tent_1_r_8 {parent=tent_1_r_7, loc=<4.1709 0 0>, rot=<0 0.0342 -0.06 -0.9976>}
 tent_1_r_9 {parent=tent_1_r_8, loc=<3.2731 0 0>, rot=<0 0.0359 -0.0767 -0.9964>}
 tent_1_r_10 {parent=tent_1_r_9, loc=<2.385 0 0>, rot=<0 0 0 -1>}
 tent_2_r_1 {parent=dyn_hairBase, loc=<-0.6196 -11.5178 3.6679>, rot=<0.6129 -0.7864 0.0528 0.0556>}
 tent_2_r_2 {parent=tent_2_r_1, loc=<8.5502 0 0>, rot=<0 -0.0118 -0.1166 -0.9931>}
 tent_2_r_3 {parent=tent_2_r_2, loc=<7.7009 0 0>, rot=<0 0.0297 -0.1187 -0.9925>}
 tent_2_r_4 {parent=tent_2_r_3, loc=<6.9526 0 0>, rot=<0 0.0165 -0.0386 -0.9991>}
 tent_2_r_5 {parent=tent_2_r_4, loc=<6.1447 0 0>, rot=<0 0.0424 -0.0852 -0.9955>}
 tent_2_r_6 {parent=tent_2_r_5, loc=<5.3448 0 0>, rot=<0 0.0198 -0.0391 -0.999>}
 tent_2_r_7 {parent=tent_2_r_6, loc=<4.5446 0 0>, rot=<0 0.0355 -0.0304 -0.9989>}
 tent_2_r_8 {parent=tent_2_r_7, loc=<3.7437 0 0>, rot=<0 0.0156 -0.0207 -0.9997>}
 tent_2_r_9 {parent=tent_2_r_8, loc=<2.9418 0 0>, rot=<0 0.0248 -0.0087 -0.9997>}
 tent_2_r_10 {parent=tent_2_r_9, loc=<2.1397 0 0>, rot=<0 0 0 -1>}
 tent_c_1 {parent=dyn_hairBase, loc=<-0.0833 -6.0905 0.0099>, rot=<0.7651 -0.6439 -0.001 0.0012>}
 tent_c_2 {parent=tent_c_1, loc=<10.7948 0 0>, rot=<0 0.0007 -0.2081 -0.9781>}
 tent_c_3 {parent=tent_c_2, loc=<9.669 0 0>, rot=<0 0.0008 -0.2238 -0.9746>}
 tent_c_4 {parent=tent_c_3, loc=<8.736 0 0>, rot=<0 0.0002 -0.0725 -0.9974>}
 tent_c_5 {parent=tent_c_4, loc=<7.7581 0 0>, rot=<0 0.0001 -0.0411 -0.9992>}
 tent_c_6 {parent=tent_c_5, loc=<6.753 0 0>, rot=<0 0 -0.0004 -1>}
 tent_c_7 {parent=tent_c_6, loc=<5.7393 0 0>, rot=<0 -0.0001 0.0317 -0.9995>}
 tent_c_8 {parent=tent_c_7, loc=<4.7269 0 0>, rot=<0 0 0.0149 -0.9999>}
 tent_c_9 {parent=tent_c_8, loc=<3.7135 0 0>, rot=<0 0 -0.0025 -1>}
 tent_c_10 {parent=tent_c_9, loc=<2.7012 0 0>, rot=<0 0 0 -1>}
 tent_1_l_1 {parent=dyn_hairBase, loc=<-1.3295 -11.2731 -6.2932>, rot=<0.5883 -0.7664 -0.1816 -0.1833>}
 tent_1_l_2 {parent=tent_1_l_1, loc=<9.5159 0 0>, rot=<0 -0.0838 -0.2101 -0.9741>}
 tent_1_l_3 {parent=tent_1_l_2, loc=<8.5045 0 0>, rot=<0 -0.0709 -0.1417 -0.9874>}
 tent_1_l_4 {parent=tent_1_l_3, loc=<7.7419 0 0>, rot=<0 -0.0648 -0.0741 -0.9951>}
 tent_1_l_5 {parent=tent_1_l_4, loc=<6.8391 0 0>, rot=<0 -0.0334 -0.0774 -0.9964>}
 tent_1_l_6 {parent=tent_1_l_5, loc=<5.9605 0 0>, rot=<0 -0.079 -0.0447 -0.9959>}
 tent_1_l_7 {parent=tent_1_l_6, loc=<5.0542 0 0>, rot=<0 -0.0611 -0.078 -0.9951>}
 tent_1_l_8 {parent=tent_1_l_7, loc=<4.1709 0 0>, rot=<0 -0.0342 -0.06 -0.9976>}
 tent_1_l_9 {parent=tent_1_l_8, loc=<3.2731 0 0>, rot=<0 -0.0359 -0.0767 -0.9964>}
 tent_1_l_10 {parent=tent_1_l_9, loc=<2.385 0 0>, rot=<0 0 0 -1>}
 tent_2_l_1 {parent=dyn_hairBase, loc=<-0.6196 -11.5178 -3.6679>, rot=<0.6129 -0.7864 -0.0528 -0.0556>}
 tent_2_l_2 {parent=tent_2_l_1, loc=<8.5502 0 0>, rot=<0 0.0118 -0.1166 -0.9931>}
 tent_2_l_3 {parent=tent_2_l_2, loc=<7.7009 0 0>, rot=<0 -0.0297 -0.1187 -0.9925>}
 tent_2_l_4 {parent=tent_2_l_3, loc=<6.9526 0 0>, rot=<0 -0.0165 -0.0386 -0.9991>}
 tent_2_l_5 {parent=tent_2_l_4, loc=<6.1446 0 0>, rot=<0 -0.0424 -0.0852 -0.9955>}
 tent_2_l_6 {parent=tent_2_l_5, loc=<5.3448 0 0>, rot=<0 -0.0198 -0.0391 -0.999>}
 tent_2_l_7 {parent=tent_2_l_6, loc=<4.5446 0 0>, rot=<0 -0.0355 -0.0304 -0.9989>}
 tent_2_l_8 {parent=tent_2_l_7, loc=<3.7437 0 0>, rot=<0 -0.0156 -0.0207 -0.9997>}
 tent_2_l_9 {parent=tent_2_l_8, loc=<2.9418 0 0>, rot=<0 -0.0248 -0.0087 -0.9997>}
 tent_2_l_10 {parent=tent_2_l_9, loc=<2.1397 0 0>, rot=<0 0 0 -1>}
 curve_1_l_CV_1 {parent=dyn_hairBase, loc=<-1.3295 -11.2731 -6.2932>, rot=<0.5883 -0.7664 -0.1816 -0.1833>}
 curve_1_l_CV_2 {parent=curve_1_l_CV_1, loc=<17.7018 0 0>, rot=<0 -0.1445 -0.3428 -0.9282>}
 curve_1_l_CV_3 {parent=curve_1_l_CV_2, loc=<23.111 0 0>, rot=<0 -0.2986 -0.3987 -0.8671>}
 curve_1_l_CV_4 {parent=curve_1_l_CV_3, loc=<19.3263 0 0>, rot=<0 0 0 -1>}
 curve_1_r_CV_1 {parent=dyn_hairBase, loc=<-1.3295 -11.2731 6.2932>, rot=<0.5883 -0.7664 0.1816 0.1833>}
 curve_1_r_CV_2 {parent=curve_1_r_CV_1, loc=<17.7018 0 0>, rot=<0 0.1445 -0.3428 -0.9282>}
 curve_1_r_CV_3 {parent=curve_1_r_CV_2, loc=<23.111 0 0>, rot=<0 0.2986 -0.3987 -0.8671>}
 curve_1_r_CV_4 {parent=curve_1_r_CV_3, loc=<19.3263 0 0>, rot=<0 0 0 -1>}
 curve_2_l_CV_1 {parent=dyn_hairBase, loc=<-0.6196 -11.5178 -3.6679>, rot=<0.6129 -0.7864 -0.0528 -0.0556>}
 curve_2_l_CV_2 {parent=curve_2_l_CV_1, loc=<17.4897 0 0>, rot=<0 0.0366 -0.2864 -0.9574>}
 curve_2_l_CV_3 {parent=curve_2_l_CV_2, loc=<15.0484 0 0>, rot=<0 -0.199 -0.1729 -0.9646>}
 curve_2_l_CV_4 {parent=curve_2_l_CV_3, loc=<17.3904 0 0>, rot=<0 0 0 -1>}
 curve_2_r_CV_1 {parent=dyn_hairBase, loc=<-0.6196 -11.5178 3.6679>, rot=<0.6129 -0.7864 0.0528 0.0556>}
 curve_2_r_CV_2 {parent=curve_2_r_CV_1, loc=<17.4897 0 0>, rot=<0 -0.0366 -0.2864 -0.9574>}
 curve_2_r_CV_3 {parent=curve_2_r_CV_2, loc=<15.0484 0 0>, rot=<0 0.199 -0.1729 -0.9646>}
 curve_2_r_CV_4 {parent=curve_2_r_CV_3, loc=<17.3904 0 0>, rot=<0 0 0 -1>}
 curve_c_CV_1 {parent=dyn_hairBase, loc=<-0.0833 -6.0905 0.0099>, rot=<0.7651 -0.6439 -0.001 0.0012>}
 curve_c_CV_2 {parent=curve_c_CV_1, loc=<24.7393 0 0>, rot=<0 0.002 -0.6167 -0.7872>}
 curve_c_CV_3 {parent=curve_c_CV_2, loc=<20.5786 0 0>, rot=<0 -0.0006 0.1584 -0.9874>}
 curve_c_CV_4 {parent=curve_c_CV_3, loc=<21.0666 0 0>, rot=<0 0 0 1>}
]
variables [
 isRoaringAlpha {type=double, default=0}
 useHarmonics {type=bool, default=True}
]
