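# Animation Blueprint for the large male Cattus face accessory, posing and simulating its tentacle bones.
# Weighted constraints tie lower curve control bones to spine, clavicle, neck, pelvis and hair-base bones.
# Interpolated roaring and not-roaring alphas blend additive offsets onto those controls before a curves Control Rig runs, starting from the PreBezierCache pose.
# Per-bone Anim Dynamics then adds spring motion, scaled by individual alphas, with spherical limits around upper spine and clavicles.
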AnimBlueprint /Game/Characters/Player/Male/Large/Bodies/M_LRG_Cattus/Meshes/Parts/M_LRG_Cattus_FaceAcc_AnimBP
target_skeleton=&/M_LRG_Cattus_FaceAcc_Skeleton
graphs [
 AnimGraph {nodes [
   n1 = "Constraint - curve_2_l_CV_3 (PPPPTPPPP)"!(
    ComponentPose=Local To Component(LocalPose=Use cached pose 'PreBezierCache'())
    Alpha=1
    ConstraintWeights_0=0.4276
    ConstraintWeights_1:float=0.1085
    ConstraintWeights_2:float=0.0339
    ConstraintWeights_3:float=0.268
    ConstraintWeights_4:float=0
    ConstraintWeights_5:float=0.0644
    ConstraintWeights_6:float=0.0216
    ConstraintWeights_7:float=0.0001
    ConstraintWeights_8:float=0.0654
   ) -> (Pose) {
    alphaBoolBlend {blendOption=Linear}
    boneToModify {boneName=curve_2_l_CV_3}
    constraintSetup [
     {targetBone {boneName=clavicle_l}}
     {targetBone {boneName=spine_03}}
     {targetBone {boneName=spine_04}}
     {targetBone {boneName=spine_05}}
     {transformType=Translation, targetBone {boneName=None}}
     {targetBone {boneName=clavicle_r}}
     {targetBone {boneName=neck_02}}
     {targetBone {boneName=neck_01}}
     {targetBone {boneName=dyn_hairBase}}
    ]
    constraintWeights [1,1,1,1,1,1,1,1,1]
   }
   n2 = "Constraint - curve_2_l_CV_4 (PPPPPPP)"!(
    ComponentPose=@n1
    Alpha=1
    ConstraintWeights_0=0.2798
    ConstraintWeights_1:float=0.0235
    ConstraintWeights_2:float=0.3517
    ConstraintWeights_3:float=0.1312
    ConstraintWeights_4:float=0.0713
    ConstraintWeights_5:float=0.1404
    ConstraintWeights_6:float=0.0021
   ) -> (Pose) {
    alphaBoolBlend {blendOption=Linear}
    boneToModify {boneName=curve_2_l_CV_4}
    constraintSetup [
     {targetBone {boneName=clavicle_l}}
     {targetBone {boneName=spine_02}}
     {targetBone {boneName=spine_03}}
     {targetBone {boneName=spine_04}}
     {targetBone {boneName=spine_05}}
     {targetBone {boneName=clavicle_r}}
     {targetBone {boneName=pelvis}}
    ]
    constraintWeights [1,1,1,1,1,1,1]
   }
   n3 = "Constraint - curve_1_l_CV_3 (PPPPPPTPPP)"!(
    ComponentPose=@n2
    Alpha=1
    ConstraintWeights_0=0.5371
    ConstraintWeights_1:float=0
    ConstraintWeights_2:float=0.0134
    ConstraintWeights_3:float=0.1993
    ConstraintWeights_4:float=0.0562
    ConstraintWeights_5:float=0.1083
    ConstraintWeights_6:float=0
    ConstraintWeights_7:float=0.0184
    ConstraintWeights_8:float=0
    ConstraintWeights_9:float=0.0013
   ) -> (Pose) {
    alphaBoolBlend {blendOption=Linear}
    boneToModify {boneName=curve_1_l_CV_3}
    constraintSetup [
     {targetBone {boneName=clavicle_l}}
     {targetBone {boneName=None}}
     {targetBone {boneName=spine_02}}
     {targetBone {boneName=spine_03}}
     {targetBone {boneName=spine_04}}
     {targetBone {boneName=spine_05}}
     {transformType=Translation, targetBone {boneName=None}}
     {targetBone {boneName=clavicle_r}}
     {targetBone {boneName=neck_02}}
     {targetBone {boneName=dyn_hairBase}}
    ]
    constraintWeights [1,1,1,1,1,1,1,1,1,1]
   }
   n4 = "Constraint - curve_1_l_CV_4 (PPPPPPPP)"!(
    ComponentPose=@n3
    Alpha=1
    ConstraintWeights_0=0.2137
    ConstraintWeights_1:float=0.0171
    ConstraintWeights_2:float=0.0558
    ConstraintWeights_3:float=0.4378
    ConstraintWeights_4:float=0.1502
    ConstraintWeights_5:float=0.0185
    ConstraintWeights_6:float=0.1034
    ConstraintWeights_7:float=0.0034
   ) -> (Pose) {
    alphaBoolBlend {blendOption=Linear}
    boneToModify {boneName=curve_1_l_CV_4}
    constraintSetup [
     {targetBone {boneName=clavicle_l}}
     {targetBone {boneName=spine_01}}
     {targetBone {boneName=spine_02}}
     {targetBone {boneName=spine_03}}
     {targetBone {boneName=spine_04}}
     {targetBone {boneName=spine_05}}
     {targetBone {boneName=clavicle_r}}
     {targetBone {boneName=pelvis}}
    ]
    constraintWeights [1,1,1,1,1,1,1,1]
   }
   n5 = "Constraint - curve_c_CV_3 (PPPPPPPP)"!(
    ComponentPose=@n4
    Alpha=1
    ConstraintWeights_0=0.2133
    ConstraintWeights_1:float=0.0987
    ConstraintWeights_2:float=0.0334
    ConstraintWeights_3:float=0.3354
    ConstraintWeights_4:float=0.2204
    ConstraintWeights_5:float=0.0295
    ConstraintWeights_6:float=0.0148
    ConstraintWeights_7:float=0.0544
   ) -> (Pose) {
    alphaBoolBlend {blendOption=Linear}
    boneToModify {boneName=curve_c_CV_3}
    constraintSetup [
     {targetBone {boneName=clavicle_l}}
     {targetBone {boneName=spine_03}}
     {targetBone {boneName=spine_04}}
     {targetBone {boneName=spine_05}}
     {targetBone {boneName=clavicle_r}}
     {targetBone {boneName=neck_02}}
     {targetBone {boneName=neck_01}}
     {targetBone {boneName=dyn_hairBase}}
    ]
    constraintWeights [1,1,1,1,1,1,1,1]
   }
   n6 = "Constraint - curve_c_CV_4 (PPPPPP)"!(
    ComponentPose=@n5
    Alpha=1
    ConstraintWeights_0=0.2052
    ConstraintWeights_1:float=0.0208
    ConstraintWeights_2:float=0.3565
    ConstraintWeights_3:float=0.1365
    ConstraintWeights_4:float=0.0699
    ConstraintWeights_5:float=0.2111
   ) -> (Pose) {
    alphaBoolBlend {blendOption=Linear}
    boneToModify {boneName=curve_c_CV_4}
    constraintSetup [
     {targetBone {boneName=clavicle_l}}
     {targetBone {boneName=spine_02}}
     {targetBone {boneName=spine_03}}
     {targetBone {boneName=spine_04}}
     {targetBone {boneName=spine_05}}
     {targetBone {boneName=clavicle_r}}
    ]
    constraintWeights [1,1,1,1,1,1]
   }
   n7 = "Constraint - curve_2_r_CV_3 (PTPPPPPP)"!(
    ComponentPose=@n6
    Alpha=1
    ConstraintWeights_0=0.0642
    ConstraintWeights_1:float=0
    ConstraintWeights_2:float=0.1086
    ConstraintWeights_3:float=0.0339
    ConstraintWeights_4:float=0.2679
    ConstraintWeights_5:float=0.4279
    ConstraintWeights_6:float=0.0215
    ConstraintWeights_7:float=0.0652
   ) -> (Pose) {
    alphaBoolBlend {blendOption=Linear}
    boneToModify {boneName=curve_2_r_CV_3}
    constraintSetup [
     {targetBone {boneName=clavicle_l}}
     {transformType=Translation, targetBone {boneName=None}}
     {targetBone {boneName=spine_03}}
     {targetBone {boneName=spine_04}}
     {targetBone {boneName=spine_05}}
     {targetBone {boneName=clavicle_r}}
     {targetBone {boneName=neck_02}}
     {targetBone {boneName=dyn_hairBase}}
    ]
    constraintWeights [1,1,1,1,1,1,1,1]
   }
   n8 = "Constraint - curve_2_r_CV_4 (PPPPPPP)"!(
    ComponentPose=@n7
    Alpha=1
    ConstraintWeights_0=0.1403
    ConstraintWeights_1:float=0.0235
    ConstraintWeights_2:float=0.3517
    ConstraintWeights_3:float=0.1312
    ConstraintWeights_4:float=0.0713
    ConstraintWeights_5:float=0.28
    ConstraintWeights_6:float=0.0022
   ) -> (Pose) {
    alphaBoolBlend {blendOption=Linear}
    boneToModify {boneName=curve_2_r_CV_4}
    constraintSetup [
     {targetBone {boneName=clavicle_l}}
     {targetBone {boneName=spine_02}}
     {targetBone {boneName=spine_03}}
     {targetBone {boneName=spine_04}}
     {targetBone {boneName=spine_05}}
     {targetBone {boneName=clavicle_r}}
     {targetBone {boneName=pelvis}}
    ]
    constraintWeights [1,1,1,1,1,1,1]
   }
   n9 = "Constraint - curve_1_r_CV_3 (PPPPPPP)"!(
    ComponentPose=@n8
    Alpha=1
    ConstraintWeights_0=0.0183
    ConstraintWeights_1:float=0.0134
    ConstraintWeights_2:float=0.1993
    ConstraintWeights_3:float=0.0561
    ConstraintWeights_4:float=0.1081
    ConstraintWeights_5:float=0.5372
    ConstraintWeights_6:float=0.0013
   ) -> (Pose) {
    alphaBoolBlend {blendOption=Linear}
    boneToModify {boneName=curve_1_r_CV_3}
    constraintSetup [
     {targetBone {boneName=clavicle_l}}
     {targetBone {boneName=spine_02}}
     {targetBone {boneName=spine_03}}
     {targetBone {boneName=spine_04}}
     {targetBone {boneName=spine_05}}
     {targetBone {boneName=clavicle_r}}
     {targetBone {boneName=dyn_hairBase}}
    ]
    constraintWeights [1,1,1,1,1,1,1]
   }
   n10 = "Constraint - curve_1_r_CV_4 (PPPPPPPP)"!(
    ComponentPose=@n9
    Alpha=1
    ConstraintWeights_0=0.1032
    ConstraintWeights_1:float=0.0171
    ConstraintWeights_2:float=0.0558
    ConstraintWeights_3:float=0.4378
    ConstraintWeights_4:float=0.1502
    ConstraintWeights_5:float=0.0185
    ConstraintWeights_6:float=0.2139
    ConstraintWeights_7:float=0.0034
   ) -> (Pose) {
    alphaBoolBlend {blendOption=Linear}
    boneToModify {boneName=curve_1_r_CV_4}
    constraintSetup [
     {targetBone {boneName=clavicle_l}}
     {targetBone {boneName=spine_01}}
     {targetBone {boneName=spine_02}}
     {targetBone {boneName=spine_03}}
     {targetBone {boneName=spine_04}}
     {targetBone {boneName=spine_05}}
     {targetBone {boneName=clavicle_r}}
     {targetBone {boneName=pelvis}}
    ]
    constraintWeights [1,1,1,1,1,1,1,1]
   }
   n11 = Get isNotRoaringAlpha() -> (isNotRoaringAlpha:double)
   n12 = "Transform (Modify) Bone - Bone: curve_1_l_CV_2"(ComponentPose=@n10, Alpha=@n11, Translation="0,5,5", Rotation="0,0,0", Scale="1,1,1") {
    rotationMode=BMM_Additive
    rotationSpace=BCS_BoneSpace
    translationMode=BMM_Additive
    alphaBoolBlend {blendOption=Linear}
    alphaScaleBiasClamp {bInterpResult=true, interpSpeedDecreasing=5, interpSpeedIncreasing=5}
    boneToModify {boneName=curve_1_l_CV_2}
    translation {y=5, z=5}
   }
   n13 = "Transform (Modify) Bone - Bone: curve_2_l_CV_3"(ComponentPose=@n12, Alpha=@n11, Translation="0,0,-10", Rotation="0,0,0", Scale="1,1,1") {
    rotationMode=BMM_Additive
    rotationSpace=BCS_BoneSpace
    translationMode=BMM_Additive
    alphaBoolBlend {blendOption=Linear}
    alphaScaleBiasClamp {bInterpResult=true, interpSpeedDecreasing=5, interpSpeedIncreasing=5, scale=0}
    boneToModify {boneName=curve_2_l_CV_3}
    translation {z=-10}
   }
   n14 = "Transform (Modify) Bone - Bone: curve_c_CV_3"(ComponentPose=@n13, Alpha=@n11, Translation="0,0,-5", Rotation="0,0,0", Scale="1,1,1") {
    rotationMode=BMM_Additive
    rotationSpace=BCS_BoneSpace
    translationMode=BMM_Additive
    alphaBoolBlend {blendOption=Linear}
    alphaScaleBiasClamp {bInterpResult=true, interpSpeedDecreasing=5, interpSpeedIncreasing=5}
    boneToModify {boneName=curve_c_CV_3}
    translation {z=-5}
   }
   n15 = "Transform (Modify) Bone - Bone: curve_2_r_CV_3"(ComponentPose=@n14, Alpha=@n11, Translation="0,0,-10", Rotation="0,0,0", Scale="1,1,1") {
    rotationMode=BMM_Additive
    rotationSpace=BCS_BoneSpace
    translationMode=BMM_Additive
    alphaBoolBlend {blendOption=Linear}
    alphaScaleBiasClamp {bInterpResult=true, interpSpeedDecreasing=5, interpSpeedIncreasing=5, scale=0}
    boneToModify {boneName=curve_2_r_CV_3}
    translation {z=-10}
   }
   n16 = "Transform (Modify) Bone - Bone: curve_1_r_CV_2"(ComponentPose=@n15, Alpha=@n11, Translation="0,-5,5", Rotation="0,0,0", Scale="1,1,1") {
    rotationMode=BMM_Additive
    rotationSpace=BCS_BoneSpace
    translationMode=BMM_Additive
    alphaBoolBlend {blendOption=Linear}
    alphaScaleBiasClamp {bInterpResult=true, interpSpeedDecreasing=5, interpSpeedIncreasing=5}
    boneToModify {boneName=curve_1_r_CV_2}
    translation {y=-5, z=5}
   }
   n17 = Get isRoaringAlpha() -> (isRoaringAlpha:double)
   n18 = "Transform (Modify) Bone - Bone: curve_1_l_CV_1"!(ComponentPose=@n16, Alpha=@n17, Rotation="20,-20,0") -> (Pose) {
    rotationMode=BMM_Additive
    rotationSpace=BCS_BoneSpace
    alphaBoolBlend {blendOption=Linear}
    alphaScaleBiasClamp {bInterpResult=true, interpSpeedDecreasing=5, interpSpeedIncreasing=5}
    boneToModify {boneName=curve_1_l_CV_1}
    rotation {pitch=20, yaw=-20}
   }
   n19 = "Transform (Modify) Bone - Bone: curve_2_l_CV_1"!(ComponentPose=@n18, Alpha=@n17, Rotation="5,-15,0") -> (Pose) {
    rotationMode=BMM_Additive
    rotationSpace=BCS_BoneSpace
    alphaBoolBlend {blendOption=Linear}
    alphaScaleBiasClamp {bInterpResult=true, interpSpeedDecreasing=5}
    boneToModify {boneName=curve_2_l_CV_1}
    rotation {pitch=5, yaw=-15}
   }
   n20 = "Transform (Modify) Bone - Bone: curve_c_CV_1"!(ComponentPose=@n19, Alpha=@n17, Rotation="0,-8,0") -> (Pose) {
    rotationMode=BMM_Additive
    rotationSpace=BCS_BoneSpace
    alphaBoolBlend {blendOption=Linear}
    alphaScaleBiasClamp {bInterpResult=true, interpSpeedDecreasing=5, interpSpeedIncreasing=5}
    boneToModify {boneName=curve_c_CV_1}
    rotation {yaw=-8}
   }
   n21 = "Transform (Modify) Bone - Bone: curve_2_r_CV_1"!(ComponentPose=@n20, Alpha=@n17, Rotation="-5,-15,0") -> (Pose) {
    rotationMode=BMM_Additive
    rotationSpace=BCS_BoneSpace
    alphaBoolBlend {blendOption=Linear}
    alphaScaleBiasClamp {bInterpResult=true, interpSpeedDecreasing=5}
    boneToModify {boneName=curve_2_r_CV_1}
    rotation {pitch=-5, yaw=-15}
   }
   n22 = "Transform (Modify) Bone - Bone: curve_1_r_CV_1"!(ComponentPose=@n21, Alpha=@n17, Rotation="-15,-5,0") -> (Pose) {
    rotationMode=BMM_Additive
    rotationSpace=BCS_BoneSpace
    alphaBoolBlend {blendOption=Linear}
    alphaScaleBiasClamp {bInterpResult=true, interpSpeedDecreasing=5, interpSpeedIncreasing=5}
    boneToModify {boneName=curve_1_r_CV_1}
    rotation {pitch=-15, yaw=-5}
   }
   n23 = "Transform (Modify) Bone - Bone: curve_1_l_CV_2"!(ComponentPose=@n22, Alpha=@n17, Rotation="10,5,0") -> (Pose) {
    rotationMode=BMM_Additive
    rotationSpace=BCS_BoneSpace
    alphaBoolBlend {blendOption=Linear}
    alphaScaleBiasClamp {bInterpResult=true, interpSpeedDecreasing=5, interpSpeedIncreasing=5}
    boneToModify {boneName=curve_1_l_CV_2}
    rotation {pitch=10, yaw=5}
   }
   n24 = "Transform (Modify) Bone - Bone: curve_2_l_CV_2"!(ComponentPose=@n23, Alpha=@n17, Rotation="12,5,0") -> (Pose) {
    rotationMode=BMM_Additive
    rotationSpace=BCS_BoneSpace
    alphaBoolBlend {blendOption=Linear}
    alphaScaleBiasClamp {bInterpResult=true, interpSpeedDecreasing=5, interpSpeedIncreasing=5}
    boneToModify {boneName=curve_2_l_CV_2}
    rotation {pitch=12, yaw=5}
   }
   n25 = "Transform (Modify) Bone - Bone: curve_c_CV_2"!(ComponentPose=@n24, Alpha=@n17, Rotation="0,20,0") -> (Pose) {
    rotationMode=BMM_Additive
    rotationSpace=BCS_BoneSpace
    alphaBoolBlend {blendOption=Linear}
    alphaScaleBiasClamp {bInterpResult=true, interpSpeedDecreasing=5, interpSpeedIncreasing=5}
    boneToModify {boneName=curve_c_CV_2}
    rotation {yaw=20}
   }
   n26 = "Transform (Modify) Bone - Bone: curve_2_r_CV_2"!(ComponentPose=@n25, Alpha=@n17, Rotation="-12,5,0") -> (Pose) {
    rotationMode=BMM_Additive
    rotationSpace=BCS_BoneSpace
    alphaBoolBlend {blendOption=Linear}
    alphaScaleBiasClamp {bInterpResult=true, interpSpeedDecreasing=5, interpSpeedIncreasing=5}
    boneToModify {boneName=curve_2_r_CV_2}
    rotation {pitch=-12, yaw=5}
   }
   n27 = "Transform (Modify) Bone - Bone: curve_1_r_CV_2"!(ComponentPose=@n26, Alpha=@n17, Rotation="-15,-5,0") -> (Pose) {
    rotationMode=BMM_Additive
    rotationSpace=BCS_BoneSpace
    alphaBoolBlend {blendOption=Linear}
    alphaScaleBiasClamp {bInterpResult=true, interpSpeedDecreasing=5, interpSpeedIncreasing=5}
    boneToModify {boneName=curve_1_r_CV_2}
    rotation {pitch=-15, yaw=-5}
   }
   n28 = "Transform (Modify) Bone - Bone: curve_1_l_CV_3"!(ComponentPose=@n27, Alpha=@n17, Rotation="20,0,0") -> (Pose) {
    rotationMode=BMM_Additive
    rotationSpace=BCS_BoneSpace
    alphaBoolBlend {blendOption=Linear}
    alphaScaleBiasClamp {bInterpResult=true, interpSpeedDecreasing=5, interpSpeedIncreasing=5}
    boneToModify {boneName=curve_1_l_CV_3}
    rotation {pitch=20}
   }
   n29 = "Transform (Modify) Bone - Bone: curve_2_l_CV_3"!(ComponentPose=@n28, Alpha=@n17, Rotation="12,30,0") -> (Pose) {
    rotationMode=BMM_Additive
    rotationSpace=BCS_BoneSpace
    alphaBoolBlend {blendOption=Linear}
    alphaScaleBiasClamp {bInterpResult=true, interpSpeedDecreasing=5, interpSpeedIncreasing=5}
    boneToModify {boneName=curve_2_l_CV_3}
    rotation {pitch=12, yaw=30}
   }
   n30 = "Transform (Modify) Bone - Bone: curve_c_CV_3"!(ComponentPose=@n29, Alpha=@n17, Rotation="0,40,0") -> (Pose) {
    rotationMode=BMM_Additive
    rotationSpace=BCS_BoneSpace
    alphaBoolBlend {blendOption=Linear}
    alphaScaleBiasClamp {bInterpResult=true, interpSpeedDecreasing=5, interpSpeedIncreasing=5}
    boneToModify {boneName=curve_c_CV_3}
    rotation {yaw=40}
   }
   n31 = "Transform (Modify) Bone - Bone: curve_2_r_CV_3"!(ComponentPose=@n30, Alpha=@n17, Rotation="-12,30,0") -> (Pose) {
    rotationMode=BMM_Additive
    rotationSpace=BCS_BoneSpace
    alphaBoolBlend {blendOption=Linear}
    alphaScaleBiasClamp {bInterpResult=true, interpSpeedDecreasing=5, interpSpeedIncreasing=5}
    boneToModify {boneName=curve_2_r_CV_3}
    rotation {pitch=-12, yaw=30}
   }
   n32 = "Transform (Modify) Bone - Bone: curve_1_r_CV_3"!(ComponentPose=@n31, Alpha=@n17, Rotation="-20,0,0") -> (Pose) {
    rotationMode=BMM_Additive
    rotationSpace=BCS_BoneSpace
    alphaBoolBlend {blendOption=Linear}
    alphaScaleBiasClamp {bInterpResult=true, interpSpeedDecreasing=5, interpSpeedIncreasing=5}
    boneToModify {boneName=curve_1_r_CV_3}
    rotation {pitch=-20}
   }
   n33 = Control Rig(Source=Component To Local(ComponentPose=@n32)) {
    alphaBoolBlend {blendOption=Linear}
    controlRigAssetReference {blueprintRigClass=&/M_LRG_Cattus_FaceAcc_CtrlRig_Curves.M_LRG_Cattus_FaceAcc_CtrlRig_Curves_C}
   }
   n34 = Anim Dynamics - Bone: tent_1_l_2(ComponentPose=Local To Component(LocalPose=@n33), Alpha=1, ExternalForce="0,0,0") {
    angularDampingOverride=1
    angularSpringConstant=1500
    bAngularSpring=true
    bGravityOverrideInSimSpace=true
    bOverrideAngularDamping=true
    bUseSphericalLimits=true
    gravityScale=0.35
    simulationSpace=RootRelative
    alphaBoolBlend {blendOption=Linear}
    alphaScaleBiasClamp {scale=0.03}
    boundBone {boneName=tent_1_l_2}
    componentLinearAccScale {z=0.45}
    componentLinearVelScale {z=0.45}
    physicsBodyDefinitions [{
      collisionType=CustomSphere
      sphereCollisionRadius=2
      boundBone {boneName=tent_1_l_2}
      boxExtents=<10 20 20>
      constraintSetup {angularLimitsMax=<10 20 60>, angularLimitsMin=<-10 -20 -60>, angularTarget=<1 0 0>}
      localJointOffset=<10 0 0>
     }]
    sphericalLimits [
     {limitRadius=31.637, drivingBone {boneName=spine_05}, sphereLocalOffset=<-1.5465 9.2897 0>}
     {limitRadius=16, drivingBone {boneName=clavicle_l}, sphereLocalOffset=<15.524 -1.769 -0.679>}
     {limitRadius=16, drivingBone {boneName=clavicle_r}, sphereLocalOffset=<-15.524 -1.769 0.679>}
    ]
   }
   n35 = Anim Dynamics - Bone: tent_1_l_3(ComponentPose=@n34, Alpha=1, ExternalForce="0,0,0") {
    angularDampingOverride=1
    angularSpringConstant=1500
    bAngularSpring=true
    bGravityOverrideInSimSpace=true
    bOverrideAngularDamping=true
    bUseSphericalLimits=true
    gravityScale=0.35
    simulationSpace=RootRelative
    alphaBoolBlend {blendOption=Linear}
    alphaScaleBiasClamp {scale=0.1}
    boundBone {boneName=tent_1_l_3}
    componentLinearAccScale {z=0.46}
    componentLinearVelScale {z=0.46}
    physicsBodyDefinitions [{
      collisionType=CustomSphere
      sphereCollisionRadius=2
      boundBone {boneName=tent_1_l_3}
      boxExtents=<10 20 20>
      constraintSetup {angularLimitsMax=<10 20 60>, angularLimitsMin=<-10 -20 -60>, angularTarget=<1 0 0>}
      localJointOffset=<10 0 0>
     }]
    sphericalLimits [
     {limitRadius=31.637, drivingBone {boneName=spine_05}, sphereLocalOffset=<-1.5465 9.2897 0>}
     {limitRadius=16, drivingBone {boneName=clavicle_l}, sphereLocalOffset=<15.524 -1.769 -0.679>}
     {limitRadius=16, drivingBone {boneName=clavicle_r}, sphereLocalOffset=<-15.524 -1.769 0.679>}
    ]
   }
   n36 = Anim Dynamics - Bone: tent_1_l_4(ComponentPose=@n35, Alpha=1, ExternalForce="0,0,0") {
    angularDampingOverride=1
    angularSpringConstant=1500
    bAngularSpring=true
    bGravityOverrideInSimSpace=true
    bOverrideAngularDamping=true
    bUseSphericalLimits=true
    gravityScale=0.35
    simulationSpace=RootRelative
    alphaBoolBlend {blendOption=Linear}
    alphaScaleBiasClamp {scale=0.15}
    boundBone {boneName=tent_1_l_4}
    componentLinearAccScale {z=0.46}
    componentLinearVelScale {z=0.46}
    physicsBodyDefinitions [{
      collisionType=CustomSphere
      sphereCollisionRadius=2
      boundBone {boneName=tent_1_l_4}
      boxExtents=<10 20 20>
      constraintSetup {angularLimitsMax=<10 20 60>, angularLimitsMin=<-10 -20 -60>, angularTarget=<0 0 0>}
      localJointOffset=<10 0 0>
     }]
    sphericalLimits [
     {limitRadius=31.637, drivingBone {boneName=spine_05}, sphereLocalOffset=<-1.5465 9.2897 0>}
     {limitRadius=16, drivingBone {boneName=clavicle_l}, sphereLocalOffset=<15.524 -1.769 -0.679>}
     {limitRadius=16, drivingBone {boneName=clavicle_r}, sphereLocalOffset=<-15.524 -1.769 0.679>}
    ]
   }
   n37 = Anim Dynamics - Bone: tent_1_l_5(ComponentPose=@n36, Alpha=1, ExternalForce="0,0,0") {
    angularDampingOverride=1
    angularSpringConstant=1500
    bAngularSpring=true
    bGravityOverrideInSimSpace=true
    bOverrideAngularDamping=true
    bUseSphericalLimits=true
    gravityScale=0.35
    simulationSpace=RootRelative
    alphaBoolBlend {blendOption=Linear}
    alphaScaleBiasClamp {scale=0.15}
    boundBone {boneName=tent_1_l_5}
    componentLinearAccScale {z=0.46}
    componentLinearVelScale {z=0.46}
    physicsBodyDefinitions [{
      collisionType=CustomSphere
      sphereCollisionRadius=2
      boundBone {boneName=tent_1_l_5}
      boxExtents=<10 20 20>
      constraintSetup {angularLimitsMax=<10 20 60>, angularLimitsMin=<-10 -20 -60>, angularTarget=<0 0 0>}
      localJointOffset=<10 0 0>
     }]
    sphericalLimits [
     {limitRadius=31.637, drivingBone {boneName=spine_05}, sphereLocalOffset=<-1.5465 9.2897 0>}
     {limitRadius=16, drivingBone {boneName=clavicle_l}, sphereLocalOffset=<15.524 -1.769 -0.679>}
     {limitRadius=16, drivingBone {boneName=clavicle_r}, sphereLocalOffset=<-15.524 -1.769 0.679>}
    ]
   }
   n38 = Anim Dynamics - Bone: tent_1_l_6(ComponentPose=@n37, Alpha=1, ExternalForce="0,0,0") {
    angularDampingOverride=1
    angularSpringConstant=1500
    bAngularSpring=true
    bGravityOverrideInSimSpace=true
    bOverrideAngularDamping=true
    bUseSphericalLimits=true
    gravityScale=0.35
    simulationSpace=RootRelative
    alphaBoolBlend {blendOption=Linear}
    alphaScaleBiasClamp {scale=0.25}
    boundBone {boneName=tent_1_l_6}
    componentLinearAccScale {z=0.45}
    componentLinearVelScale {z=0.45}
    physicsBodyDefinitions [{
      collisionType=CustomSphere
      sphereCollisionRadius=2
      boundBone {boneName=tent_1_l_6}
      boxExtents=<10 20 20>
      constraintSetup {angularLimitsMax=<10 20 60>, angularLimitsMin=<-10 -20 -60>, angularTarget=<1 0 0>}
      localJointOffset=<10 0 0>
     }]
    sphericalLimits [
     {limitRadius=31.637, drivingBone {boneName=spine_05}, sphereLocalOffset=<-1.5465 9.2897 0>}
     {limitRadius=16, drivingBone {boneName=clavicle_l}, sphereLocalOffset=<15.524 -1.769 -0.679>}
     {limitRadius=16, drivingBone {boneName=clavicle_r}, sphereLocalOffset=<-15.524 -1.769 0.679>}
    ]
   }
   n39 = Anim Dynamics - Bone: tent_1_l_7(ComponentPose=@n38, Alpha=1, ExternalForce="0,0,0") {
    angularDampingOverride=1
    angularSpringConstant=1500
    bAngularSpring=true
    bGravityOverrideInSimSpace=true
    bOverrideAngularDamping=true
    bUseSphericalLimits=true
    gravityScale=0.35
    simulationSpace=RootRelative
    alphaBoolBlend {blendOption=Linear}
    alphaScaleBiasClamp {scale=0.25}
    boundBone {boneName=tent_1_l_7}
    componentLinearAccScale {z=0.46}
    componentLinearVelScale {z=0.46}
    physicsBodyDefinitions [{
      collisionType=CustomSphere
      sphereCollisionRadius=2
      boundBone {boneName=tent_1_l_7}
      boxExtents=<10 20 20>
      constraintSetup {angularLimitsMax=<10 20 60>, angularLimitsMin=<-10 -20 -60>, angularTarget=<1 0 0>}
      localJointOffset=<10 0 0>
     }]
    sphericalLimits [
     {limitRadius=31.637, drivingBone {boneName=spine_05}, sphereLocalOffset=<-1.5465 9.2897 0>}
     {limitRadius=16, drivingBone {boneName=clavicle_l}, sphereLocalOffset=<15.524 -1.769 -0.679>}
     {limitRadius=16, drivingBone {boneName=clavicle_r}, sphereLocalOffset=<-15.524 -1.769 0.679>}
    ]
   }
   n40 = Anim Dynamics - Bone: tent_1_l_8(ComponentPose=@n39, Alpha=1, ExternalForce="0,0,0") {
    angularDampingOverride=1
    angularSpringConstant=1500
    bAngularSpring=true
    bGravityOverrideInSimSpace=true
    bOverrideAngularDamping=true
    bUseSphericalLimits=true
    gravityScale=0.35
    simulationSpace=RootRelative
    alphaBoolBlend {blendOption=Linear}
    alphaScaleBiasClamp {scale=0.45}
    boundBone {boneName=tent_1_l_8}
    componentLinearAccScale {z=0.46}
    componentLinearVelScale {z=0.46}
    physicsBodyDefinitions [{
      collisionType=CustomSphere
      sphereCollisionRadius=2
      boundBone {boneName=tent_1_l_8}
      boxExtents=<10 20 20>
      constraintSetup {angularLimitsMax=<10 20 60>, angularLimitsMin=<-10 -20 -60>, angularTarget=<0 0 0>}
      localJointOffset=<10 0 0>
     }]
    sphericalLimits [
     {limitRadius=31.637, drivingBone {boneName=spine_05}, sphereLocalOffset=<-1.5465 9.2897 0>}
     {limitRadius=16, drivingBone {boneName=clavicle_l}, sphereLocalOffset=<15.524 -1.769 -0.679>}
     {limitRadius=16, drivingBone {boneName=clavicle_r}, sphereLocalOffset=<-15.524 -1.769 0.679>}
    ]
   }
   n41 = Anim Dynamics - Bone: tent_1_l_9(ComponentPose=@n40, Alpha=1, ExternalForce="0,0,0") {
    angularDampingOverride=1
    angularSpringConstant=1500
    bAngularSpring=true
    bGravityOverrideInSimSpace=true
    bOverrideAngularDamping=true
    bUseSphericalLimits=true
    gravityScale=0.35
    simulationSpace=RootRelative
    alphaBoolBlend {blendOption=Linear}
    alphaScaleBiasClamp {scale=0.45}
    boundBone {boneName=tent_1_l_9}
    componentLinearAccScale {z=0.46}
    componentLinearVelScale {z=0.46}
    physicsBodyDefinitions [{
      collisionType=CustomSphere
      sphereCollisionRadius=2
      boundBone {boneName=tent_1_l_9}
      boxExtents=<10 20 20>
      constraintSetup {angularLimitsMax=<10 20 60>, angularLimitsMin=<-10 -20 -60>, angularTarget=<0 0 0>}
      localJointOffset=<10 0 0>
     }]
    sphericalLimits [
     {limitRadius=31.637, drivingBone {boneName=spine_05}, sphereLocalOffset=<-1.5465 9.2897 0>}
     {limitRadius=16, drivingBone {boneName=clavicle_l}, sphereLocalOffset=<15.524 -1.769 -0.679>}
     {limitRadius=16, drivingBone {boneName=clavicle_r}, sphereLocalOffset=<-15.524 -1.769 0.679>}
    ]
   }
   n42 = Anim Dynamics - Bone: tent_2_l_2(ComponentPose=@n41, Alpha=1, ExternalForce="0,0,0") {
    angularDampingOverride=1
    angularSpringConstant=1500
    bAngularSpring=true
    bGravityOverrideInSimSpace=true
    bOverrideAngularDamping=true
    bUseSphericalLimits=true
    gravityScale=0.35
    simulationSpace=RootRelative
    alphaBoolBlend {blendOption=Linear}
    alphaScaleBiasClamp {scale=0.03}
    boundBone {boneName=tent_2_l_2}
    componentLinearAccScale {z=0.45}
    componentLinearVelScale {z=0.45}
    physicsBodyDefinitions [{
      collisionType=CustomSphere
      sphereCollisionRadius=2
      boundBone {boneName=tent_2_l_2}
      boxExtents=<11 20 20>
      constraintSetup {angularLimitsMax=<0 20 60>, angularLimitsMin=<0 -20 -60>, angularTarget=<1 0 0>}
      localJointOffset=<10 0 0>
     }]
    sphericalLimits [
     {limitRadius=31.637, drivingBone {boneName=spine_05}, sphereLocalOffset=<-1.5465 9.2897 0>}
     {limitRadius=16, drivingBone {boneName=clavicle_l}, sphereLocalOffset=<15.524 -1.769 -0.679>}
     {limitRadius=16, drivingBone {boneName=clavicle_r}, sphereLocalOffset=<-15.524 -1.769 0.679>}
    ]
   }
   n43 = Anim Dynamics - Bone: tent_2_l_3(ComponentPose=@n42, Alpha=1, ExternalForce="0,0,0") {
    angularDampingOverride=1
    angularSpringConstant=1500
    bAngularSpring=true
    bGravityOverrideInSimSpace=true
    bOverrideAngularDamping=true
    bUseSphericalLimits=true
    gravityScale=0.35
    simulationSpace=RootRelative
    alphaBoolBlend {blendOption=Linear}
    alphaScaleBiasClamp {scale=0.06}
    boundBone {boneName=tent_2_l_3}
    componentLinearAccScale {z=0.46}
    componentLinearVelScale {z=0.46}
    physicsBodyDefinitions [{
      collisionType=CustomSphere
      sphereCollisionRadius=2
      boundBone {boneName=tent_2_l_3}
      boxExtents=<11 20 20>
      constraintSetup {angularLimitsMax=<0 20 60>, angularLimitsMin=<0 -20 -60>, angularTarget=<1 0 0>}
      localJointOffset=<10 0 0>
     }]
    sphericalLimits [
     {limitRadius=31.637, drivingBone {boneName=spine_05}, sphereLocalOffset=<-1.5465 9.2897 0>}
     {limitRadius=16, drivingBone {boneName=clavicle_l}, sphereLocalOffset=<15.524 -1.769 -0.679>}
     {limitRadius=16, drivingBone {boneName=clavicle_r}, sphereLocalOffset=<-15.524 -1.769 0.679>}
    ]
   }
   n44 = Anim Dynamics - Bone: tent_2_l_4(ComponentPose=@n43, Alpha=1, ExternalForce="0,0,0") {
    angularDampingOverride=1
    angularSpringConstant=1500
    bAngularSpring=true
    bGravityOverrideInSimSpace=true
    bOverrideAngularDamping=true
    bUseSphericalLimits=true
    gravityScale=0.35
    simulationSpace=RootRelative
    alphaBoolBlend {blendOption=Linear}
    alphaScaleBiasClamp {scale=0.06}
    boundBone {boneName=tent_2_l_4}
    componentLinearAccScale {z=0.46}
    componentLinearVelScale {z=0.46}
    physicsBodyDefinitions [{
      collisionType=CustomSphere
      sphereCollisionRadius=2
      boundBone {boneName=tent_2_l_4}
      boxExtents=<11 20 20>
      constraintSetup {angularLimitsMax=<0 20 60>, angularLimitsMin=<0 -20 -60>, angularTarget=<0 0 0>}
      localJointOffset=<10 0 0>
     }]
    sphericalLimits [
     {limitRadius=31.637, drivingBone {boneName=spine_05}, sphereLocalOffset=<-1.5465 9.2897 0>}
     {limitRadius=16, drivingBone {boneName=clavicle_l}, sphereLocalOffset=<15.524 -1.769 -0.679>}
     {limitRadius=16, drivingBone {boneName=clavicle_r}, sphereLocalOffset=<-15.524 -1.769 0.679>}
    ]
   }
   n45 = Anim Dynamics - Bone: tent_2_l_5(ComponentPose=@n44, Alpha=1, ExternalForce="0,0,0") {
    angularDampingOverride=1
    angularSpringConstant=1500
    bAngularSpring=true
    bGravityOverrideInSimSpace=true
    bOverrideAngularDamping=true
    bUseSphericalLimits=true
    gravityScale=0.35
    simulationSpace=RootRelative
    alphaBoolBlend {blendOption=Linear}
    alphaScaleBiasClamp {scale=0.06}
    boundBone {boneName=tent_2_l_5}
    componentLinearAccScale {z=0.46}
    componentLinearVelScale {z=0.46}
    physicsBodyDefinitions [{
      collisionType=CustomSphere
      sphereCollisionRadius=2
      boundBone {boneName=tent_2_l_5}
      boxExtents=<11 20 20>
      constraintSetup {angularLimitsMax=<0 20 60>, angularLimitsMin=<0 -20 -60>, angularTarget=<0 0 0>}
      localJointOffset=<10 0 0>
     }]
    sphericalLimits [
     {limitRadius=31.637, drivingBone {boneName=spine_05}, sphereLocalOffset=<-1.5465 9.2897 0>}
     {limitRadius=16, drivingBone {boneName=clavicle_l}, sphereLocalOffset=<15.524 -1.769 -0.679>}
     {limitRadius=16, drivingBone {boneName=clavicle_r}, sphereLocalOffset=<-15.524 -1.769 0.679>}
    ]
   }
   n46 = Anim Dynamics - Bone: tent_2_l_6(ComponentPose=@n45, Alpha=1, ExternalForce="0,0,0") {
    angularDampingOverride=1
    angularSpringConstant=1500
    bAngularSpring=true
    bGravityOverrideInSimSpace=true
    bOverrideAngularDamping=true
    bUseSphericalLimits=true
    gravityScale=0.35
    simulationSpace=RootRelative
    alphaBoolBlend {blendOption=Linear}
    alphaScaleBiasClamp {scale=0.06}
    boundBone {boneName=tent_2_l_6}
    componentLinearAccScale {z=0.45}
    componentLinearVelScale {z=0.45}
    physicsBodyDefinitions [{
      collisionType=CustomSphere
      sphereCollisionRadius=2
      boundBone {boneName=tent_2_l_6}
      boxExtents=<11 20 20>
      constraintSetup {angularLimitsMax=<0 20 60>, angularLimitsMin=<0 -20 -60>, angularTarget=<1 0 0>}
      localJointOffset=<10 0 0>
     }]
    sphericalLimits [
     {limitRadius=31.637, drivingBone {boneName=spine_05}, sphereLocalOffset=<-1.5465 9.2897 0>}
     {limitRadius=16, drivingBone {boneName=clavicle_l}, sphereLocalOffset=<15.524 -1.769 -0.679>}
     {limitRadius=16, drivingBone {boneName=clavicle_r}, sphereLocalOffset=<-15.524 -1.769 0.679>}
    ]
   }
   n47 = Anim Dynamics - Bone: tent_2_l_7(ComponentPose=@n46, Alpha=1, ExternalForce="0,0,0") {
    angularDampingOverride=1
    angularSpringConstant=1500
    bAngularSpring=true
    bGravityOverrideInSimSpace=true
    bOverrideAngularDamping=true
    bUseSphericalLimits=true
    gravityScale=0.35
    simulationSpace=RootRelative
    alphaBoolBlend {blendOption=Linear}
    alphaScaleBiasClamp {scale=0.06}
    boundBone {boneName=tent_2_l_7}
    componentLinearAccScale {z=0.46}
    componentLinearVelScale {z=0.46}
    physicsBodyDefinitions [{
      collisionType=CustomSphere
      sphereCollisionRadius=2
      boundBone {boneName=tent_2_l_7}
      boxExtents=<11 20 20>
      constraintSetup {angularLimitsMax=<0 20 60>, angularLimitsMin=<0 -20 -60>, angularTarget=<1 0 0>}
      localJointOffset=<10 0 0>
     }]
    sphericalLimits [
     {limitRadius=31.637, drivingBone {boneName=spine_05}, sphereLocalOffset=<-1.5465 9.2897 0>}
     {limitRadius=16, drivingBone {boneName=clavicle_l}, sphereLocalOffset=<15.524 -1.769 -0.679>}
     {limitRadius=16, drivingBone {boneName=clavicle_r}, sphereLocalOffset=<-15.524 -1.769 0.679>}
    ]
   }
   n48 = Anim Dynamics - Bone: tent_2_l_8(ComponentPose=@n47, Alpha=1, ExternalForce="0,0,0") {
    angularDampingOverride=1
    angularSpringConstant=1500
    bAngularSpring=true
    bGravityOverrideInSimSpace=true
    bOverrideAngularDamping=true
    bUseSphericalLimits=true
    gravityScale=0.35
    simulationSpace=RootRelative
    alphaBoolBlend {blendOption=Linear}
    alphaScaleBiasClamp {scale=0.45}
    boundBone {boneName=tent_2_l_8}
    componentLinearAccScale {z=0.46}
    componentLinearVelScale {z=0.46}
    physicsBodyDefinitions [{
      collisionType=CustomSphere
      sphereCollisionRadius=2
      boundBone {boneName=tent_2_l_8}
      boxExtents=<11 20 20>
      constraintSetup {angularLimitsMax=<0 20 60>, angularLimitsMin=<0 -20 -60>, angularTarget=<0 0 0>}
      localJointOffset=<10 0 0>
     }]
    sphericalLimits [
     {limitRadius=31.637, drivingBone {boneName=spine_05}, sphereLocalOffset=<-1.5465 9.2897 0>}
     {limitRadius=16, drivingBone {boneName=clavicle_l}, sphereLocalOffset=<15.524 -1.769 -0.679>}
     {limitRadius=16, drivingBone {boneName=clavicle_r}, sphereLocalOffset=<-15.524 -1.769 0.679>}
    ]
   }
   n49 = Anim Dynamics - Bone: tent_2_l_9(ComponentPose=@n48, Alpha=1, ExternalForce="0,0,0") {
    angularDampingOverride=1
    angularSpringConstant=1500
    bAngularSpring=true
    bGravityOverrideInSimSpace=true
    bOverrideAngularDamping=true
    bUseSphericalLimits=true
    gravityScale=0.35
    simulationSpace=RootRelative
    alphaBoolBlend {blendOption=Linear}
    alphaScaleBiasClamp {scale=0.45}
    boundBone {boneName=tent_2_l_9}
    componentLinearAccScale {z=0.46}
    componentLinearVelScale {z=0.46}
    physicsBodyDefinitions [{
      collisionType=CustomSphere
      sphereCollisionRadius=2
      boundBone {boneName=tent_2_l_9}
      boxExtents=<11 20 20>
      constraintSetup {angularLimitsMax=<0 20 60>, angularLimitsMin=<0 -20 -60>, angularTarget=<0 0 0>}
      localJointOffset=<10 0 0>
     }]
    sphericalLimits [
     {limitRadius=31.637, drivingBone {boneName=spine_05}, sphereLocalOffset=<-1.5465 9.2897 0>}
     {limitRadius=16, drivingBone {boneName=clavicle_l}, sphereLocalOffset=<15.524 -1.769 -0.679>}
     {limitRadius=16, drivingBone {boneName=clavicle_r}, sphereLocalOffset=<-15.524 -1.769 0.679>}
    ]
   }
   n50 = Anim Dynamics - Bone: tent_2_r_2(ComponentPose=@n49, Alpha=1, ExternalForce="0,0,0") {
    angularDampingOverride=1
    angularSpringConstant=1500
    bAngularSpring=true
    bGravityOverrideInSimSpace=true
    bOverrideAngularDamping=true
    bUseSphericalLimits=true
    gravityScale=0.35
    simulationSpace=RootRelative
    alphaBoolBlend {blendOption=Linear}
    alphaScaleBiasClamp {scale=0.03}
    boundBone {boneName=tent_2_r_2}
    componentLinearAccScale {z=0.45}
    componentLinearVelScale {z=0.45}
    physicsBodyDefinitions [{
      collisionType=CustomSphere
      sphereCollisionRadius=2
      boundBone {boneName=tent_2_r_2}
      boxExtents=<12 20 20>
      constraintSetup {angularLimitsMax=<0 20 60>, angularLimitsMin=<0 -20 -60>, angularTarget=<1 0 0>}
      localJointOffset=<10 0 0>
     }]
    sphericalLimits [
     {limitRadius=31.637, drivingBone {boneName=spine_05}, sphereLocalOffset=<-1.5465 9.2897 0>}
     {limitRadius=16, drivingBone {boneName=clavicle_l}, sphereLocalOffset=<15.524 -1.769 -0.679>}
     {limitRadius=16, drivingBone {boneName=clavicle_r}, sphereLocalOffset=<-15.524 -1.769 0.679>}
    ]
   }
   n51 = Anim Dynamics - Bone: tent_2_r_3(ComponentPose=@n50, Alpha=1, ExternalForce="0,0,0") {
    angularDampingOverride=1
    angularSpringConstant=1500
    bAngularSpring=true
    bGravityOverrideInSimSpace=true
    bOverrideAngularDamping=true
    bUseSphericalLimits=true
    gravityScale=0.35
    simulationSpace=RootRelative
    alphaBoolBlend {blendOption=Linear}
    alphaScaleBiasClamp {scale=0.06}
    boundBone {boneName=tent_2_r_3}
    componentLinearAccScale {z=0.46}
    componentLinearVelScale {z=0.46}
    physicsBodyDefinitions [{
      collisionType=CustomSphere
      sphereCollisionRadius=2
      boundBone {boneName=tent_2_r_3}
      boxExtents=<12 20 20>
      constraintSetup {angularLimitsMax=<0 20 60>, angularLimitsMin=<0 -20 -60>, angularTarget=<1 0 0>}
      localJointOffset=<10 0 0>
     }]
    sphericalLimits [
     {limitRadius=31.637, drivingBone {boneName=spine_05}, sphereLocalOffset=<-1.5465 9.2897 0>}
     {limitRadius=16, drivingBone {boneName=clavicle_l}, sphereLocalOffset=<15.524 -1.769 -0.679>}
     {limitRadius=16, drivingBone {boneName=clavicle_r}, sphereLocalOffset=<-15.524 -1.769 0.679>}
    ]
   }
   n52 = Anim Dynamics - Bone: tent_2_r_4(ComponentPose=@n51, Alpha=1, ExternalForce="0,0,0") {
    angularDampingOverride=1
    angularSpringConstant=1500
    bAngularSpring=true
    bGravityOverrideInSimSpace=true
    bOverrideAngularDamping=true
    bUseSphericalLimits=true
    gravityScale=0.35
    simulationSpace=RootRelative
    alphaBoolBlend {blendOption=Linear}
    alphaScaleBiasClamp {scale=0.06}
    boundBone {boneName=tent_2_r_4}
    componentLinearAccScale {z=0.46}
    componentLinearVelScale {z=0.46}
    physicsBodyDefinitions [{
      collisionType=CustomSphere
      sphereCollisionRadius=2
      boundBone {boneName=tent_2_r_4}
      boxExtents=<12 20 20>
      constraintSetup {angularLimitsMax=<0 20 60>, angularLimitsMin=<0 -20 -60>, angularTarget=<0 0 0>}
      localJointOffset=<10 0 0>
     }]
    sphericalLimits [
     {limitRadius=31.637, drivingBone {boneName=spine_05}, sphereLocalOffset=<-1.5465 9.2897 0>}
     {limitRadius=16, drivingBone {boneName=clavicle_l}, sphereLocalOffset=<15.524 -1.769 -0.679>}
     {limitRadius=16, drivingBone {boneName=clavicle_r}, sphereLocalOffset=<-15.524 -1.769 0.679>}
    ]
   }
   n53 = Anim Dynamics - Bone: tent_2_r_5(ComponentPose=@n52, Alpha=1, ExternalForce="0,0,0") {
    angularDampingOverride=1
    angularSpringConstant=1500
    bAngularSpring=true
    bGravityOverrideInSimSpace=true
    bOverrideAngularDamping=true
    bUseSphericalLimits=true
    gravityScale=0.35
    simulationSpace=RootRelative
    alphaBoolBlend {blendOption=Linear}
    alphaScaleBiasClamp {scale=0.06}
    boundBone {boneName=tent_2_r_5}
    componentLinearAccScale {z=0.46}
    componentLinearVelScale {z=0.46}
    physicsBodyDefinitions [{
      collisionType=CustomSphere
      sphereCollisionRadius=2
      boundBone {boneName=tent_2_r_5}
      boxExtents=<12 20 20>
      constraintSetup {angularLimitsMax=<0 20 60>, angularLimitsMin=<0 -20 -60>, angularTarget=<0 0 0>}
      localJointOffset=<10 0 0>
     }]
    sphericalLimits [
     {limitRadius=31.637, drivingBone {boneName=spine_05}, sphereLocalOffset=<-1.5465 9.2897 0>}
     {limitRadius=16, drivingBone {boneName=clavicle_l}, sphereLocalOffset=<15.524 -1.769 -0.679>}
     {limitRadius=16, drivingBone {boneName=clavicle_r}, sphereLocalOffset=<-15.524 -1.769 0.679>}
    ]
   }
   n54 = Anim Dynamics - Bone: tent_2_r_6(ComponentPose=@n53, Alpha=1, ExternalForce="0,0,0") {
    angularDampingOverride=1
    angularSpringConstant=1500
    bAngularSpring=true
    bGravityOverrideInSimSpace=true
    bOverrideAngularDamping=true
    bUseSphericalLimits=true
    gravityScale=0.35
    simulationSpace=RootRelative
    alphaBoolBlend {blendOption=Linear}
    alphaScaleBiasClamp {scale=0.06}
    boundBone {boneName=tent_2_r_6}
    componentLinearAccScale {z=0.45}
    componentLinearVelScale {z=0.45}
    physicsBodyDefinitions [{
      collisionType=CustomSphere
      sphereCollisionRadius=2
      boundBone {boneName=tent_2_r_6}
      boxExtents=<12 20 20>
      constraintSetup {angularLimitsMax=<0 20 60>, angularLimitsMin=<0 -20 -60>, angularTarget=<1 0 0>}
      localJointOffset=<10 0 0>
     }]
    sphericalLimits [
     {limitRadius=31.637, drivingBone {boneName=spine_05}, sphereLocalOffset=<-1.5465 9.2897 0>}
     {limitRadius=16, drivingBone {boneName=clavicle_l}, sphereLocalOffset=<15.524 -1.769 -0.679>}
     {limitRadius=16, drivingBone {boneName=clavicle_r}, sphereLocalOffset=<-15.524 -1.769 0.679>}
    ]
   }
   n55 = Anim Dynamics - Bone: tent_2_r_7(ComponentPose=@n54, Alpha=1, ExternalForce="0,0,0") {
    angularDampingOverride=1
    angularSpringConstant=1500
    bAngularSpring=true
    bGravityOverrideInSimSpace=true
    bOverrideAngularDamping=true
    bUseSphericalLimits=true
    gravityScale=0.35
    simulationSpace=RootRelative
    alphaBoolBlend {blendOption=Linear}
    alphaScaleBiasClamp {scale=0.06}
    boundBone {boneName=tent_2_r_7}
    componentLinearAccScale {z=0.46}
    componentLinearVelScale {z=0.46}
    physicsBodyDefinitions [{
      collisionType=CustomSphere
      sphereCollisionRadius=2
      boundBone {boneName=tent_2_r_7}
      boxExtents=<12 20 20>
      constraintSetup {angularLimitsMax=<0 20 60>, angularLimitsMin=<0 -20 -60>, angularTarget=<1 0 0>}
      localJointOffset=<10 0 0>
     }]
    sphericalLimits [
     {limitRadius=31.637, drivingBone {boneName=spine_05}, sphereLocalOffset=<-1.5465 9.2897 0>}
     {limitRadius=16, drivingBone {boneName=clavicle_l}, sphereLocalOffset=<15.524 -1.769 -0.679>}
     {limitRadius=16, drivingBone {boneName=clavicle_r}, sphereLocalOffset=<-15.524 -1.769 0.679>}
    ]
   }
   n56 = Anim Dynamics - Bone: tent_2_r_8(ComponentPose=@n55, Alpha=1, ExternalForce="0,0,0") {
    angularDampingOverride=1
    angularSpringConstant=1500
    bAngularSpring=true
    bGravityOverrideInSimSpace=true
    bOverrideAngularDamping=true
    bUseSphericalLimits=true
    gravityScale=0.35
    simulationSpace=RootRelative
    alphaBoolBlend {blendOption=Linear}
    alphaScaleBiasClamp {scale=0.45}
    boundBone {boneName=tent_2_r_8}
    componentLinearAccScale {z=0.46}
    componentLinearVelScale {z=0.46}
    physicsBodyDefinitions [{
      collisionType=CustomSphere
      sphereCollisionRadius=2
      boundBone {boneName=tent_2_r_8}
      boxExtents=<12 20 20>
      constraintSetup {angularLimitsMax=<0 20 60>, angularLimitsMin=<0 -20 -60>, angularTarget=<0 0 0>}
      localJointOffset=<10 0 0>
     }]
    sphericalLimits [
     {limitRadius=31.637, drivingBone {boneName=spine_05}, sphereLocalOffset=<-1.5465 9.2897 0>}
     {limitRadius=16, drivingBone {boneName=clavicle_l}, sphereLocalOffset=<15.524 -1.769 -0.679>}
     {limitRadius=16, drivingBone {boneName=clavicle_r}, sphereLocalOffset=<-15.524 -1.769 0.679>}
    ]
   }
   n57 = Anim Dynamics - Bone: tent_2_r_9(ComponentPose=@n56, Alpha=1, ExternalForce="0,0,0") {
    angularDampingOverride=1
    angularSpringConstant=1500
    bAngularSpring=true
    bGravityOverrideInSimSpace=true
    bOverrideAngularDamping=true
    bUseSphericalLimits=true
    gravityScale=0.35
    simulationSpace=RootRelative
    alphaBoolBlend {blendOption=Linear}
    alphaScaleBiasClamp {scale=0.45}
    boundBone {boneName=tent_2_r_9}
    componentLinearAccScale {z=0.46}
    componentLinearVelScale {z=0.46}
    physicsBodyDefinitions [{
      collisionType=CustomSphere
      sphereCollisionRadius=2
      boundBone {boneName=tent_2_r_9}
      boxExtents=<12 20 20>
      constraintSetup {angularLimitsMax=<0 20 60>, angularLimitsMin=<0 -20 -60>, angularTarget=<0 0 0>}
      localJointOffset=<10 0 0>
     }]
    sphericalLimits [
     {limitRadius=31.637, drivingBone {boneName=spine_05}, sphereLocalOffset=<-1.5465 9.2897 0>}
     {limitRadius=16, drivingBone {boneName=clavicle_l}, sphereLocalOffset=<15.524 -1.769 -0.679>}
     {limitRadius=16, drivingBone {boneName=clavicle_r}, sphereLocalOffset=<-15.524 -1.769 0.679>}
    ]
   }
   n58 = Anim Dynamics - Bone: tent_1_r_2(ComponentPose=@n57, Alpha=1, ExternalForce="0,0,0") {
    angularDampingOverride=1
    angularSpringConstant=1500
    bAngularSpring=true
    bGravityOverrideInSimSpace=true
    bOverrideAngularDamping=true
    bUseSphericalLimits=true
    gravityScale=0.35
    simulationSpace=RootRelative
    alphaBoolBlend {blendOption=Linear}
    alphaScaleBiasClamp {scale=0.03}
    boundBone {boneName=tent_1_r_2}
    componentLinearAccScale {z=0.45}
    componentLinearVelScale {z=0.45}
    physicsBodyDefinitions [{
      collisionType=CustomSphere
      sphereCollisionRadius=2
      boundBone {boneName=tent_1_r_2}
      boxExtents=<9 20 20>
      constraintSetup {angularLimitsMax=<0 20 60>, angularLimitsMin=<0 -20 -60>, angularTarget=<1 0 0>}
      localJointOffset=<10 0 0>
     }]
    sphericalLimits [
     {limitRadius=31.637, drivingBone {boneName=spine_05}, sphereLocalOffset=<-1.5465 9.2897 0>}
     {limitRadius=16, drivingBone {boneName=clavicle_l}, sphereLocalOffset=<15.524 -1.769 -0.679>}
     {limitRadius=16, drivingBone {boneName=clavicle_r}, sphereLocalOffset=<-15.524 -1.769 0.679>}
    ]
   }
   n59 = Anim Dynamics - Bone: tent_1_r_3(ComponentPose=@n58, Alpha=1, ExternalForce="0,0,0") {
    angularDampingOverride=1
    angularSpringConstant=1500
    bAngularSpring=true
    bGravityOverrideInSimSpace=true
    bOverrideAngularDamping=true
    bUseSphericalLimits=true
    gravityScale=0.35
    simulationSpace=RootRelative
    alphaBoolBlend {blendOption=Linear}
    alphaScaleBiasClamp {scale=0.1}
    boundBone {boneName=tent_1_r_3}
    componentLinearAccScale {z=0.46}
    componentLinearVelScale {z=0.46}
    physicsBodyDefinitions [{
      collisionType=CustomSphere
      sphereCollisionRadius=2
      boundBone {boneName=tent_1_r_3}
      boxExtents=<9 20 20>
      constraintSetup {angularLimitsMax=<0 20 60>, angularLimitsMin=<0 -20 -60>, angularTarget=<1 0 0>}
      localJointOffset=<10 0 0>
     }]
    sphericalLimits [
     {limitRadius=31.637, drivingBone {boneName=spine_05}, sphereLocalOffset=<-1.5465 9.2897 0>}
     {limitRadius=16, drivingBone {boneName=clavicle_l}, sphereLocalOffset=<15.524 -1.769 -0.679>}
     {limitRadius=16, drivingBone {boneName=clavicle_r}, sphereLocalOffset=<-15.524 -1.769 0.679>}
    ]
   }
   n60 = Anim Dynamics - Bone: tent_1_r_4(ComponentPose=@n59, Alpha=1, ExternalForce="0,0,0") {
    angularDampingOverride=1
    angularSpringConstant=1500
    bAngularSpring=true
    bGravityOverrideInSimSpace=true
    bOverrideAngularDamping=true
    bUseSphericalLimits=true
    gravityScale=0.35
    simulationSpace=RootRelative
    alphaBoolBlend {blendOption=Linear}
    alphaScaleBiasClamp {scale=0.15}
    boundBone {boneName=tent_1_r_4}
    componentLinearAccScale {z=0.46}
    componentLinearVelScale {z=0.46}
    physicsBodyDefinitions [{
      collisionType=CustomSphere
      sphereCollisionRadius=2
      boundBone {boneName=tent_1_r_4}
      boxExtents=<9 20 20>
      constraintSetup {angularLimitsMax=<0 20 60>, angularLimitsMin=<0 -20 -60>, angularTarget=<0 0 0>}
      localJointOffset=<10 0 0>
     }]
    sphericalLimits [
     {limitRadius=31.637, drivingBone {boneName=spine_05}, sphereLocalOffset=<-1.5465 9.2897 0>}
     {limitRadius=16, drivingBone {boneName=clavicle_l}, sphereLocalOffset=<15.524 -1.769 -0.679>}
     {limitRadius=16, drivingBone {boneName=clavicle_r}, sphereLocalOffset=<-15.524 -1.769 0.679>}
    ]
   }
   n61 = Anim Dynamics - Bone: tent_1_r_5(ComponentPose=@n60, Alpha=1, ExternalForce="0,0,0") {
    angularDampingOverride=1
    angularSpringConstant=1500
    bAngularSpring=true
    bGravityOverrideInSimSpace=true
    bOverrideAngularDamping=true
    bUseSphericalLimits=true
    gravityScale=0.35
    simulationSpace=RootRelative
    alphaBoolBlend {blendOption=Linear}
    alphaScaleBiasClamp {scale=0.15}
    boundBone {boneName=tent_1_r_5}
    componentLinearAccScale {z=0.46}
    componentLinearVelScale {z=0.46}
    physicsBodyDefinitions [{
      collisionType=CustomSphere
      sphereCollisionRadius=2
      boundBone {boneName=tent_1_r_5}
      boxExtents=<9 20 20>
      constraintSetup {angularLimitsMax=<0 20 60>, angularLimitsMin=<0 -20 -60>, angularTarget=<0 0 0>}
      localJointOffset=<10 0 0>
     }]
    sphericalLimits [
     {limitRadius=31.637, drivingBone {boneName=spine_05}, sphereLocalOffset=<-1.5465 9.2897 0>}
     {limitRadius=16, drivingBone {boneName=clavicle_l}, sphereLocalOffset=<15.524 -1.769 -0.679>}
     {limitRadius=16, drivingBone {boneName=clavicle_r}, sphereLocalOffset=<-15.524 -1.769 0.679>}
    ]
   }
   n62 = Anim Dynamics - Bone: tent_1_r_6(ComponentPose=@n61, Alpha=1, ExternalForce="0,0,0") {
    angularDampingOverride=1
    angularSpringConstant=1500
    bAngularSpring=true
    bGravityOverrideInSimSpace=true
    bOverrideAngularDamping=true
    bUseSphericalLimits=true
    gravityScale=0.35
    simulationSpace=RootRelative
    alphaBoolBlend {blendOption=Linear}
    alphaScaleBiasClamp {scale=0.25}
    boundBone {boneName=tent_1_r_6}
    componentLinearAccScale {z=0.45}
    componentLinearVelScale {z=0.45}
    physicsBodyDefinitions [{
      collisionType=CustomSphere
      sphereCollisionRadius=2
      boundBone {boneName=tent_1_r_6}
      boxExtents=<9 20 20>
      constraintSetup {angularLimitsMax=<0 20 60>, angularLimitsMin=<0 -20 -60>, angularTarget=<1 0 0>}
      localJointOffset=<10 0 0>
     }]
    sphericalLimits [
     {limitRadius=31.637, drivingBone {boneName=spine_05}, sphereLocalOffset=<-1.5465 9.2897 0>}
     {limitRadius=16, drivingBone {boneName=clavicle_l}, sphereLocalOffset=<15.524 -1.769 -0.679>}
     {limitRadius=16, drivingBone {boneName=clavicle_r}, sphereLocalOffset=<-15.524 -1.769 0.679>}
    ]
   }
   n63 = Anim Dynamics - Bone: tent_1_r_7(ComponentPose=@n62, Alpha=1, ExternalForce="0,0,0") {
    angularDampingOverride=1
    angularSpringConstant=1500
    bAngularSpring=true
    bGravityOverrideInSimSpace=true
    bOverrideAngularDamping=true
    bUseSphericalLimits=true
    gravityScale=0.35
    simulationSpace=RootRelative
    alphaBoolBlend {blendOption=Linear}
    alphaScaleBiasClamp {scale=0.25}
    boundBone {boneName=tent_1_r_7}
    componentLinearAccScale {z=0.46}
    componentLinearVelScale {z=0.46}
    physicsBodyDefinitions [{
      collisionType=CustomSphere
      sphereCollisionRadius=2
      boundBone {boneName=tent_1_r_7}
      boxExtents=<9 20 20>
      constraintSetup {angularLimitsMax=<0 20 60>, angularLimitsMin=<0 -20 -60>, angularTarget=<1 0 0>}
      localJointOffset=<10 0 0>
     }]
    sphericalLimits [
     {limitRadius=31.637, drivingBone {boneName=spine_05}, sphereLocalOffset=<-1.5465 9.2897 0>}
     {limitRadius=16, drivingBone {boneName=clavicle_l}, sphereLocalOffset=<15.524 -1.769 -0.679>}
     {limitRadius=16, drivingBone {boneName=clavicle_r}, sphereLocalOffset=<-15.524 -1.769 0.679>}
    ]
   }
   n64 = Anim Dynamics - Bone: tent_1_r_8(ComponentPose=@n63, Alpha=1, ExternalForce="0,0,0") {
    angularDampingOverride=1
    angularSpringConstant=1500
    bAngularSpring=true
    bGravityOverrideInSimSpace=true
    bOverrideAngularDamping=true
    bUseSphericalLimits=true
    gravityScale=0.35
    simulationSpace=RootRelative
    alphaBoolBlend {blendOption=Linear}
    alphaScaleBiasClamp {scale=0.45}
    boundBone {boneName=tent_1_r_8}
    componentLinearAccScale {z=0.46}
    componentLinearVelScale {z=0.46}
    physicsBodyDefinitions [{
      collisionType=CustomSphere
      sphereCollisionRadius=2
      boundBone {boneName=tent_1_r_8}
      boxExtents=<9 20 20>
      constraintSetup {angularLimitsMax=<0 20 60>, angularLimitsMin=<0 -20 -60>, angularTarget=<0 0 0>}
      localJointOffset=<10 0 0>
     }]
    sphericalLimits [
     {limitRadius=31.637, drivingBone {boneName=spine_05}, sphereLocalOffset=<-1.5465 9.2897 0>}
     {limitRadius=16, drivingBone {boneName=clavicle_l}, sphereLocalOffset=<15.524 -1.769 -0.679>}
     {limitRadius=16, drivingBone {boneName=clavicle_r}, sphereLocalOffset=<-15.524 -1.769 0.679>}
    ]
   }
   n65 = Anim Dynamics - Bone: tent_1_r_9(ComponentPose=@n64, Alpha=1, ExternalForce="0,0,0") {
    angularDampingOverride=1
    angularSpringConstant=1500
    bAngularSpring=true
    bGravityOverrideInSimSpace=true
    bOverrideAngularDamping=true
    bUseSphericalLimits=true
    gravityScale=0.35
    simulationSpace=RootRelative
    alphaBoolBlend {blendOption=Linear}
    alphaScaleBiasClamp {scale=0.45}
    boundBone {boneName=tent_1_r_9}
    componentLinearAccScale {z=0.46}
    componentLinearVelScale {z=0.46}
    physicsBodyDefinitions [{
      collisionType=CustomSphere
      sphereCollisionRadius=2
      boundBone {boneName=tent_1_r_9}
      boxExtents=<9 20 20>
      constraintSetup {angularLimitsMax=<0 20 60>, angularLimitsMin=<0 -20 -60>, angularTarget=<0 0 0>}
      localJointOffset=<10 0 0>
     }]
    sphericalLimits [
     {limitRadius=31.637, drivingBone {boneName=spine_05}, sphereLocalOffset=<-1.5465 9.2897 0>}
     {limitRadius=16, drivingBone {boneName=clavicle_l}, sphereLocalOffset=<15.524 -1.769 -0.679>}
     {limitRadius=16, drivingBone {boneName=clavicle_r}, sphereLocalOffset=<-15.524 -1.769 0.679>}
    ]
   }
   n66 = Anim Dynamics - Bone: tent_c_2(ComponentPose=@n65, Alpha=1, ExternalForce="0,0,0") {
    angularDampingOverride=1
    angularSpringConstant=1500
    bAngularSpring=true
    bGravityOverrideInSimSpace=true
    bOverrideAngularDamping=true
    bUseSphericalLimits=true
    gravityScale=0.35
    simulationSpace=RootRelative
    alphaBoolBlend {blendOption=Linear}
    alphaScaleBiasClamp {scale=0.03}
    boundBone {boneName=tent_c_2}
    componentLinearAccScale {z=0.45}
    componentLinearVelScale {z=0.45}
    physicsBodyDefinitions [{
      collisionType=CustomSphere
      sphereCollisionRadius=2
      boundBone {boneName=tent_c_2}
      boxExtents=<11 20 20>
      constraintSetup {angularLimitsMax=<0 20 60>, angularLimitsMin=<0 -20 -60>, angularTarget=<1 0 0>}
      localJointOffset=<10 0 0>
     }]
    sphericalLimits [
     {limitRadius=31.637, drivingBone {boneName=spine_05}, sphereLocalOffset=<-1.5465 9.2897 0>}
     {limitRadius=16, drivingBone {boneName=clavicle_l}, sphereLocalOffset=<15.524 -1.769 -0.679>}
     {limitRadius=16, drivingBone {boneName=clavicle_r}, sphereLocalOffset=<-15.524 -1.769 0.679>}
    ]
   }
   n67 = Anim Dynamics - Bone: tent_c_3(ComponentPose=@n66, Alpha=1, ExternalForce="0,0,0") {
    angularDampingOverride=1
    angularSpringConstant=1500
    bAngularSpring=true
    bGravityOverrideInSimSpace=true
    bOverrideAngularDamping=true
    bUseSphericalLimits=true
    gravityScale=0.35
    simulationSpace=RootRelative
    alphaBoolBlend {blendOption=Linear}
    alphaScaleBiasClamp {scale=0.06}
    boundBone {boneName=tent_c_3}
    componentLinearAccScale {z=0.46}
    componentLinearVelScale {z=0.46}
    physicsBodyDefinitions [{
      collisionType=CustomSphere
      sphereCollisionRadius=2
      boundBone {boneName=tent_c_3}
      boxExtents=<11 20 20>
      constraintSetup {angularLimitsMax=<0 20 60>, angularLimitsMin=<0 -20 -60>, angularTarget=<1 0 0>}
      localJointOffset=<10 0 0>
     }]
    sphericalLimits [
     {limitRadius=31.637, drivingBone {boneName=spine_05}, sphereLocalOffset=<-1.5465 9.2897 0>}
     {limitRadius=16, drivingBone {boneName=clavicle_l}, sphereLocalOffset=<15.524 -1.769 -0.679>}
     {limitRadius=16, drivingBone {boneName=clavicle_r}, sphereLocalOffset=<-15.524 -1.769 0.679>}
    ]
   }
   n68 = Anim Dynamics - Bone: tent_c_4(ComponentPose=@n67, Alpha=1, ExternalForce="0,0,0") {
    angularDampingOverride=1
    angularSpringConstant=1500
    bAngularSpring=true
    bGravityOverrideInSimSpace=true
    bOverrideAngularDamping=true
    bUseSphericalLimits=true
    gravityScale=0.35
    simulationSpace=RootRelative
    alphaBoolBlend {blendOption=Linear}
    alphaScaleBiasClamp {scale=0.06}
    boundBone {boneName=tent_c_4}
    componentLinearAccScale {z=0.46}
    componentLinearVelScale {z=0.46}
    physicsBodyDefinitions [{
      collisionType=CustomSphere
      sphereCollisionRadius=2
      boundBone {boneName=tent_c_4}
      boxExtents=<11 20 20>
      constraintSetup {angularLimitsMax=<0 20 60>, angularLimitsMin=<0 -20 -60>, angularTarget=<0 0 0>}
      localJointOffset=<10 0 0>
     }]
    sphericalLimits [
     {limitRadius=31.637, drivingBone {boneName=spine_05}, sphereLocalOffset=<-1.5465 9.2897 0>}
     {limitRadius=16, drivingBone {boneName=clavicle_l}, sphereLocalOffset=<15.524 -1.769 -0.679>}
     {limitRadius=16, drivingBone {boneName=clavicle_r}, sphereLocalOffset=<-15.524 -1.769 0.679>}
    ]
   }
   n69 = Anim Dynamics - Bone: tent_c_5(ComponentPose=@n68, Alpha=1, ExternalForce="0,0,0") {
    angularDampingOverride=1
    angularSpringConstant=1500
    bAngularSpring=true
    bGravityOverrideInSimSpace=true
    bOverrideAngularDamping=true
    bUseSphericalLimits=true
    gravityScale=0.35
    simulationSpace=RootRelative
    alphaBoolBlend {blendOption=Linear}
    alphaScaleBiasClamp {scale=0.08}
    boundBone {boneName=tent_c_5}
    componentLinearAccScale {z=0.46}
    componentLinearVelScale {z=0.46}
    physicsBodyDefinitions [{
      collisionType=CustomSphere
      sphereCollisionRadius=2
      boundBone {boneName=tent_c_5}
      boxExtents=<11 20 20>
      constraintSetup {angularLimitsMax=<0 20 60>, angularLimitsMin=<0 -20 -60>, angularTarget=<0 0 0>}
      localJointOffset=<10 0 0>
     }]
    sphericalLimits [
     {limitRadius=31.637, drivingBone {boneName=spine_05}, sphereLocalOffset=<-1.5465 9.2897 0>}
     {limitRadius=16, drivingBone {boneName=clavicle_l}, sphereLocalOffset=<15.524 -1.769 -0.679>}
     {limitRadius=16, drivingBone {boneName=clavicle_r}, sphereLocalOffset=<-15.524 -1.769 0.679>}
    ]
   }
   n70 = Anim Dynamics - Bone: tent_c_6(ComponentPose=@n69, Alpha=1, ExternalForce="0,0,0") {
    angularDampingOverride=1
    angularSpringConstant=1500
    bAngularSpring=true
    bGravityOverrideInSimSpace=true
    bOverrideAngularDamping=true
    bUseSphericalLimits=true
    gravityScale=0.35
    simulationSpace=RootRelative
    alphaBoolBlend {blendOption=Linear}
    alphaScaleBiasClamp {scale=0.08}
    boundBone {boneName=tent_c_6}
    componentLinearAccScale {z=0.45}
    componentLinearVelScale {z=0.45}
    physicsBodyDefinitions [{
      collisionType=CustomSphere
      sphereCollisionRadius=2
      boundBone {boneName=tent_c_6}
      boxExtents=<11 20 20>
      constraintSetup {angularLimitsMax=<0 20 60>, angularLimitsMin=<0 -20 -60>, angularTarget=<1 0 0>}
      localJointOffset=<10 0 0>
     }]
    sphericalLimits [
     {limitRadius=31.637, drivingBone {boneName=spine_05}, sphereLocalOffset=<-1.5465 9.2897 0>}
     {limitRadius=16, drivingBone {boneName=clavicle_l}, sphereLocalOffset=<15.524 -1.769 -0.679>}
     {limitRadius=16, drivingBone {boneName=clavicle_r}, sphereLocalOffset=<-15.524 -1.769 0.679>}
    ]
   }
   n71 = Anim Dynamics - Bone: tent_c_7(ComponentPose=@n70, Alpha=1, ExternalForce="0,0,0") {
    angularDampingOverride=1
    angularSpringConstant=1500
    bAngularSpring=true
    bGravityOverrideInSimSpace=true
    bOverrideAngularDamping=true
    bUseSphericalLimits=true
    gravityScale=0.35
    simulationSpace=RootRelative
    alphaBoolBlend {blendOption=Linear}
    alphaScaleBiasClamp {scale=0.08}
    boundBone {boneName=tent_c_7}
    componentLinearAccScale {z=0.46}
    componentLinearVelScale {z=0.46}
    physicsBodyDefinitions [{
      collisionType=CustomSphere
      sphereCollisionRadius=2
      boundBone {boneName=tent_c_7}
      boxExtents=<11 20 20>
      constraintSetup {angularLimitsMax=<0 20 60>, angularLimitsMin=<0 -20 -60>, angularTarget=<1 0 0>}
      localJointOffset=<10 0 0>
     }]
    sphericalLimits [
     {limitRadius=31.637, drivingBone {boneName=spine_05}, sphereLocalOffset=<-1.5465 9.2897 0>}
     {limitRadius=16, drivingBone {boneName=clavicle_l}, sphereLocalOffset=<15.524 -1.769 -0.679>}
     {limitRadius=16, drivingBone {boneName=clavicle_r}, sphereLocalOffset=<-15.524 -1.769 0.679>}
    ]
   }
   n72 = Anim Dynamics - Bone: tent_c_8(ComponentPose=@n71, Alpha=1, ExternalForce="0,0,0") {
    angularDampingOverride=1
    angularSpringConstant=1500
    bAngularSpring=true
    bGravityOverrideInSimSpace=true
    bOverrideAngularDamping=true
    bUseSphericalLimits=true
    gravityScale=0.35
    simulationSpace=RootRelative
    alphaBoolBlend {blendOption=Linear}
    alphaScaleBiasClamp {scale=0.08}
    boundBone {boneName=tent_c_8}
    componentLinearAccScale {z=0.46}
    componentLinearVelScale {z=0.46}
    physicsBodyDefinitions [{
      collisionType=CustomSphere
      sphereCollisionRadius=2
      boundBone {boneName=tent_c_8}
      boxExtents=<11 20 20>
      constraintSetup {angularLimitsMax=<0 20 60>, angularLimitsMin=<0 -20 -60>, angularTarget=<0 0 0>}
      localJointOffset=<10 0 0>
     }]
    sphericalLimits [
     {limitRadius=31.637, drivingBone {boneName=spine_05}, sphereLocalOffset=<-1.5465 9.2897 0>}
     {limitRadius=16, drivingBone {boneName=clavicle_l}, sphereLocalOffset=<15.524 -1.769 -0.679>}
     {limitRadius=16, drivingBone {boneName=clavicle_r}, sphereLocalOffset=<-15.524 -1.769 0.679>}
    ]
   }
   n73 = Anim Dynamics - Bone: tent_c_9(ComponentPose=@n72, Alpha=1, ExternalForce="0,0,0") {
    angularDampingOverride=1
    angularSpringConstant=1500
    bAngularSpring=true
    bGravityOverrideInSimSpace=true
    bOverrideAngularDamping=true
    bUseSphericalLimits=true
    gravityScale=0.35
    simulationSpace=RootRelative
    alphaBoolBlend {blendOption=Linear}
    alphaScaleBiasClamp {scale=0.08}
    boundBone {boneName=tent_c_9}
    componentLinearAccScale {z=0.46}
    componentLinearVelScale {z=0.46}
    physicsBodyDefinitions [{
      collisionType=CustomSphere
      sphereCollisionRadius=2
      boundBone {boneName=tent_c_9}
      boxExtents=<11 20 20>
      constraintSetup {angularLimitsMax=<0 20 60>, angularLimitsMin=<0 -20 -60>, angularTarget=<0 0 0>}
      localJointOffset=<10 0 0>
     }]
    sphericalLimits [
     {limitRadius=31.637, drivingBone {boneName=spine_05}, sphereLocalOffset=<-1.5465 9.2897 0>}
     {limitRadius=16, drivingBone {boneName=clavicle_l}, sphereLocalOffset=<15.524 -1.769 -0.679>}
     {limitRadius=16, drivingBone {boneName=clavicle_r}, sphereLocalOffset=<-15.524 -1.769 0.679>}
    ]
   }
   n74 = Get isNotRoaringAlpha()
   n75 = "Constraint - curve_2_l_CV_3 (PPPPTPPPP)"(
    ComponentPose=Local To Component(LocalPose=Use cached pose 'PreBezierCache'())
    Alpha=@n74
    ConstraintWeights_0=0.4276
    ConstraintWeights_1=0.1085
    ConstraintWeights_2=0.0339
    ConstraintWeights_3=0.268
    ConstraintWeights_4=0
    ConstraintWeights_5=0.0644
    ConstraintWeights_6=0.0216
    ConstraintWeights_7=0.0001
    ConstraintWeights_8=0.0654
   ) {
    alphaBoolBlend {blendOption=Linear}
    boneToModify {boneName=curve_2_l_CV_3}
    constraintSetup [
     {targetBone {boneName=clavicle_l}}
     {targetBone {boneName=spine_03}}
     {targetBone {boneName=spine_04}}
     {targetBone {boneName=spine_05}}
     {transformType=Translation, targetBone {boneName=None}}
     {targetBone {boneName=clavicle_r}}
     {targetBone {boneName=neck_02}}
     {targetBone {boneName=neck_01}}
     {targetBone {boneName=dyn_hairBase}}
    ]
    constraintWeights [1,1,1,1,1,1,1,1,1]
   }
   n76 = "Constraint - curve_2_l_CV_4 (PPPPPPP)"(
    ComponentPose=@n75
    Alpha=@n74
    ConstraintWeights_0=0.2798
    ConstraintWeights_1=0.0235
    ConstraintWeights_2=0.3517
    ConstraintWeights_3=0.1312
    ConstraintWeights_4=0.0713
    ConstraintWeights_5=0.1404
    ConstraintWeights_6=0.0021
   ) {
    alphaBoolBlend {blendOption=Linear}
    boneToModify {boneName=curve_2_l_CV_4}
    constraintSetup [
     {targetBone {boneName=clavicle_l}}
     {targetBone {boneName=spine_02}}
     {targetBone {boneName=spine_03}}
     {targetBone {boneName=spine_04}}
     {targetBone {boneName=spine_05}}
     {targetBone {boneName=clavicle_r}}
     {targetBone {boneName=pelvis}}
    ]
    constraintWeights [1,1,1,1,1,1,1]
   }
   n77 = "Constraint - curve_1_l_CV_3 (PPPPPPTPPP)"(
    ComponentPose=@n76
    Alpha=@n74
    ConstraintWeights_0=0.5371
    ConstraintWeights_1=0
    ConstraintWeights_2=0.0134
    ConstraintWeights_3=0.1993
    ConstraintWeights_4=0.0562
    ConstraintWeights_5=0.1083
    ConstraintWeights_6=0
    ConstraintWeights_7=0.0184
    ConstraintWeights_8=0
    ConstraintWeights_9=0.0013
   ) {
    alphaBoolBlend {blendOption=Linear}
    boneToModify {boneName=curve_1_l_CV_3}
    constraintSetup [
     {targetBone {boneName=clavicle_l}}
     {targetBone {boneName=None}}
     {targetBone {boneName=spine_02}}
     {targetBone {boneName=spine_03}}
     {targetBone {boneName=spine_04}}
     {targetBone {boneName=spine_05}}
     {transformType=Translation, targetBone {boneName=None}}
     {targetBone {boneName=clavicle_r}}
     {targetBone {boneName=neck_02}}
     {targetBone {boneName=dyn_hairBase}}
    ]
    constraintWeights [1,1,1,1,1,1,1,1,1,1]
   }
   n78 = "Constraint - curve_1_l_CV_4 (PPPPPPPP)"(
    ComponentPose=@n77
    Alpha=@n74
    ConstraintWeights_0=0.2137
    ConstraintWeights_1=0.0171
    ConstraintWeights_2=0.0558
    ConstraintWeights_3=0.4378
    ConstraintWeights_4=0.1502
    ConstraintWeights_5=0.0185
    ConstraintWeights_6=0.1034
    ConstraintWeights_7=0.0034
   ) {
    alphaBoolBlend {blendOption=Linear}
    boneToModify {boneName=curve_1_l_CV_4}
    constraintSetup [
     {targetBone {boneName=clavicle_l}}
     {targetBone {boneName=spine_01}}
     {targetBone {boneName=spine_02}}
     {targetBone {boneName=spine_03}}
     {targetBone {boneName=spine_04}}
     {targetBone {boneName=spine_05}}
     {targetBone {boneName=clavicle_r}}
     {targetBone {boneName=pelvis}}
    ]
    constraintWeights [1,1,1,1,1,1,1,1]
   }
   n79 = "Constraint - curve_c_CV_3 (PPPPPPPP)"(
    ComponentPose=@n78
    Alpha=@n74
    ConstraintWeights_0=0.2133
    ConstraintWeights_1=0.0987
    ConstraintWeights_2=0.0334
    ConstraintWeights_3=0.3354
    ConstraintWeights_4=0.2204
    ConstraintWeights_5=0.0295
    ConstraintWeights_6=0.0148
    ConstraintWeights_7=0.0544
   ) {
    alphaBoolBlend {blendOption=Linear}
    boneToModify {boneName=curve_c_CV_3}
    constraintSetup [
     {targetBone {boneName=clavicle_l}}
     {targetBone {boneName=spine_03}}
     {targetBone {boneName=spine_04}}
     {targetBone {boneName=spine_05}}
     {targetBone {boneName=clavicle_r}}
     {targetBone {boneName=neck_02}}
     {targetBone {boneName=neck_01}}
     {targetBone {boneName=dyn_hairBase}}
    ]
    constraintWeights [1,1,1,1,1,1,1,1]
   }
   n80 = "Constraint - curve_c_CV_4 (PPPPPP)"(
    ComponentPose=@n79
    Alpha=@n74
    ConstraintWeights_0=0.2052
    ConstraintWeights_1=0.0208
    ConstraintWeights_2=0.3565
    ConstraintWeights_3=0.1365
    ConstraintWeights_4=0.0699
    ConstraintWeights_5=0.2111
   ) {
    alphaBoolBlend {blendOption=Linear}
    boneToModify {boneName=curve_c_CV_4}
    constraintSetup [
     {targetBone {boneName=clavicle_l}}
     {targetBone {boneName=spine_02}}
     {targetBone {boneName=spine_03}}
     {targetBone {boneName=spine_04}}
     {targetBone {boneName=spine_05}}
     {targetBone {boneName=clavicle_r}}
    ]
    constraintWeights [1,1,1,1,1,1]
   }
   n81 = "Constraint - curve_2_r_CV_3 (PTPPPPPP)"(
    ComponentPose=@n80
    Alpha=@n74
    ConstraintWeights_0=0.0642
    ConstraintWeights_1=0
    ConstraintWeights_2=0.1086
    ConstraintWeights_3=0.0339
    ConstraintWeights_4=0.2679
    ConstraintWeights_5=0.4279
    ConstraintWeights_6=0.0215
    ConstraintWeights_7=0.0652
   ) {
    alphaBoolBlend {blendOption=Linear}
    boneToModify {boneName=curve_2_r_CV_3}
    constraintSetup [
     {targetBone {boneName=clavicle_l}}
     {transformType=Translation, targetBone {boneName=None}}
     {targetBone {boneName=spine_03}}
     {targetBone {boneName=spine_04}}
     {targetBone {boneName=spine_05}}
     {targetBone {boneName=clavicle_r}}
     {targetBone {boneName=neck_02}}
     {targetBone {boneName=dyn_hairBase}}
    ]
    constraintWeights [1,1,1,1,1,1,1,1]
   }
   n82 = "Constraint - curve_2_r_CV_4 (PPPPPPP)"(
    ComponentPose=@n81
    Alpha=@n74
    ConstraintWeights_0=0.1403
    ConstraintWeights_1=0.0235
    ConstraintWeights_2=0.3517
    ConstraintWeights_3=0.1312
    ConstraintWeights_4=0.0713
    ConstraintWeights_5=0.28
    ConstraintWeights_6=0.0022
   ) {
    alphaBoolBlend {blendOption=Linear}
    boneToModify {boneName=curve_2_r_CV_4}
    constraintSetup [
     {targetBone {boneName=clavicle_l}}
     {targetBone {boneName=spine_02}}
     {targetBone {boneName=spine_03}}
     {targetBone {boneName=spine_04}}
     {targetBone {boneName=spine_05}}
     {targetBone {boneName=clavicle_r}}
     {targetBone {boneName=pelvis}}
    ]
    constraintWeights [1,1,1,1,1,1,1]
   }
   n83 = "Constraint - curve_1_r_CV_3 (PPPPPPP)"(
    ComponentPose=@n82
    Alpha=@n74
    ConstraintWeights_0=0.0183
    ConstraintWeights_1=0.0134
    ConstraintWeights_2=0.1993
    ConstraintWeights_3=0.0561
    ConstraintWeights_4=0.1081
    ConstraintWeights_5=0.5372
    ConstraintWeights_6=0.0013
   ) {
    alphaBoolBlend {blendOption=Linear}
    boneToModify {boneName=curve_1_r_CV_3}
    constraintSetup [
     {targetBone {boneName=clavicle_l}}
     {targetBone {boneName=spine_02}}
     {targetBone {boneName=spine_03}}
     {targetBone {boneName=spine_04}}
     {targetBone {boneName=spine_05}}
     {targetBone {boneName=clavicle_r}}
     {targetBone {boneName=dyn_hairBase}}
    ]
    constraintWeights [1,1,1,1,1,1,1]
   }
   n84 = "Constraint - curve_1_r_CV_4 (PPPPPPPP)"(
    ComponentPose=@n83
    Alpha=@n74
    ConstraintWeights_0=0.1032
    ConstraintWeights_1=0.0171
    ConstraintWeights_2=0.0558
    ConstraintWeights_3=0.4378
    ConstraintWeights_4=0.1502
    ConstraintWeights_5=0.0185
    ConstraintWeights_6=0.2139
    ConstraintWeights_7=0.0034
   ) {
    alphaBoolBlend {blendOption=Linear}
    boneToModify {boneName=curve_1_r_CV_4}
    constraintSetup [
     {targetBone {boneName=clavicle_l}}
     {targetBone {boneName=spine_01}}
     {targetBone {boneName=spine_02}}
     {targetBone {boneName=spine_03}}
     {targetBone {boneName=spine_04}}
     {targetBone {boneName=spine_05}}
     {targetBone {boneName=clavicle_r}}
     {targetBone {boneName=pelvis}}
    ]
    constraintWeights [1,1,1,1,1,1,1,1]
   }
   n85 = Get isNotRoaringAlpha()
   n86 = "Transform (Modify) Bone - Bone: curve_c_CV_3"!(ComponentPose=@n84, Alpha=@n85, Translation:Vector="-0.000,-5,0.000", Rotation="0,0,0", Scale:Vector="1,1,1") -> (Pose) {translationMode=BMM_Additive, alphaBoolBlend {blendOption=Linear}, boneToModify {boneName=curve_c_CV_3}, translation {y=-5}}
   n87 = "Transform (Modify) Bone - Bone: curve_c_CV_4"(ComponentPose=@n86, Alpha=@n85, Translation="0.000,10,-20", Rotation="0,0,0", Scale="1,1,1") {translationMode=BMM_Additive, alphaBoolBlend {blendOption=Linear}, boneToModify {boneName=curve_c_CV_4}, translation {y=10, z=-20}}
   n88 = "Transform (Modify) Bone - Bone: curve_2_r_CV_2"!(ComponentPose=@n87, Alpha=@n85, Translation:Vector="0.000,-1,-0.000", Rotation="0,0,0", Scale:Vector="1,1,1") -> (Pose) {
    translationMode=BMM_Additive
    translationSpace=BCS_BoneSpace
    alphaBoolBlend {blendOption=Linear}
    boneToModify {boneName=curve_2_r_CV_2}
    translation {y=-1}
   }
   n89 = "Transform (Modify) Bone - Bone: curve_1_r_CV_2"!(ComponentPose=@n88, Alpha=@n85, Translation:Vector="-0.000,-5,-0.000", Rotation="0,0,0", Scale:Vector="1,1,1") -> (Pose) {
    translationMode=BMM_Additive
    translationSpace=BCS_BoneSpace
    alphaBoolBlend {blendOption=Linear}
    boneToModify {boneName=curve_1_r_CV_2}
    translation {y=-5}
   }
   n90 = "Transform (Modify) Bone - Bone: curve_1_l_CV_2"!(ComponentPose=@n89, Alpha=@n85, Translation:Vector="-0.000,-5,-0.000", Rotation="0,0,0", Scale:Vector="1,1,1") -> (Pose) {
    translationMode=BMM_Additive
    translationSpace=BCS_BoneSpace
    alphaBoolBlend {blendOption=Linear}
    boneToModify {boneName=curve_1_l_CV_2}
    translation {y=-5}
   }
   n91 = "Transform (Modify) Bone - Bone: curve_1_l_CV_4"(ComponentPose=@n90, Alpha=@n85, Translation="0.000,0,-5", Rotation="0,0,0", Scale="1,1,1") {translationMode=BMM_Additive, alphaBoolBlend {blendOption=Linear}, boneToModify {boneName=curve_1_l_CV_4}, translation {z=-5}}
   n92 = "Transform (Modify) Bone - Bone: curve_1_r_CV_4"(ComponentPose=@n91, Alpha=@n85, Translation="0.000,0.000,-5", Rotation="0,0,0", Scale="1,1,1") {translationMode=BMM_Additive, alphaBoolBlend {blendOption=Linear}, boneToModify {boneName=curve_1_r_CV_4}, translation {z=-5}}
   n93 = "Transform (Modify) Bone - Bone: curve_2_l_CV_3"!(ComponentPose=@n92, Alpha=1, Translation:Vector="0.000,-5,0.000", Rotation="0,0,0", Scale:Vector="1,1,1") -> (Pose) {translationMode=BMM_Additive, alphaBoolBlend {blendOption=Linear}, boneToModify {boneName=curve_2_l_CV_3}, translation {y=-5}}
   n94 = "Transform (Modify) Bone - Bone: curve_2_r_CV_3"!(ComponentPose=@n93, Alpha=1, Translation:Vector="0.000,-5,0", Rotation="0,0,0", Scale:Vector="1,1,1") -> (Pose) {translationMode=BMM_Additive, alphaBoolBlend {blendOption=Linear}, boneToModify {boneName=curve_2_r_CV_3}, translation {y=-5}}
   n95 = "Transform (Modify) Bone - Bone: curve_2_l_CV_4"(ComponentPose=@n94, Alpha=@n85, Translation="0.000,5,-15", Rotation="0,0,0", Scale="1,1,1") {translationMode=BMM_Additive, alphaBoolBlend {blendOption=Linear}, boneToModify {boneName=curve_2_l_CV_4}, translation {y=5, z=-15}}
   n96 = "Transform (Modify) Bone - Bone: curve_2_r_CV_4"(ComponentPose=@n95, Alpha=@n85, Translation="0.000,5,-15", Rotation="0,0,0", Scale="1,1,1") {translationMode=BMM_Additive, alphaBoolBlend {blendOption=Linear}, boneToModify {boneName=curve_2_r_CV_4}, translation {y=5, z=-15}}
   n97 = Get isRoaringAlpha()
   n98 = "Transform (Modify) Bone - Bone: curve_1_l_CV_1"(ComponentPose=@n96, Alpha=@n97, Rotation="5,-40,0") {
    rotationMode=BMM_Additive
    rotationSpace=BCS_BoneSpace
    alphaBoolBlend {blendOption=Linear}
    alphaScaleBiasClamp {bInterpResult=true, interpSpeedDecreasing=5, interpSpeedIncreasing=5}
    boneToModify {boneName=curve_1_l_CV_1}
    rotation {pitch=5, yaw=-40}
   }
   n99 = "Transform (Modify) Bone - Bone: curve_2_l_CV_1"(ComponentPose=@n98, Alpha=@n97, Rotation="2,-10,0") {
    rotationMode=BMM_Additive
    rotationSpace=BCS_BoneSpace
    alphaBoolBlend {blendOption=Linear}
    alphaScaleBiasClamp {bInterpResult=true, interpSpeedDecreasing=5, interpSpeedIncreasing=5}
    boneToModify {boneName=curve_2_l_CV_1}
    rotation {pitch=2, yaw=-10}
   }
   n100 = "Transform (Modify) Bone - Bone: curve_c_CV_1"(ComponentPose=@n99, Alpha=@n97, Rotation="0,-10,0") {
    rotationMode=BMM_Additive
    rotationSpace=BCS_BoneSpace
    alphaBoolBlend {blendOption=Linear}
    alphaScaleBiasClamp {bInterpResult=true, interpSpeedDecreasing=5, interpSpeedIncreasing=5}
    boneToModify {boneName=curve_c_CV_1}
    rotation {yaw=-10}
   }
   n101 = "Transform (Modify) Bone - Bone: curve_2_r_CV_1"(ComponentPose=@n100, Alpha=@n97, Rotation="-2,-10,0") {
    rotationMode=BMM_Additive
    rotationSpace=BCS_BoneSpace
    alphaBoolBlend {blendOption=Linear}
    alphaScaleBiasClamp {bInterpResult=true, interpSpeedDecreasing=5, interpSpeedIncreasing=5}
    boneToModify {boneName=curve_2_r_CV_1}
    rotation {pitch=-2, yaw=-10}
   }
   n102 = "Transform (Modify) Bone - Bone: curve_1_r_CV_1"(ComponentPose=@n101, Alpha=@n97, Rotation="-5,-40,0") {
    rotationMode=BMM_Additive
    rotationSpace=BCS_BoneSpace
    alphaBoolBlend {blendOption=Linear}
    alphaScaleBiasClamp {bInterpResult=true, interpSpeedDecreasing=5, interpSpeedIncreasing=5}
    boneToModify {boneName=curve_1_r_CV_1}
    rotation {pitch=-5, yaw=-40}
   }
   n103 = "Transform (Modify) Bone - Bone: curve_1_l_CV_2"!(ComponentPose=@n102, Alpha=@n97, Rotation="5,-60,0") -> (Pose) {
    rotationMode=BMM_Additive
    rotationSpace=BCS_BoneSpace
    alphaBoolBlend {blendOption=Linear}
    alphaScaleBiasClamp {bInterpResult=true, interpSpeedDecreasing=5, interpSpeedIncreasing=5}
    boneToModify {boneName=curve_1_l_CV_2}
    rotation {pitch=5, yaw=-60}
   }
   n104 = "Transform (Modify) Bone - Bone: curve_2_l_CV_2"(ComponentPose=@n103, Alpha=@n97, Rotation="2,-60,0") {
    rotationMode=BMM_Additive
    rotationSpace=BCS_BoneSpace
    alphaBoolBlend {blendOption=Linear}
    alphaScaleBiasClamp {bInterpResult=true, interpSpeedDecreasing=5, interpSpeedIncreasing=5}
    boneToModify {boneName=curve_2_l_CV_2}
    rotation {pitch=2, yaw=-60}
   }
   n105 = "Transform (Modify) Bone - Bone: curve_c_CV_2"(ComponentPose=@n104, Alpha=@n97, Rotation="0,-90,0") {
    rotationMode=BMM_Additive
    rotationSpace=BCS_BoneSpace
    alphaBoolBlend {blendOption=Linear}
    alphaScaleBiasClamp {bInterpResult=true, interpSpeedDecreasing=5, interpSpeedIncreasing=5}
    boneToModify {boneName=curve_c_CV_2}
    rotation {yaw=-90}
   }
   n106 = "Transform (Modify) Bone - Bone: curve_2_r_CV_2"!(ComponentPose=@n105, Alpha=@n97, Rotation="-2,-60,0") -> (Pose) {
    rotationMode=BMM_Additive
    rotationSpace=BCS_BoneSpace
    alphaBoolBlend {blendOption=Linear}
    alphaScaleBiasClamp {bInterpResult=true, interpSpeedDecreasing=5, interpSpeedIncreasing=5}
    boneToModify {boneName=curve_2_r_CV_2}
    rotation {pitch=-2, yaw=-60}
   }
   n107 = "Transform (Modify) Bone - Bone: curve_1_r_CV_2"!(ComponentPose=@n106, Alpha=@n97, Rotation="-5,-60,0") -> (Pose) {
    rotationMode=BMM_Additive
    rotationSpace=BCS_BoneSpace
    alphaBoolBlend {blendOption=Linear}
    alphaScaleBiasClamp {bInterpResult=true, interpSpeedDecreasing=5, interpSpeedIncreasing=5}
    boneToModify {boneName=curve_1_r_CV_2}
    rotation {pitch=-5, yaw=-60}
   }
   n108 = "Transform (Modify) Bone - Bone: curve_1_l_CV_3"(ComponentPose=@n107, Alpha=@n97, Rotation="5,-90,0") {
    rotationMode=BMM_Additive
    rotationSpace=BCS_BoneSpace
    alphaBoolBlend {blendOption=Linear}
    alphaScaleBiasClamp {bInterpResult=true, interpSpeedDecreasing=5, interpSpeedIncreasing=5}
    boneToModify {boneName=curve_1_l_CV_3}
    rotation {pitch=5, yaw=-90}
   }
   n109 = "Transform (Modify) Bone - Bone: curve_2_l_CV_3"!(ComponentPose=@n108, Alpha=@n97, Rotation="2,-90,0") -> (Pose) {
    rotationMode=BMM_Additive
    rotationSpace=BCS_BoneSpace
    alphaBoolBlend {blendOption=Linear}
    alphaScaleBiasClamp {bInterpResult=true, interpSpeedDecreasing=5, interpSpeedIncreasing=5}
    boneToModify {boneName=curve_2_l_CV_3}
    rotation {pitch=2, yaw=-90}
   }
   n110 = "Transform (Modify) Bone - Bone: curve_c_CV_3"!(ComponentPose=@n109, Alpha=@n97, Rotation="0,-60,0") -> (Pose) {
    rotationMode=BMM_Additive
    rotationSpace=BCS_BoneSpace
    alphaBoolBlend {blendOption=Linear}
    alphaScaleBiasClamp {bInterpResult=true, interpSpeedDecreasing=5, interpSpeedIncreasing=5}
    boneToModify {boneName=curve_c_CV_3}
    rotation {yaw=-60}
   }
   n111 = "Transform (Modify) Bone - Bone: curve_2_r_CV_3"!(ComponentPose=@n110, Alpha=@n97, Rotation="-2,-90,0") -> (Pose) {
    rotationMode=BMM_Additive
    rotationSpace=BCS_BoneSpace
    alphaBoolBlend {blendOption=Linear}
    alphaScaleBiasClamp {bInterpResult=true, interpSpeedDecreasing=5, interpSpeedIncreasing=5}
    boneToModify {boneName=curve_2_r_CV_3}
    rotation {pitch=-2, yaw=-90}
   }
   n112 = "Transform (Modify) Bone - Bone: curve_1_r_CV_3"(ComponentPose=@n111, Alpha=@n97, Rotation="-5,-90,0") {
    rotationMode=BMM_Additive
    rotationSpace=BCS_BoneSpace
    alphaBoolBlend {blendOption=Linear}
    alphaScaleBiasClamp {bInterpResult=true, interpSpeedDecreasing=5, interpSpeedIncreasing=5}
    boneToModify {boneName=curve_1_r_CV_3}
    rotation {pitch=-5, yaw=-90}
   }
   n113 = Control Rig(Source=Component To Local(ComponentPose=@n112)) {
    alphaBoolBlend {blendOption=Linear}
    controlRigAssetReference {blueprintRigClass=&M_LRG_Cattus_FaceAcc_CtrlRig_Curves_C}
   }
   n114 = Control Rig!(Source=@n113, Alpha, isRoaringAlpha:double=Get isRoaringAlpha(), useHarmonics:bool=Get bAnimDynamicsEnabled() -> (bAnimDynamicsEnabled:bool)) -> (Pose) {
    alphaBoolBlend {blendOption=Linear}
    controlRigAssetReference {blueprintRigClass=&/M_LRG_Cattus_FaceAcc_CtrlRig.M_LRG_Cattus_FaceAcc_CtrlRig_C}
   }
   n115 = Anim Dynamics - Bone: tent_1_l_2(ComponentPose=Local To Component(LocalPose=@n114), Alpha=1, ExternalForce="0,0,0") {
    angularDampingOverride=1
    angularSpringConstant=1500
    bAngularSpring=true
    bGravityOverrideInSimSpace=true
    bOverrideAngularDamping=true
    bUseSphericalLimits=true
    gravityScale=0.35
    simulationSpace=RootRelative
    alphaBoolBlend {blendOption=Linear}
    alphaScaleBiasClamp {scale=0.3}
    boundBone {boneName=tent_1_l_2}
    componentLinearAccScale {z=0.43}
    componentLinearVelScale {z=0.43}
    physicsBodyDefinitions [{
      collisionType=CustomSphere
      sphereCollisionRadius=2
      boundBone {boneName=tent_1_l_2}
      boxExtents=<10 20 20>
      constraintSetup {angularLimitsMax=<10 5 60>, angularLimitsMin=<-10 -5 -60>, angularTarget=<1 0 0>}
      localJointOffset=<10 0 0>
     }]
    sphericalLimits [
     {limitRadius=34, drivingBone {boneName=spine_05}, sphereLocalOffset=<-1.6787 12.8914 0>}
     {limitRadius=16, drivingBone {boneName=clavicle_l}, sphereLocalOffset=<15.524 -1.769 -0.679>}
     {limitRadius=16, drivingBone {boneName=clavicle_r}, sphereLocalOffset=<-15.524 -1.769 0.679>}
    ]
   }
   n116 = Anim Dynamics - Bone: tent_1_l_3(ComponentPose=@n115, Alpha=1, ExternalForce="0,0,0") {
    angularDampingOverride=1
    angularSpringConstant=1500
    bAngularSpring=true
    bGravityOverrideInSimSpace=true
    bOverrideAngularDamping=true
    bUseSphericalLimits=true
    gravityScale=0.35
    simulationSpace=RootRelative
    alphaBoolBlend {blendOption=Linear}
    alphaScaleBiasClamp {scale=0.6}
    boundBone {boneName=tent_1_l_3}
    componentLinearAccScale {z=0.44}
    componentLinearVelScale {z=0.44}
    physicsBodyDefinitions [{
      collisionType=CustomSphere
      sphereCollisionRadius=2
      boundBone {boneName=tent_1_l_3}
      boxExtents=<10 20 20>
      constraintSetup {angularLimitsMax=<10 5 60>, angularLimitsMin=<-10 -5 -60>, angularTarget=<1 0 0>}
      localJointOffset=<10 0 0>
     }]
    sphericalLimits [
     {limitRadius=34, drivingBone {boneName=spine_05}, sphereLocalOffset=<-1.6787 12.8914 0>}
     {limitRadius=16, drivingBone {boneName=clavicle_l}, sphereLocalOffset=<15.524 -1.769 -0.679>}
     {limitRadius=16, drivingBone {boneName=clavicle_r}, sphereLocalOffset=<-15.524 -1.769 0.679>}
    ]
   }
   n117 = Get isRoaringDynamicsAlpha() -> (isRoaringDynamicsAlpha:double)
   n118 = Anim Dynamics - Bone: tent_1_l_4(ComponentPose=@n116, Alpha=@n117, ExternalForce="0,0,0") {
    angularDampingOverride=1
    angularSpringConstant=1500
    bAngularSpring=true
    bGravityOverrideInSimSpace=true
    bOverrideAngularDamping=true
    bUseSphericalLimits=true
    gravityScale=0.35
    simulationSpace=RootRelative
    alphaBoolBlend {blendOption=Linear}
    alphaScaleBiasClamp {bInterpResult=true}
    boundBone {boneName=tent_1_l_4}
    componentLinearAccScale {z=0.44}
    componentLinearVelScale {z=0.44}
    physicsBodyDefinitions [{
      collisionType=CustomSphere
      sphereCollisionRadius=2
      boundBone {boneName=tent_1_l_4}
      boxExtents=<10 20 20>
      constraintSetup {angularLimitsMax=<10 5 60>, angularLimitsMin=<-10 -5 -60>, angularTarget=<0 0 0>}
      localJointOffset=<10 0 0>
     }]
    sphericalLimits [
     {limitRadius=34, drivingBone {boneName=spine_05}, sphereLocalOffset=<-1.6787 12.8914 0>}
     {limitRadius=16, drivingBone {boneName=clavicle_l}, sphereLocalOffset=<15.524 -1.769 -0.679>}
     {limitRadius=16, drivingBone {boneName=clavicle_r}, sphereLocalOffset=<-15.524 -1.769 0.679>}
    ]
   }
   n119 = Anim Dynamics - Bone: tent_1_l_5(ComponentPose=@n118, Alpha=@n117, ExternalForce="0,0,0") {
    angularDampingOverride=1
    angularSpringConstant=1500
    bAngularSpring=true
    bGravityOverrideInSimSpace=true
    bOverrideAngularDamping=true
    bUseSphericalLimits=true
    gravityScale=0.35
    simulationSpace=RootRelative
    alphaBoolBlend {blendOption=Linear}
    alphaScaleBiasClamp {bInterpResult=true}
    boundBone {boneName=tent_1_l_5}
    componentLinearAccScale {z=0.45}
    componentLinearVelScale {z=0.45}
    physicsBodyDefinitions [{
      collisionType=CustomSphere
      sphereCollisionRadius=2
      boundBone {boneName=tent_1_l_5}
      boxExtents=<10 20 20>
      constraintSetup {angularLimitsMax=<10 5 60>, angularLimitsMin=<-10 -5 -60>, angularTarget=<0 0 0>}
      localJointOffset=<10 0 0>
     }]
    sphericalLimits [
     {limitRadius=34, drivingBone {boneName=spine_05}, sphereLocalOffset=<-1.6787 12.8914 0>}
     {limitRadius=16, drivingBone {boneName=clavicle_l}, sphereLocalOffset=<15.524 -1.769 -0.679>}
     {limitRadius=16, drivingBone {boneName=clavicle_r}, sphereLocalOffset=<-15.524 -1.769 0.679>}
    ]
   }
   n120 = Anim Dynamics - Bone: tent_1_l_6(ComponentPose=@n119, Alpha=@n117, ExternalForce="0,0,0") {
    angularDampingOverride=1
    angularSpringConstant=1500
    bAngularSpring=true
    bGravityOverrideInSimSpace=true
    bOverrideAngularDamping=true
    bUseSphericalLimits=true
    gravityScale=0.35
    simulationSpace=RootRelative
    alphaBoolBlend {blendOption=Linear}
    alphaScaleBiasClamp {bInterpResult=true}
    boundBone {boneName=tent_1_l_6}
    componentLinearAccScale {z=0.45}
    componentLinearVelScale {z=0.45}
    physicsBodyDefinitions [{
      collisionType=CustomSphere
      sphereCollisionRadius=2
      boundBone {boneName=tent_1_l_6}
      boxExtents=<10 20 20>
      constraintSetup {angularLimitsMax=<10 5 60>, angularLimitsMin=<-10 -5 -60>, angularTarget=<1 0 0>}
      localJointOffset=<10 0 0>
     }]
    sphericalLimits [
     {limitRadius=34, drivingBone {boneName=spine_05}, sphereLocalOffset=<-1.6787 12.8914 0>}
     {limitRadius=16, drivingBone {boneName=clavicle_l}, sphereLocalOffset=<15.524 -1.769 -0.679>}
     {limitRadius=16, drivingBone {boneName=clavicle_r}, sphereLocalOffset=<-15.524 -1.769 0.679>}
    ]
   }
   n121 = Anim Dynamics - Bone: tent_1_l_7(ComponentPose=@n120, Alpha=@n117, ExternalForce="0,0,0") {
    angularDampingOverride=1
    angularSpringConstant=1500
    bAngularSpring=true
    bGravityOverrideInSimSpace=true
    bOverrideAngularDamping=true
    bUseSphericalLimits=true
    gravityScale=0.35
    simulationSpace=RootRelative
    alphaBoolBlend {blendOption=Linear}
    alphaScaleBiasClamp {bInterpResult=true}
    boundBone {boneName=tent_1_l_7}
    componentLinearAccScale {z=0.46}
    componentLinearVelScale {z=0.46}
    physicsBodyDefinitions [{
      collisionType=CustomSphere
      sphereCollisionRadius=2
      boundBone {boneName=tent_1_l_7}
      boxExtents=<10 20 20>
      constraintSetup {angularLimitsMax=<10 5 60>, angularLimitsMin=<-10 -5 -60>, angularTarget=<1 0 0>}
      localJointOffset=<10 0 0>
     }]
    sphericalLimits [
     {limitRadius=35, drivingBone {boneName=spine_05}, sphereLocalOffset=<-1.6787 12.8914 0>}
     {limitRadius=16, drivingBone {boneName=clavicle_l}, sphereLocalOffset=<15.524 -1.769 -0.679>}
     {limitRadius=16, drivingBone {boneName=clavicle_r}, sphereLocalOffset=<-15.524 -1.769 0.679>}
    ]
   }
   n122 = Anim Dynamics - Bone: tent_1_l_8(ComponentPose=@n121, Alpha=@n117, ExternalForce="0,0,0") {
    angularDampingOverride=1
    angularSpringConstant=1500
    bAngularSpring=true
    bGravityOverrideInSimSpace=true
    bOverrideAngularDamping=true
    bUseSphericalLimits=true
    gravityScale=0.35
    simulationSpace=RootRelative
    alphaBoolBlend {blendOption=Linear}
    alphaScaleBiasClamp {bInterpResult=true}
    boundBone {boneName=tent_1_l_8}
    componentLinearAccScale {z=0.46}
    componentLinearVelScale {z=0.46}
    physicsBodyDefinitions [{
      collisionType=CustomSphere
      sphereCollisionRadius=2
      boundBone {boneName=tent_1_l_8}
      boxExtents=<10 20 20>
      constraintSetup {angularLimitsMax=<10 5 60>, angularLimitsMin=<-10 -5 -60>, angularTarget=<0 0 0>}
      localJointOffset=<10 0 0>
     }]
    sphericalLimits [
     {limitRadius=35, drivingBone {boneName=spine_05}, sphereLocalOffset=<-1.6787 12.8914 0>}
     {limitRadius=16, drivingBone {boneName=clavicle_l}, sphereLocalOffset=<15.524 -1.769 -0.679>}
     {limitRadius=16, drivingBone {boneName=clavicle_r}, sphereLocalOffset=<-15.524 -1.769 0.679>}
    ]
   }
   n123 = Anim Dynamics - Bone: tent_1_l_9(ComponentPose=@n122, Alpha=@n117, ExternalForce="0,0,0") {
    angularDampingOverride=1
    angularSpringConstant=1500
    bAngularSpring=true
    bGravityOverrideInSimSpace=true
    bOverrideAngularDamping=true
    bUseSphericalLimits=true
    gravityScale=0.35
    simulationSpace=RootRelative
    alphaBoolBlend {blendOption=Linear}
    alphaScaleBiasClamp {bInterpResult=true}
    boundBone {boneName=tent_1_l_9}
    componentLinearAccScale {z=0.46}
    componentLinearVelScale {z=0.46}
    physicsBodyDefinitions [{
      collisionType=CustomSphere
      sphereCollisionRadius=2
      boundBone {boneName=tent_1_l_9}
      boxExtents=<10 20 20>
      constraintSetup {angularLimitsMax=<10 5 60>, angularLimitsMin=<-10 -5 -60>, angularTarget=<0 0 0>}
      localJointOffset=<10 0 0>
     }]
    sphericalLimits [
     {limitRadius=35, drivingBone {boneName=spine_05}, sphereLocalOffset=<-1.6787 12.8914 0>}
     {limitRadius=16, drivingBone {boneName=clavicle_l}, sphereLocalOffset=<15.524 -1.769 -0.679>}
     {limitRadius=16, drivingBone {boneName=clavicle_r}, sphereLocalOffset=<-15.524 -1.769 0.679>}
    ]
   }
   n124 = Anim Dynamics - Bone: tent_2_l_2(ComponentPose=@n123, Alpha=1, ExternalForce="0,0,0") {
    angularDampingOverride=1
    angularSpringConstant=1500
    bAngularSpring=true
    bGravityOverrideInSimSpace=true
    bOverrideAngularDamping=true
    bUseSphericalLimits=true
    gravityScale=0.35
    simulationSpace=RootRelative
    alphaBoolBlend {blendOption=Linear}
    alphaScaleBiasClamp {scale=0.3}
    boundBone {boneName=tent_2_l_2}
    componentLinearAccScale {z=0.43}
    componentLinearVelScale {z=0.43}
    physicsBodyDefinitions [{
      collisionType=CustomSphere
      sphereCollisionRadius=2
      boundBone {boneName=tent_2_l_2}
      boxExtents=<11 20 20>
      constraintSetup {angularLimitsMax=<0 5 60>, angularLimitsMin=<0 -5 -60>, angularTarget=<1 0 0>}
      localJointOffset=<10 0 0>
     }]
    sphericalLimits [
     {limitRadius=34, drivingBone {boneName=spine_05}, sphereLocalOffset=<-1.6787 12.8914 0>}
     {limitRadius=16, drivingBone {boneName=clavicle_l}, sphereLocalOffset=<15.524 -1.769 -0.679>}
     {limitRadius=16, drivingBone {boneName=clavicle_r}, sphereLocalOffset=<-15.524 -1.769 0.679>}
    ]
   }
   n125 = Anim Dynamics - Bone: tent_2_l_3(ComponentPose=@n124, Alpha=1, ExternalForce="0,0,0") {
    angularDampingOverride=1
    angularSpringConstant=1500
    bAngularSpring=true
    bGravityOverrideInSimSpace=true
    bOverrideAngularDamping=true
    bUseSphericalLimits=true
    gravityScale=0.35
    simulationSpace=RootRelative
    alphaBoolBlend {blendOption=Linear}
    alphaScaleBiasClamp {scale=0.6}
    boundBone {boneName=tent_2_l_3}
    componentLinearAccScale {z=0.44}
    componentLinearVelScale {z=0.44}
    physicsBodyDefinitions [{
      collisionType=CustomSphere
      sphereCollisionRadius=2
      boundBone {boneName=tent_2_l_3}
      boxExtents=<11 20 20>
      constraintSetup {angularLimitsMax=<0 5 60>, angularLimitsMin=<0 -5 -60>, angularTarget=<1 0 0>}
      localJointOffset=<10 0 0>
     }]
    sphericalLimits [
     {limitRadius=34, drivingBone {boneName=spine_05}, sphereLocalOffset=<-1.6787 12.8914 0>}
     {limitRadius=16, drivingBone {boneName=clavicle_l}, sphereLocalOffset=<15.524 -1.769 -0.679>}
     {limitRadius=16, drivingBone {boneName=clavicle_r}, sphereLocalOffset=<-15.524 -1.769 0.679>}
    ]
   }
   n126 = Anim Dynamics - Bone: tent_2_l_4(ComponentPose=@n125, Alpha=@n117, ExternalForce="0,0,0") {
    angularDampingOverride=1
    angularSpringConstant=1500
    bAngularSpring=true
    bGravityOverrideInSimSpace=true
    bOverrideAngularDamping=true
    bUseSphericalLimits=true
    gravityScale=0.35
    simulationSpace=RootRelative
    alphaBoolBlend {blendOption=Linear}
    alphaScaleBiasClamp {bInterpResult=true}
    boundBone {boneName=tent_2_l_4}
    componentLinearAccScale {z=0.44}
    componentLinearVelScale {z=0.44}
    physicsBodyDefinitions [{
      collisionType=CustomSphere
      sphereCollisionRadius=2
      boundBone {boneName=tent_2_l_4}
      boxExtents=<11 20 20>
      constraintSetup {angularLimitsMax=<0 5 60>, angularLimitsMin=<0 -5 -60>, angularTarget=<0 0 0>}
      localJointOffset=<10 0 0>
     }]
    sphericalLimits [
     {limitRadius=34, drivingBone {boneName=spine_05}, sphereLocalOffset=<-1.6787 12.8914 0>}
     {limitRadius=16, drivingBone {boneName=clavicle_l}, sphereLocalOffset=<15.524 -1.769 -0.679>}
     {limitRadius=16, drivingBone {boneName=clavicle_r}, sphereLocalOffset=<-15.524 -1.769 0.679>}
    ]
   }
   n127 = Anim Dynamics - Bone: tent_2_l_5(ComponentPose=@n126, Alpha=@n117, ExternalForce="0,0,0") {
    angularDampingOverride=1
    angularSpringConstant=1500
    bAngularSpring=true
    bGravityOverrideInSimSpace=true
    bOverrideAngularDamping=true
    bUseSphericalLimits=true
    gravityScale=0.35
    simulationSpace=RootRelative
    alphaBoolBlend {blendOption=Linear}
    alphaScaleBiasClamp {bInterpResult=true}
    boundBone {boneName=tent_2_l_5}
    componentLinearAccScale {z=0.45}
    componentLinearVelScale {z=0.45}
    physicsBodyDefinitions [{
      collisionType=CustomSphere
      sphereCollisionRadius=2
      boundBone {boneName=tent_2_l_5}
      boxExtents=<11 20 20>
      constraintSetup {angularLimitsMax=<0 5 60>, angularLimitsMin=<0 -5 -60>, angularTarget=<0 0 0>}
      localJointOffset=<10 0 0>
     }]
    sphericalLimits [
     {limitRadius=34, drivingBone {boneName=spine_05}, sphereLocalOffset=<-1.6787 12.8914 0>}
     {limitRadius=16, drivingBone {boneName=clavicle_l}, sphereLocalOffset=<15.524 -1.769 -0.679>}
     {limitRadius=16, drivingBone {boneName=clavicle_r}, sphereLocalOffset=<-15.524 -1.769 0.679>}
    ]
   }
   n128 = Anim Dynamics - Bone: tent_2_l_6(ComponentPose=@n127, Alpha=@n117, ExternalForce="0,0,0") {
    angularDampingOverride=1
    angularSpringConstant=1500
    bAngularSpring=true
    bGravityOverrideInSimSpace=true
    bOverrideAngularDamping=true
    bUseSphericalLimits=true
    gravityScale=0.35
    simulationSpace=RootRelative
    alphaBoolBlend {blendOption=Linear}
    alphaScaleBiasClamp {bInterpResult=true}
    boundBone {boneName=tent_2_l_6}
    componentLinearAccScale {z=0.45}
    componentLinearVelScale {z=0.45}
    physicsBodyDefinitions [{
      collisionType=CustomSphere
      sphereCollisionRadius=2
      boundBone {boneName=tent_2_l_6}
      boxExtents=<11 20 20>
      constraintSetup {angularLimitsMax=<0 5 60>, angularLimitsMin=<0 -5 -60>, angularTarget=<1 0 0>}
      localJointOffset=<10 0 0>
     }]
    sphericalLimits [
     {limitRadius=34, drivingBone {boneName=spine_05}, sphereLocalOffset=<-1.6787 12.8914 0>}
     {limitRadius=16, drivingBone {boneName=clavicle_l}, sphereLocalOffset=<15.524 -1.769 -0.679>}
     {limitRadius=16, drivingBone {boneName=clavicle_r}, sphereLocalOffset=<-15.524 -1.769 0.679>}
    ]
   }
   n129 = Anim Dynamics - Bone: tent_2_l_7(ComponentPose=@n128, Alpha=@n117, ExternalForce="0,0,0") {
    angularDampingOverride=1
    angularSpringConstant=1500
    bAngularSpring=true
    bGravityOverrideInSimSpace=true
    bOverrideAngularDamping=true
    bUseSphericalLimits=true
    gravityScale=0.35
    simulationSpace=RootRelative
    alphaBoolBlend {blendOption=Linear}
    alphaScaleBiasClamp {bInterpResult=true}
    boundBone {boneName=tent_2_l_7}
    componentLinearAccScale {z=0.46}
    componentLinearVelScale {z=0.46}
    physicsBodyDefinitions [{
      collisionType=CustomSphere
      sphereCollisionRadius=2
      boundBone {boneName=tent_2_l_7}
      boxExtents=<11 20 20>
      constraintSetup {angularLimitsMax=<0 5 60>, angularLimitsMin=<0 -5 -60>, angularTarget=<1 0 0>}
      localJointOffset=<10 0 0>
     }]
    sphericalLimits [
     {limitRadius=35, drivingBone {boneName=spine_05}, sphereLocalOffset=<-1.6787 12.8914 0>}
     {limitRadius=16, drivingBone {boneName=clavicle_l}, sphereLocalOffset=<15.524 -1.769 -0.679>}
     {limitRadius=16, drivingBone {boneName=clavicle_r}, sphereLocalOffset=<-15.524 -1.769 0.679>}
    ]
   }
   n130 = Anim Dynamics - Bone: tent_2_l_8(ComponentPose=@n129, Alpha=@n117, ExternalForce="0,0,0") {
    angularDampingOverride=1
    angularSpringConstant=1500
    bAngularSpring=true
    bGravityOverrideInSimSpace=true
    bOverrideAngularDamping=true
    bUseSphericalLimits=true
    gravityScale=0.35
    simulationSpace=RootRelative
    alphaBoolBlend {blendOption=Linear}
    alphaScaleBiasClamp {bInterpResult=true}
    boundBone {boneName=tent_2_l_8}
    componentLinearAccScale {z=0.46}
    componentLinearVelScale {z=0.46}
    physicsBodyDefinitions [{
      collisionType=CustomSphere
      sphereCollisionRadius=2
      boundBone {boneName=tent_2_l_8}
      boxExtents=<11 20 20>
      constraintSetup {angularLimitsMax=<0 5 60>, angularLimitsMin=<0 -5 -60>, angularTarget=<0 0 0>}
      localJointOffset=<10 0 0>
     }]
    sphericalLimits [
     {limitRadius=35, drivingBone {boneName=spine_05}, sphereLocalOffset=<-1.6787 12.8914 0>}
     {limitRadius=16, drivingBone {boneName=clavicle_l}, sphereLocalOffset=<15.524 -1.769 -0.679>}
     {limitRadius=16, drivingBone {boneName=clavicle_r}, sphereLocalOffset=<-15.524 -1.769 0.679>}
    ]
   }
   n131 = Anim Dynamics - Bone: tent_2_l_9(ComponentPose=@n130, Alpha=@n117, ExternalForce="0,0,0") {
    angularDampingOverride=1
    angularSpringConstant=1500
    bAngularSpring=true
    bGravityOverrideInSimSpace=true
    bOverrideAngularDamping=true
    bUseSphericalLimits=true
    gravityScale=0.35
    simulationSpace=RootRelative
    alphaBoolBlend {blendOption=Linear}
    alphaScaleBiasClamp {bInterpResult=true}
    boundBone {boneName=tent_2_l_9}
    componentLinearAccScale {z=0.46}
    componentLinearVelScale {z=0.46}
    physicsBodyDefinitions [{
      collisionType=CustomSphere
      sphereCollisionRadius=2
      boundBone {boneName=tent_2_l_9}
      boxExtents=<11 20 20>
      constraintSetup {angularLimitsMax=<0 5 60>, angularLimitsMin=<0 -5 -60>, angularTarget=<0 0 0>}
      localJointOffset=<10 0 0>
     }]
    sphericalLimits [
     {limitRadius=35, drivingBone {boneName=spine_05}, sphereLocalOffset=<-1.6787 12.8914 0>}
     {limitRadius=16, drivingBone {boneName=clavicle_l}, sphereLocalOffset=<15.524 -1.769 -0.679>}
     {limitRadius=16, drivingBone {boneName=clavicle_r}, sphereLocalOffset=<-15.524 -1.769 0.679>}
    ]
   }
   n132 = Anim Dynamics - Bone: tent_c_2(ComponentPose=@n131, Alpha=1, ExternalForce="0,0,0") {
    angularDampingOverride=1
    angularSpringConstant=1500
    bAngularSpring=true
    bGravityOverrideInSimSpace=true
    bOverrideAngularDamping=true
    bUseSphericalLimits=true
    gravityScale=0.35
    simulationSpace=RootRelative
    alphaBoolBlend {blendOption=Linear}
    alphaScaleBiasClamp {scale=0.3}
    boundBone {boneName=tent_c_2}
    componentLinearAccScale {z=0.43}
    componentLinearVelScale {z=0.43}
    physicsBodyDefinitions [{
      collisionType=CustomSphere
      sphereCollisionRadius=2
      boundBone {boneName=tent_c_2}
      boxExtents=<11 20 20>
      constraintSetup {angularLimitsMax=<0 10 60>, angularLimitsMin=<0 -10 -60>, angularTarget=<1 0 0>}
      localJointOffset=<10 0 0>
     }]
    sphericalLimits [
     {limitRadius=34, drivingBone {boneName=spine_05}, sphereLocalOffset=<-1.6787 12.8914 0>}
     {limitRadius=16, drivingBone {boneName=clavicle_l}, sphereLocalOffset=<15.524 -1.769 -0.679>}
     {limitRadius=16, drivingBone {boneName=clavicle_r}, sphereLocalOffset=<-15.524 -1.769 0.679>}
    ]
   }
   n133 = Anim Dynamics - Bone: tent_c_3(ComponentPose=@n132, Alpha=1, ExternalForce="0,0,0") {
    angularDampingOverride=1
    angularSpringConstant=1500
    bAngularSpring=true
    bGravityOverrideInSimSpace=true
    bOverrideAngularDamping=true
    bUseSphericalLimits=true
    gravityScale=0.35
    simulationSpace=RootRelative
    alphaBoolBlend {blendOption=Linear}
    alphaScaleBiasClamp {scale=0.6}
    boundBone {boneName=tent_c_3}
    componentLinearAccScale {z=0.44}
    componentLinearVelScale {z=0.44}
    physicsBodyDefinitions [{
      collisionType=CustomSphere
      sphereCollisionRadius=2
      boundBone {boneName=tent_c_3}
      boxExtents=<11 20 20>
      constraintSetup {angularLimitsMax=<0 10 60>, angularLimitsMin=<0 -10 -60>, angularTarget=<1 0 0>}
      localJointOffset=<10 0 0>
     }]
    sphericalLimits [
     {limitRadius=34, drivingBone {boneName=spine_05}, sphereLocalOffset=<-1.6787 12.8914 0>}
     {limitRadius=16, drivingBone {boneName=clavicle_l}, sphereLocalOffset=<15.524 -1.769 -0.679>}
     {limitRadius=16, drivingBone {boneName=clavicle_r}, sphereLocalOffset=<-15.524 -1.769 0.679>}
    ]
   }
   n134 = Anim Dynamics - Bone: tent_c_4(ComponentPose=@n133, Alpha=@n117, ExternalForce="0,0,0") {
    angularDampingOverride=1
    angularSpringConstant=1500
    bAngularSpring=true
    bGravityOverrideInSimSpace=true
    bOverrideAngularDamping=true
    bUseSphericalLimits=true
    gravityScale=0.35
    simulationSpace=RootRelative
    alphaBoolBlend {blendOption=Linear}
    alphaScaleBiasClamp {bInterpResult=true}
    boundBone {boneName=tent_c_4}
    componentLinearAccScale {z=0.44}
    componentLinearVelScale {z=0.44}
    physicsBodyDefinitions [{
      collisionType=CustomSphere
      sphereCollisionRadius=2
      boundBone {boneName=tent_c_4}
      boxExtents=<11 20 20>
      constraintSetup {angularLimitsMax=<0 10 60>, angularLimitsMin=<0 -10 -60>, angularTarget=<0 0 0>}
      localJointOffset=<10 0 0>
     }]
    sphericalLimits [
     {limitRadius=34, drivingBone {boneName=spine_05}, sphereLocalOffset=<-1.6787 12.8914 0>}
     {limitRadius=16, drivingBone {boneName=clavicle_l}, sphereLocalOffset=<15.524 -1.769 -0.679>}
     {limitRadius=16, drivingBone {boneName=clavicle_r}, sphereLocalOffset=<-15.524 -1.769 0.679>}
    ]
   }
   n135 = Anim Dynamics - Bone: tent_c_5(ComponentPose=@n134, Alpha=@n117, ExternalForce="0,0,0") {
    angularDampingOverride=1
    angularSpringConstant=1500
    bAngularSpring=true
    bGravityOverrideInSimSpace=true
    bOverrideAngularDamping=true
    bUseSphericalLimits=true
    gravityScale=0.35
    simulationSpace=RootRelative
    alphaBoolBlend {blendOption=Linear}
    alphaScaleBiasClamp {bInterpResult=true}
    boundBone {boneName=tent_c_5}
    componentLinearAccScale {z=0.45}
    componentLinearVelScale {z=0.45}
    physicsBodyDefinitions [{
      collisionType=CustomSphere
      sphereCollisionRadius=2
      boundBone {boneName=tent_c_5}
      boxExtents=<11 20 20>
      constraintSetup {angularLimitsMax=<0 10 60>, angularLimitsMin=<0 -10 -60>, angularTarget=<0 0 0>}
      localJointOffset=<10 0 0>
     }]
    sphericalLimits [
     {limitRadius=34, drivingBone {boneName=spine_05}, sphereLocalOffset=<-1.6787 12.8914 0>}
     {limitRadius=16, drivingBone {boneName=clavicle_l}, sphereLocalOffset=<15.524 -1.769 -0.679>}
     {limitRadius=16, drivingBone {boneName=clavicle_r}, sphereLocalOffset=<-15.524 -1.769 0.679>}
    ]
   }
   n136 = Anim Dynamics - Bone: tent_c_6(ComponentPose=@n135, Alpha=@n117, ExternalForce="0,0,0") {
    angularDampingOverride=1
    angularSpringConstant=1500
    bAngularSpring=true
    bGravityOverrideInSimSpace=true
    bOverrideAngularDamping=true
    bUseSphericalLimits=true
    gravityScale=0.35
    simulationSpace=RootRelative
    alphaBoolBlend {blendOption=Linear}
    alphaScaleBiasClamp {bInterpResult=true}
    boundBone {boneName=tent_c_6}
    componentLinearAccScale {z=0.45}
    componentLinearVelScale {z=0.45}
    physicsBodyDefinitions [{
      collisionType=CustomSphere
      sphereCollisionRadius=2
      boundBone {boneName=tent_c_6}
      boxExtents=<11 20 20>
      constraintSetup {angularLimitsMax=<0 10 60>, angularLimitsMin=<0 -10 -60>, angularTarget=<1 0 0>}
      localJointOffset=<10 0 0>
     }]
    sphericalLimits [
     {limitRadius=34, drivingBone {boneName=spine_05}, sphereLocalOffset=<-1.6787 12.8914 0>}
     {limitRadius=16, drivingBone {boneName=clavicle_l}, sphereLocalOffset=<15.524 -1.769 -0.679>}
     {limitRadius=16, drivingBone {boneName=clavicle_r}, sphereLocalOffset=<-15.524 -1.769 0.679>}
    ]
   }
   n137 = Anim Dynamics - Bone: tent_c_7(ComponentPose=@n136, Alpha=@n117, ExternalForce="0,0,0") {
    angularDampingOverride=1
    angularSpringConstant=1500
    bAngularSpring=true
    bGravityOverrideInSimSpace=true
    bOverrideAngularDamping=true
    bUseSphericalLimits=true
    gravityScale=0.35
    simulationSpace=RootRelative
    alphaBoolBlend {blendOption=Linear}
    alphaScaleBiasClamp {bInterpResult=true}
    boundBone {boneName=tent_c_7}
    componentLinearAccScale {z=0.46}
    componentLinearVelScale {z=0.46}
    physicsBodyDefinitions [{
      collisionType=CustomSphere
      sphereCollisionRadius=2
      boundBone {boneName=tent_c_7}
      boxExtents=<11 20 20>
      constraintSetup {angularLimitsMax=<0 10 60>, angularLimitsMin=<0 -10 -60>, angularTarget=<1 0 0>}
      localJointOffset=<10 0 0>
     }]
    sphericalLimits [
     {limitRadius=35, drivingBone {boneName=spine_05}, sphereLocalOffset=<-1.6787 12.8914 0>}
     {limitRadius=16, drivingBone {boneName=clavicle_l}, sphereLocalOffset=<15.524 -1.769 -0.679>}
     {limitRadius=16, drivingBone {boneName=clavicle_r}, sphereLocalOffset=<-15.524 -1.769 0.679>}
    ]
   }
   n138 = Anim Dynamics - Bone: tent_c_8(ComponentPose=@n137, Alpha=@n117, ExternalForce="0,0,0") {
    angularDampingOverride=1
    angularSpringConstant=1500
    bAngularSpring=true
    bGravityOverrideInSimSpace=true
    bOverrideAngularDamping=true
    bUseSphericalLimits=true
    gravityScale=0.35
    simulationSpace=RootRelative
    alphaBoolBlend {blendOption=Linear}
    alphaScaleBiasClamp {bInterpResult=true}
    boundBone {boneName=tent_c_8}
    componentLinearAccScale {z=0.46}
    componentLinearVelScale {z=0.46}
    physicsBodyDefinitions [{
      collisionType=CustomSphere
      sphereCollisionRadius=2
      boundBone {boneName=tent_c_8}
      boxExtents=<11 20 20>
      constraintSetup {angularLimitsMax=<0 10 60>, angularLimitsMin=<0 -10 -60>, angularTarget=<0 0 0>}
      localJointOffset=<10 0 0>
     }]
    sphericalLimits [
     {limitRadius=35, drivingBone {boneName=spine_05}, sphereLocalOffset=<-1.6787 12.8914 0>}
     {limitRadius=16, drivingBone {boneName=clavicle_l}, sphereLocalOffset=<15.524 -1.769 -0.679>}
     {limitRadius=16, drivingBone {boneName=clavicle_r}, sphereLocalOffset=<-15.524 -1.769 0.679>}
    ]
   }
   n139 = Anim Dynamics - Bone: tent_c_9(ComponentPose=@n138, Alpha=@n117, ExternalForce="0,0,0") {
    angularDampingOverride=1
    angularSpringConstant=1500
    bAngularSpring=true
    bGravityOverrideInSimSpace=true
    bOverrideAngularDamping=true
    bUseSphericalLimits=true
    gravityScale=0.35
    simulationSpace=RootRelative
    alphaBoolBlend {blendOption=Linear}
    alphaScaleBiasClamp {bInterpResult=true}
    boundBone {boneName=tent_c_9}
    componentLinearAccScale {z=0.46}
    componentLinearVelScale {z=0.46}
    physicsBodyDefinitions [{
      collisionType=CustomSphere
      sphereCollisionRadius=2
      boundBone {boneName=tent_c_9}
      boxExtents=<11 20 20>
      constraintSetup {angularLimitsMax=<0 10 60>, angularLimitsMin=<0 -10 -60>, angularTarget=<0 0 0>}
      localJointOffset=<10 0 0>
     }]
    sphericalLimits [
     {limitRadius=35, drivingBone {boneName=spine_05}, sphereLocalOffset=<-1.6787 12.8914 0>}
     {limitRadius=16, drivingBone {boneName=clavicle_l}, sphereLocalOffset=<15.524 -1.769 -0.679>}
     {limitRadius=16, drivingBone {boneName=clavicle_r}, sphereLocalOffset=<-15.524 -1.769 0.679>}
    ]
   }
   n140 = Anim Dynamics - Bone: tent_2_r_2(ComponentPose=@n139, Alpha=1, ExternalForce="0,0,0") {
    angularDampingOverride=1
    angularSpringConstant=1500
    bAngularSpring=true
    bGravityOverrideInSimSpace=true
    bOverrideAngularDamping=true
    bUseSphericalLimits=true
    gravityScale=0.35
    simulationSpace=RootRelative
    alphaBoolBlend {blendOption=Linear}
    alphaScaleBiasClamp {scale=0.3}
    boundBone {boneName=tent_2_r_2}
    componentLinearAccScale {z=0.43}
    componentLinearVelScale {z=0.43}
    physicsBodyDefinitions [{
      collisionType=CustomSphere
      sphereCollisionRadius=2
      boundBone {boneName=tent_2_r_2}
      boxExtents=<12 20 20>
      constraintSetup {angularLimitsMax=<0 10 60>, angularLimitsMin=<0 -10 -60>, angularTarget=<1 0 0>}
      localJointOffset=<10 0 0>
     }]
    sphericalLimits [
     {limitRadius=34, drivingBone {boneName=spine_05}, sphereLocalOffset=<-1.6787 12.8914 0>}
     {limitRadius=16, drivingBone {boneName=clavicle_l}, sphereLocalOffset=<15.524 -1.769 -0.679>}
     {limitRadius=16, drivingBone {boneName=clavicle_r}, sphereLocalOffset=<-15.524 -1.769 0.679>}
    ]
   }
   n141 = Anim Dynamics - Bone: tent_2_r_3(ComponentPose=@n140, Alpha=1, ExternalForce="0,0,0") {
    angularDampingOverride=1
    angularSpringConstant=1500
    bAngularSpring=true
    bGravityOverrideInSimSpace=true
    bOverrideAngularDamping=true
    bUseSphericalLimits=true
    gravityScale=0.35
    simulationSpace=RootRelative
    alphaBoolBlend {blendOption=Linear}
    alphaScaleBiasClamp {scale=0.6}
    boundBone {boneName=tent_2_r_3}
    componentLinearAccScale {z=0.44}
    componentLinearVelScale {z=0.44}
    physicsBodyDefinitions [{
      collisionType=CustomSphere
      sphereCollisionRadius=2
      boundBone {boneName=tent_2_r_3}
      boxExtents=<12 20 20>
      constraintSetup {angularLimitsMax=<0 10 60>, angularLimitsMin=<0 -10 -60>, angularTarget=<1 0 0>}
      localJointOffset=<10 0 0>
     }]
    sphericalLimits [
     {limitRadius=34, drivingBone {boneName=spine_05}, sphereLocalOffset=<-1.6787 12.8914 0>}
     {limitRadius=16, drivingBone {boneName=clavicle_l}, sphereLocalOffset=<15.524 -1.769 -0.679>}
     {limitRadius=16, drivingBone {boneName=clavicle_r}, sphereLocalOffset=<-15.524 -1.769 0.679>}
    ]
   }
   n142 = Anim Dynamics - Bone: tent_2_r_4(ComponentPose=@n141, Alpha=@n117, ExternalForce="0,0,0") {
    angularDampingOverride=1
    angularSpringConstant=1500
    bAngularSpring=true
    bGravityOverrideInSimSpace=true
    bOverrideAngularDamping=true
    bUseSphericalLimits=true
    gravityScale=0.35
    simulationSpace=RootRelative
    alphaBoolBlend {blendOption=Linear}
    alphaScaleBiasClamp {bInterpResult=true}
    boundBone {boneName=tent_2_r_4}
    componentLinearAccScale {z=0.44}
    componentLinearVelScale {z=0.44}
    physicsBodyDefinitions [{
      collisionType=CustomSphere
      sphereCollisionRadius=2
      boundBone {boneName=tent_2_r_4}
      boxExtents=<12 20 20>
      constraintSetup {angularLimitsMax=<0 10 60>, angularLimitsMin=<0 -10 -60>, angularTarget=<0 0 0>}
      localJointOffset=<10 0 0>
     }]
    sphericalLimits [
     {limitRadius=34, drivingBone {boneName=spine_05}, sphereLocalOffset=<-1.6787 12.8914 0>}
     {limitRadius=16, drivingBone {boneName=clavicle_l}, sphereLocalOffset=<15.524 -1.769 -0.679>}
     {limitRadius=16, drivingBone {boneName=clavicle_r}, sphereLocalOffset=<-15.524 -1.769 0.679>}
    ]
   }
   n143 = Anim Dynamics - Bone: tent_2_r_5(ComponentPose=@n142, Alpha=@n117, ExternalForce="0,0,0") {
    angularDampingOverride=1
    angularSpringConstant=1500
    bAngularSpring=true
    bGravityOverrideInSimSpace=true
    bOverrideAngularDamping=true
    bUseSphericalLimits=true
    gravityScale=0.35
    simulationSpace=RootRelative
    alphaBoolBlend {blendOption=Linear}
    alphaScaleBiasClamp {bInterpResult=true}
    boundBone {boneName=tent_2_r_5}
    componentLinearAccScale {z=0.45}
    componentLinearVelScale {z=0.45}
    physicsBodyDefinitions [{
      collisionType=CustomSphere
      sphereCollisionRadius=2
      boundBone {boneName=tent_2_r_5}
      boxExtents=<12 20 20>
      constraintSetup {angularLimitsMax=<0 10 60>, angularLimitsMin=<0 -10 -60>, angularTarget=<0 0 0>}
      localJointOffset=<10 0 0>
     }]
    sphericalLimits [
     {limitRadius=34, drivingBone {boneName=spine_05}, sphereLocalOffset=<-1.6787 12.8914 0>}
     {limitRadius=16, drivingBone {boneName=clavicle_l}, sphereLocalOffset=<15.524 -1.769 -0.679>}
     {limitRadius=16, drivingBone {boneName=clavicle_r}, sphereLocalOffset=<-15.524 -1.769 0.679>}
    ]
   }
   n144 = Anim Dynamics - Bone: tent_2_r_6(ComponentPose=@n143, Alpha=@n117, ExternalForce="0,0,0") {
    angularDampingOverride=1
    angularSpringConstant=1500
    bAngularSpring=true
    bGravityOverrideInSimSpace=true
    bOverrideAngularDamping=true
    bUseSphericalLimits=true
    gravityScale=0.35
    simulationSpace=RootRelative
    alphaBoolBlend {blendOption=Linear}
    alphaScaleBiasClamp {bInterpResult=true}
    boundBone {boneName=tent_2_r_6}
    componentLinearAccScale {z=0.45}
    componentLinearVelScale {z=0.45}
    physicsBodyDefinitions [{
      collisionType=CustomSphere
      sphereCollisionRadius=2
      boundBone {boneName=tent_2_r_6}
      boxExtents=<12 20 20>
      constraintSetup {angularLimitsMax=<0 10 60>, angularLimitsMin=<0 -10 -60>, angularTarget=<1 0 0>}
      localJointOffset=<10 0 0>
     }]
    sphericalLimits [
     {limitRadius=34, drivingBone {boneName=spine_05}, sphereLocalOffset=<-1.6787 12.8914 0>}
     {limitRadius=16, drivingBone {boneName=clavicle_l}, sphereLocalOffset=<15.524 -1.769 -0.679>}
     {limitRadius=16, drivingBone {boneName=clavicle_r}, sphereLocalOffset=<-15.524 -1.769 0.679>}
    ]
   }
   n145 = Anim Dynamics - Bone: tent_2_r_7(ComponentPose=@n144, Alpha=@n117, ExternalForce="0,0,0") {
    angularDampingOverride=1
    angularSpringConstant=1500
    bAngularSpring=true
    bGravityOverrideInSimSpace=true
    bOverrideAngularDamping=true
    bUseSphericalLimits=true
    gravityScale=0.35
    simulationSpace=RootRelative
    alphaBoolBlend {blendOption=Linear}
    alphaScaleBiasClamp {bInterpResult=true}
    boundBone {boneName=tent_2_r_7}
    componentLinearAccScale {z=0.46}
    componentLinearVelScale {z=0.46}
    physicsBodyDefinitions [{
      collisionType=CustomSphere
      sphereCollisionRadius=2
      boundBone {boneName=tent_2_r_7}
      boxExtents=<12 20 20>
      constraintSetup {angularLimitsMax=<0 10 60>, angularLimitsMin=<0 -10 -60>, angularTarget=<1 0 0>}
      localJointOffset=<10 0 0>
     }]
    sphericalLimits [
     {limitRadius=35, drivingBone {boneName=spine_05}, sphereLocalOffset=<-1.6787 12.8914 0>}
     {limitRadius=16, drivingBone {boneName=clavicle_l}, sphereLocalOffset=<15.524 -1.769 -0.679>}
     {limitRadius=16, drivingBone {boneName=clavicle_r}, sphereLocalOffset=<-15.524 -1.769 0.679>}
    ]
   }
   n146 = Anim Dynamics - Bone: tent_2_r_8(ComponentPose=@n145, Alpha=@n117, ExternalForce="0,0,0") {
    angularDampingOverride=1
    angularSpringConstant=1500
    bAngularSpring=true
    bGravityOverrideInSimSpace=true
    bOverrideAngularDamping=true
    bUseSphericalLimits=true
    gravityScale=0.35
    simulationSpace=RootRelative
    alphaBoolBlend {blendOption=Linear}
    alphaScaleBiasClamp {bInterpResult=true}
    boundBone {boneName=tent_2_r_8}
    componentLinearAccScale {z=0.46}
    componentLinearVelScale {z=0.46}
    physicsBodyDefinitions [{
      collisionType=CustomSphere
      sphereCollisionRadius=2
      boundBone {boneName=tent_2_r_8}
      boxExtents=<12 20 20>
      constraintSetup {angularLimitsMax=<0 10 60>, angularLimitsMin=<0 -10 -60>, angularTarget=<0 0 0>}
      localJointOffset=<10 0 0>
     }]
    sphericalLimits [
     {limitRadius=35, drivingBone {boneName=spine_05}, sphereLocalOffset=<-1.6787 12.8914 0>}
     {limitRadius=16, drivingBone {boneName=clavicle_l}, sphereLocalOffset=<15.524 -1.769 -0.679>}
     {limitRadius=16, drivingBone {boneName=clavicle_r}, sphereLocalOffset=<-15.524 -1.769 0.679>}
    ]
   }
   n147 = Anim Dynamics - Bone: tent_2_r_9(ComponentPose=@n146, Alpha=@n117, ExternalForce="0,0,0") {
    angularDampingOverride=1
    angularSpringConstant=1500
    bAngularSpring=true
    bGravityOverrideInSimSpace=true
    bOverrideAngularDamping=true
    bUseSphericalLimits=true
    gravityScale=0.35
    simulationSpace=RootRelative
    alphaBoolBlend {blendOption=Linear}
    alphaScaleBiasClamp {bInterpResult=true}
    boundBone {boneName=tent_2_r_9}
    componentLinearAccScale {z=0.46}
    componentLinearVelScale {z=0.46}
    physicsBodyDefinitions [{
      collisionType=CustomSphere
      sphereCollisionRadius=2
      boundBone {boneName=tent_2_r_9}
      boxExtents=<12 20 20>
      constraintSetup {angularLimitsMax=<0 10 60>, angularLimitsMin=<0 -10 -60>, angularTarget=<0 0 0>}
      localJointOffset=<10 0 0>
     }]
    sphericalLimits [
     {limitRadius=35, drivingBone {boneName=spine_05}, sphereLocalOffset=<-1.6787 12.8914 0>}
     {limitRadius=16, drivingBone {boneName=clavicle_l}, sphereLocalOffset=<15.524 -1.769 -0.679>}
     {limitRadius=16, drivingBone {boneName=clavicle_r}, sphereLocalOffset=<-15.524 -1.769 0.679>}
    ]
   }
   n148 = Anim Dynamics - Bone: tent_1_r_2(ComponentPose=@n147, Alpha=1, ExternalForce="0,0,0") {
    angularDampingOverride=1
    angularSpringConstant=1500
    bAngularSpring=true
    bGravityOverrideInSimSpace=true
    bOverrideAngularDamping=true
    bUseSphericalLimits=true
    gravityScale=0.35
    simulationSpace=RootRelative
    alphaBoolBlend {blendOption=Linear}
    alphaScaleBiasClamp {scale=0.3}
    boundBone {boneName=tent_1_r_2}
    componentLinearAccScale {z=0.43}
    componentLinearVelScale {z=0.43}
    physicsBodyDefinitions [{
      collisionType=CustomSphere
      sphereCollisionRadius=2
      boundBone {boneName=tent_1_r_2}
      boxExtents=<9 20 20>
      constraintSetup {angularLimitsMax=<0 10 60>, angularLimitsMin=<0 -10 -60>, angularTarget=<1 0 0>}
      localJointOffset=<10 0 0>
     }]
    sphericalLimits [
     {limitRadius=34, drivingBone {boneName=spine_05}, sphereLocalOffset=<-1.6787 12.8914 0>}
     {limitRadius=16, drivingBone {boneName=clavicle_l}, sphereLocalOffset=<15.524 -1.769 -0.679>}
     {limitRadius=16, drivingBone {boneName=clavicle_r}, sphereLocalOffset=<-15.524 -1.769 0.679>}
    ]
   }
   n149 = Anim Dynamics - Bone: tent_1_r_3(ComponentPose=@n148, Alpha=1, ExternalForce="0,0,0") {
    angularDampingOverride=1
    angularSpringConstant=1500
    bAngularSpring=true
    bGravityOverrideInSimSpace=true
    bOverrideAngularDamping=true
    bUseSphericalLimits=true
    gravityScale=0.35
    simulationSpace=RootRelative
    alphaBoolBlend {blendOption=Linear}
    alphaScaleBiasClamp {scale=0.6}
    boundBone {boneName=tent_1_r_3}
    componentLinearAccScale {z=0.44}
    componentLinearVelScale {z=0.44}
    physicsBodyDefinitions [{
      collisionType=CustomSphere
      sphereCollisionRadius=2
      boundBone {boneName=tent_1_r_3}
      boxExtents=<9 20 20>
      constraintSetup {angularLimitsMax=<0 10 60>, angularLimitsMin=<0 -10 -60>, angularTarget=<1 0 0>}
      localJointOffset=<10 0 0>
     }]
    sphericalLimits [
     {limitRadius=34, drivingBone {boneName=spine_05}, sphereLocalOffset=<-1.6787 12.8914 0>}
     {limitRadius=16, drivingBone {boneName=clavicle_l}, sphereLocalOffset=<15.524 -1.769 -0.679>}
     {limitRadius=16, drivingBone {boneName=clavicle_r}, sphereLocalOffset=<-15.524 -1.769 0.679>}
    ]
   }
   n150 = Anim Dynamics - Bone: tent_1_r_4(ComponentPose=@n149, Alpha=@n117, ExternalForce="0,0,0") {
    angularDampingOverride=1
    angularSpringConstant=1500
    bAngularSpring=true
    bGravityOverrideInSimSpace=true
    bOverrideAngularDamping=true
    bUseSphericalLimits=true
    gravityScale=0.35
    simulationSpace=RootRelative
    alphaBoolBlend {blendOption=Linear}
    alphaScaleBiasClamp {bInterpResult=true}
    boundBone {boneName=tent_1_r_4}
    componentLinearAccScale {z=0.44}
    componentLinearVelScale {z=0.44}
    physicsBodyDefinitions [{
      collisionType=CustomSphere
      sphereCollisionRadius=2
      boundBone {boneName=tent_1_r_4}
      boxExtents=<9 20 20>
      constraintSetup {angularLimitsMax=<0 10 60>, angularLimitsMin=<0 -10 -60>, angularTarget=<0 0 0>}
      localJointOffset=<10 0 0>
     }]
    sphericalLimits [
     {limitRadius=34, drivingBone {boneName=spine_05}, sphereLocalOffset=<-1.6787 12.8914 0>}
     {limitRadius=16, drivingBone {boneName=clavicle_l}, sphereLocalOffset=<15.524 -1.769 -0.679>}
     {limitRadius=16, drivingBone {boneName=clavicle_r}, sphereLocalOffset=<-15.524 -1.769 0.679>}
    ]
   }
   n151 = Anim Dynamics - Bone: tent_1_r_5(ComponentPose=@n150, Alpha=@n117, ExternalForce="0,0,0") {
    angularDampingOverride=1
    angularSpringConstant=1500
    bAngularSpring=true
    bGravityOverrideInSimSpace=true
    bOverrideAngularDamping=true
    bUseSphericalLimits=true
    gravityScale=0.35
    simulationSpace=RootRelative
    alphaBoolBlend {blendOption=Linear}
    alphaScaleBiasClamp {bInterpResult=true}
    boundBone {boneName=tent_1_r_5}
    componentLinearAccScale {z=0.45}
    componentLinearVelScale {z=0.45}
    physicsBodyDefinitions [{
      collisionType=CustomSphere
      sphereCollisionRadius=2
      boundBone {boneName=tent_1_r_5}
      boxExtents=<9 20 20>
      constraintSetup {angularLimitsMax=<0 10 60>, angularLimitsMin=<0 -10 -60>, angularTarget=<0 0 0>}
      localJointOffset=<10 0 0>
     }]
    sphericalLimits [
     {limitRadius=34, drivingBone {boneName=spine_05}, sphereLocalOffset=<-1.6787 12.8914 0>}
     {limitRadius=16, drivingBone {boneName=clavicle_l}, sphereLocalOffset=<15.524 -1.769 -0.679>}
     {limitRadius=16, drivingBone {boneName=clavicle_r}, sphereLocalOffset=<-15.524 -1.769 0.679>}
    ]
   }
   n152 = Anim Dynamics - Bone: tent_1_r_6(ComponentPose=@n151, Alpha=@n117, ExternalForce="0,0,0") {
    angularDampingOverride=1
    angularSpringConstant=1500
    bAngularSpring=true
    bGravityOverrideInSimSpace=true
    bOverrideAngularDamping=true
    bUseSphericalLimits=true
    gravityScale=0.35
    simulationSpace=RootRelative
    alphaBoolBlend {blendOption=Linear}
    alphaScaleBiasClamp {bInterpResult=true}
    boundBone {boneName=tent_1_r_6}
    componentLinearAccScale {z=0.45}
    componentLinearVelScale {z=0.45}
    physicsBodyDefinitions [{
      collisionType=CustomSphere
      sphereCollisionRadius=2
      boundBone {boneName=tent_1_r_6}
      boxExtents=<9 20 20>
      constraintSetup {angularLimitsMax=<0 10 60>, angularLimitsMin=<0 -10 -60>, angularTarget=<1 0 0>}
      localJointOffset=<10 0 0>
     }]
    sphericalLimits [
     {limitRadius=34, drivingBone {boneName=spine_05}, sphereLocalOffset=<-1.6787 12.8914 0>}
     {limitRadius=16, drivingBone {boneName=clavicle_l}, sphereLocalOffset=<15.524 -1.769 -0.679>}
     {limitRadius=16, drivingBone {boneName=clavicle_r}, sphereLocalOffset=<-15.524 -1.769 0.679>}
    ]
   }
   n153 = Anim Dynamics - Bone: tent_1_r_7(ComponentPose=@n152, Alpha=@n117, ExternalForce="0,0,0") {
    angularDampingOverride=1
    angularSpringConstant=1500
    bAngularSpring=true
    bGravityOverrideInSimSpace=true
    bOverrideAngularDamping=true
    bUseSphericalLimits=true
    gravityScale=0.35
    simulationSpace=RootRelative
    alphaBoolBlend {blendOption=Linear}
    alphaScaleBiasClamp {bInterpResult=true}
    boundBone {boneName=tent_1_r_7}
    componentLinearAccScale {z=0.46}
    componentLinearVelScale {z=0.46}
    physicsBodyDefinitions [{
      collisionType=CustomSphere
      sphereCollisionRadius=2
      boundBone {boneName=tent_1_r_7}
      boxExtents=<9 20 20>
      constraintSetup {angularLimitsMax=<0 10 60>, angularLimitsMin=<0 -10 -60>, angularTarget=<1 0 0>}
      localJointOffset=<10 0 0>
     }]
    sphericalLimits [
     {limitRadius=35, drivingBone {boneName=spine_05}, sphereLocalOffset=<-1.6787 12.8914 0>}
     {limitRadius=16, drivingBone {boneName=clavicle_l}, sphereLocalOffset=<15.524 -1.769 -0.679>}
     {limitRadius=16, drivingBone {boneName=clavicle_r}, sphereLocalOffset=<-15.524 -1.769 0.679>}
    ]
   }
   n154 = Anim Dynamics - Bone: tent_1_r_8(ComponentPose=@n153, Alpha=@n117, ExternalForce="0,0,0") {
    angularDampingOverride=1
    angularSpringConstant=1500
    bAngularSpring=true
    bGravityOverrideInSimSpace=true
    bOverrideAngularDamping=true
    bUseSphericalLimits=true
    gravityScale=0.35
    simulationSpace=RootRelative
    alphaBoolBlend {blendOption=Linear}
    alphaScaleBiasClamp {bInterpResult=true}
    boundBone {boneName=tent_1_r_8}
    componentLinearAccScale {z=0.46}
    componentLinearVelScale {z=0.46}
    physicsBodyDefinitions [{
      collisionType=CustomSphere
      sphereCollisionRadius=2
      boundBone {boneName=tent_1_r_8}
      boxExtents=<9 20 20>
      constraintSetup {angularLimitsMax=<0 10 60>, angularLimitsMin=<0 -10 -60>, angularTarget=<0 0 0>}
      localJointOffset=<10 0 0>
     }]
    sphericalLimits [
     {limitRadius=35, drivingBone {boneName=spine_05}, sphereLocalOffset=<-1.6787 12.8914 0>}
     {limitRadius=16, drivingBone {boneName=clavicle_l}, sphereLocalOffset=<15.524 -1.769 -0.679>}
     {limitRadius=16, drivingBone {boneName=clavicle_r}, sphereLocalOffset=<-15.524 -1.769 0.679>}
    ]
   }
   n155 = Anim Dynamics - Bone: tent_1_r_9(ComponentPose=@n154, Alpha=1, ExternalForce="0,0,0") {
    angularDampingOverride=1
    angularSpringConstant=1500
    bAngularSpring=true
    bGravityOverrideInSimSpace=true
    bOverrideAngularDamping=true
    bUseSphericalLimits=true
    gravityScale=0.35
    simulationSpace=RootRelative
    alphaBoolBlend {blendOption=Linear}
    alphaScaleBiasClamp {bInterpResult=true}
    boundBone {boneName=tent_1_r_9}
    componentLinearAccScale {z=0.46}
    componentLinearVelScale {z=0.46}
    physicsBodyDefinitions [{
      collisionType=CustomSphere
      sphereCollisionRadius=2
      boundBone {boneName=tent_1_r_9}
      boxExtents=<9 20 20>
      constraintSetup {angularLimitsMax=<0 10 60>, angularLimitsMin=<0 -10 -60>, angularTarget=<0 0 0>}
      localJointOffset=<10 0 0>
     }]
    sphericalLimits [
     {limitRadius=35, drivingBone {boneName=spine_05}, sphereLocalOffset=<-1.6787 12.8914 0>}
     {limitRadius=16, drivingBone {boneName=clavicle_l}, sphereLocalOffset=<15.524 -1.769 -0.679>}
     {limitRadius=16, drivingBone {boneName=clavicle_r}, sphereLocalOffset=<-15.524 -1.769 0.679>}
    ]
   }
   Output Pose(Result=Blend Poses by bool(
     BlendPose_0=Component To Local(ComponentPose=@n73)
     BlendPose_1=Component To Local(ComponentPose=@n155)
     bActiveValue=Get isWearingBackbling()
    ) {blendType=Linear}) {layerGroup=None}
   Save cached pose 'PreBezierCache'(Pose=Copy Pose From Mesh() {bUseAttachedParent=true}) {cacheName=PreBezierCache}
  ]}
 EventGraph {nodes [
   Event BlueprintUpdateAnimation()
   n1 = GetAnimInstance(self=GetOwnerSkeletalMeshComponent())
   > Is Valid(InputObject=@n1)
   > n2 = Branch(Condition=Montage_IsPlaying(self=@n1, Montage=/Game/Animation/Game/MainPlayer/Emotes/CattusRoar/Emote_CattusRoar_CMM_M))
   > Do Once(Reset=@n3.then)
   > Set isRoaringAlpha(isRoaringAlpha:double=1) -> (Output_Get:double)
   > Set isRoaringDynamicsAlpha(isRoaringDynamicsAlpha:double=0.4) -> (Output_Get:double)
   > Set isNotRoaringAlpha(isNotRoaringAlpha:double=0) -> (Output_Get:double)
   > Delay(Duration=4)
   > Set isRoaringAlpha(isRoaringAlpha=0)
   > Set isRoaringDynamicsAlpha(isRoaringDynamicsAlpha=1)
   > Set isNotRoaringAlpha(isNotRoaringAlpha=1)
   Set isRoaringAlpha(execute=@n2.else, isRoaringAlpha=0)
   > Set isRoaringDynamicsAlpha(isRoaringDynamicsAlpha=1)
   > n3 = Set isNotRoaringAlpha(isNotRoaringAlpha=1)
   Event OnCharacterCustomizationLoaded()
   n4 = CheckForBackbling()
   > Set isWearingBackbling(isWearingBackbling=AND Boolean(A=@n4, B=NOT Boolean(A=@n4.bIsBackblingLong)))
  ]}
]
variables [
 isNotRoaringAlpha {type=number, default=1}
 isRoaringAlpha {type=number}
 isRoaringDynamicsAlpha {type=number, default=1}
 isWearingBackbling {type=boolean}
]
